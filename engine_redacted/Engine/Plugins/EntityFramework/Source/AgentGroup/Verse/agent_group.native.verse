# Copyright Epic Games, Inc. All Rights Reserved.

using { /Verse.org/Native }
using { /Verse.org/Simulation }

@doc("Interface that defines a class as being usable as member info in an agent group")
@available{MinUploadedAtFNVersion := 4000}
member_info_interface<public><native> := interface{}

@available{MinUploadedAtFNVersion := 4000}
@doc("Interface that defines a class as providing an agent group.")
agent_group_interface<public><native>(member_info:subtype(member_info_interface)) := interface<unique>:
    
    @doc("Get the members of this agent group with their member info")
    GetMemberMap<public>()<reads> : [agent]member_info

    @doc("Signalled whenever an agent successfully joins this agent group.{
    }Passes in the agent that joined along with their member_info.")
    AddMemberEvent<public> : listenable(tuple(agent, member_info))

    @doc("Signalled whenever an agent successfully leaves this agent group.{
    }Passes in the agent that left along with their member_info.")
    RemoveMemberEvent<public> : listenable(tuple(agent, member_info))

    @doc("Signalled whenever the MemberInfo class is re-instantiated for a given agent.{
    }Passes in the agent that was updated along with their new member_info.")
    MemberInfoChangeEvent<public> : listenable(tuple(agent, member_info))

@doc("An agent group is defined as a set of agents that share a common ownership.{
}This class stores agents and specific information about each member via the member_info type provided.")
@available{MinUploadedAtFNVersion := 4000}
agent_group<public><native>(member_info:subtype(member_info_interface)) := class<unique>(agent_group_interface(member_info)):

    @doc("Get the members of this agent group with their member info")
    GetMemberMap<override><native_callable>()<reads> : [agent]member_info = MemberMap

    @doc("Attempt to add the given agent to this agent group.{
    }This function returns a result that will either succeed or return an error.")
    AddMember<public>(Agent:agent, MemberInfo:member_info)<transacts> : result(void, add_member_error) =
        if:
            # Already a member, so this is a member_info replacement rather than a join.
            MemberMap[Agent]
            set MemberMap[Agent] = MemberInfo
        then:
            SignalMemberInfoChangeEvent(Agent, MemberInfo)
            MakeSuccess()
        else if (set MemberMap[Agent] = MemberInfo):
            SignalAddMemberEvent(Agent, MemberInfo)
            MakeSuccess()
        else:
            MakeError(add_member_error {})

    @doc("Attempt to remove the given agent from this agent group.{
    }This function returns a result that will either succeed or return an error.")
    RemoveMember<public>(Agent:agent)<transacts> : result(void, remove_member_error) =
        if (ExistingMemberInfo := MemberMap[Agent]):
            RemoveFromMemberMap(Agent)
            SignalRemoveMemberEvent(Agent, ExistingMemberInfo)
        # If the member map does not contain the agent, return success as the end result is the same.
        MakeSuccess()

    @doc("Signalled whenever an agent successfully joins this agent group.{
    }Passes in the agent that joined along with their member_info.")
    AddMemberEvent<override><native><final> : listenable(tuple(agent, member_info)) = redacted_internal(tuple(agent, member_info)){}

    @doc("Signalled whenever an agent successfully leaves this agent group.{
    }Passes in the agent that left along with their member_info.")
    RemoveMemberEvent<override><native><final> : listenable(tuple(agent, member_info)) = redacted_internal(tuple(agent, member_info)){}

    @doc("Signalled whenever the MemberInfo class is re-instantiated for a given agent.{
    }Passes in the agent that was updated along with their new member_info.")
    MemberInfoChangeEvent<override><native><final> : listenable(tuple(agent, member_info)) = redacted_internal(tuple(agent, member_info)){}

    # Removal is native as that can modify the map directly rather than make a copy
    RemoveFromMemberMap<private><native>(Agent:agent)<transacts> : void

    # `no_rollback` effect, which a <transacts> function cannot call. Delivery failures are not reported back.
    SignalAddMemberEvent<private><native>(Agent:agent, MemberInfo:member_info)<transacts> : void
    SignalRemoveMemberEvent<private><native>(Agent:agent, MemberInfo:member_info)<transacts> : void
    SignalMemberInfoChangeEvent<private><native>(Agent:agent, MemberInfo:member_info)<transacts> : void


@available{MinUploadedAtFNVersion := 4000}
@doc("Base class for all errors returned from agent_group.AddMember.")
add_member_error<native><public> := class<computes>():

@available{MinUploadedAtFNVersion := 4000}
@doc("Base class for all errors returned from agent_group.RemoveMember.")
remove_member_error<native><public> := class<computes>():