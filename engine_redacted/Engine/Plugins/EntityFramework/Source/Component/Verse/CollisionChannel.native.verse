# Copyright Epic Games, Inc. All Rights Reserved.

using { /Verse.org/Native }
using { /Verse.org/SpatialMath }
using { /Verse.org/Verse }


@doc("Specifies how a collision volume pair should interact. See collision_profile.")
collision_interaction<public><native> := enum:
    @doc("{
    }The pair will not be detected by Overlap and Sweep queries. {
    }The pair will not collide in the physics simulation.")
    Ignore,

    @doc("{
    }The pair will be detected by Overlap and Sweep queries. {
    }The pair will not collide in the physics simulation.")
    Overlap,
    
    @doc("{
    }The pair will be detected by Overlap and Sweep queries. {
    }The pair will collide in the physics simulation.")
    Block

@doc("{
}Every volume has a collision channel as part of its collision_profile. {
}It is used to determine how two volumes interact. See collision_profile.")
collision_channel<public><native> := class<concrete>() {}

@doc("An alias for the function type used to map a collision_channel to a collision_interaction used by  collision_profile.")
collision_channel_to_interaction<public> := type{_(InChannel:collision_channel)<computes>:collision_interaction}


@doc("{
}A collision profile determines how a volume interacts with other volumes {
}for Overlap queries, Sweep queries, and physics simulation. When two volumes {
}are being tested to see how they interact, the algorithm looks like this:\n{
}   GetInteraction(A:collision_profile, B:collision_profile):collision_interaction = \n{
}       InteractionA = B.GetChannelInteraction(A.Channel) \n{
}       InteractionB = A.GetChannelInteraction(B.Channel) \n{
}       return Min(InteractionA, InteractionB) \n")
collision_profile<public><native> := class<concrete>():

    @doc("The collision channel for the owning object.")
    Channel<public><native>:collision_channel = collision_channel{}

    @doc("{
    }How the owning object should interact with other objects.{
    }GetChannelInteraction is a function which maps a collision_channel to a {
    }collision_interaction. It can be implemented as an simple sequence of if {
    }statements. For example, to block all channels except camera:\n{
    }    BlockAllIgnoreCamera(Channel:collision_channel)<computes>:collision_interaction =\n{
    }        if (CollisionChannels.camera[Channel]):\n{
    }            return collision_interaction.Ignore\n{
    }        return collision_interaction.Block\n{
    }    MyProfile<public>:collision_profile = MakeCollisionProfile(CollisionChannels.dynamic, BlockAllIgnoreCamera)")
    GetChannelInteraction<public><native>:collision_channel_to_interaction = RedactedInternal

# [PUBLIC SEMI-HACK] A workaround to allow creation of named collision_profiles at module scope
@doc("Create a collision_profile. See the CollisionProfiles module.")
MakeCollisionProfile<internal><native>(channel_type:subtype(collision_channel), ChannelToInteraction:collision_channel_to_interaction)<converges>:collision_profile

@doc("The set of built-in collision_channels. This is a closed set for now.")
CollisionChannels<public> := module:
    stationary<public><native> := class(collision_channel) {}

    dynamic<public><native> := class(collision_channel) {}

    avatar<public><native> := class(collision_channel) {}

    visibility<public><native> := class(collision_channel) {}

    camera<public><native> := class(collision_channel) {}

    physics<public><native> := class(collision_channel) {}


@doc("A set of useful collision profiles. New profiles can be created in Verse code")
CollisionProfiles<public> := module:
    # Private utility methods for profile specifiers below (but private is not allowed at module scope)
    IgnoreAllInteraction<internal>(:collision_channel)<computes>:collision_interaction = collision_interaction.Ignore

    OverlapAllInteraction<internal>(:collision_channel)<computes>:collision_interaction = collision_interaction.Overlap

    BlockAllInteraction<internal>(:collision_channel)<computes>:collision_interaction = collision_interaction.Block

    BlockVisibleInteraction<internal>(Channel:collision_channel)<computes>:collision_interaction =
        if (CollisionChannels.visibility[Channel]):
            return collision_interaction.Ignore
        return collision_interaction.Block

    @doc("Stationary collision channel, Ignore all channels")
    StationaryIgnoreAll<public>:collision_profile = MakeCollisionProfile(CollisionChannels.stationary, IgnoreAllInteraction)

    @doc("Stationary collision channel, Overlap all channels")
    StationaryOverlapAll<public>:collision_profile = MakeCollisionProfile(CollisionChannels.stationary, OverlapAllInteraction)

    @doc("Stationary collision channel, Block all channels")
    StationaryBlockAll<public>:collision_profile = MakeCollisionProfile(CollisionChannels.stationary, BlockAllInteraction)

    @doc("Dynamic collision channel, Ignore all channels")
    DynamicIgnoreAll<public>:collision_profile = MakeCollisionProfile(CollisionChannels.dynamic, IgnoreAllInteraction)

    @doc("Dynamic collision channel, Overlap all channels")
    DynamicOverlapAll<public>:collision_profile = MakeCollisionProfile(CollisionChannels.dynamic, OverlapAllInteraction)

    @doc("Dynamic collision channel, Block all channels")
    DynamicBlockAll<public>:collision_profile = MakeCollisionProfile(CollisionChannels.dynamic, BlockAllInteraction)

    @doc("Stationary collision channel, Block all channels except visibility (used for e.g., invisible wall, glass)")
    StationaryBlockVisible<public>:collision_profile = MakeCollisionProfile(CollisionChannels.stationary, BlockVisibleInteraction)

    @doc("Visibility collision channel, Overlap all channel (used for e.g., visibility testing)")
    VisibilityOverlapAll<public>:collision_profile = MakeCollisionProfile(CollisionChannels.visibility, OverlapAllInteraction)
