# Copyright Epic Games, Inc. All Rights Reserved.

@deprecated{Message := "Use `CollisionChannels` instead."}
collision_channels<public> := module {}

@deprecated{Message := "Use `CollisionProfiles` instead."}
collision_profiles<public> := module {}
