# Copyright Epic Games, Inc. All Rights Reserved.

using { /Verse.org/Native }
using { /Verse.org/SpatialMath }

@doc("{
}The results of an overlap query. See entity.FindOverlapHits(). We will get one overlap_hit for {
}each intersection of any volume in SourceVolumes with any other volume.")
overlap_hit<public><native> := struct:

    @doc("{
    }The source component and volume (query input). For compound inputs (like an entity hierarchy) {
    }this will be a component/volume in that hierarchy. The SourceTransform is the transform of {
    }SourceVolume used for the overlap test. For single volume inputs like a sphere, the Source {
    }volume and transform are just the inputs to the overlap test, and the component is false.")
    SourceComponent<public><native>:?component

    @doc("The source volume (query input)")
    SourceVolume<public><native>:collision_volume

    @doc("The source volume transform")
    SourceGlobalTransform<public><native>:transform

    @doc("The component that was hit by SourceVolume")
    TargetComponent<public><native>:component

    @doc("The volume that was hit by SourceVolume")
    TargetVolume<public><native>:collision_element

    @available{MinUploadedAtFNVersion := 4300}
    @doc("Hit data for the source mesh")

    @available{MinUploadedAtFNVersion := 4300}
    @doc("Hit data for the target mesh")

    #@doc("Axis aligned world-space bounds, guaranteed to contain the intersection of the source and target meshes")

@doc("{
}The results of a sweep query. See entity.FindSweepHits(). We will get one sweep_hit for {
}each intersection of any volume in SourceVolumes with any other volume.")
sweep_hit<public><native> := struct:

    @doc("{
    }The source component and volume (query input). For compound inputs (like an entity hierarchy) {
    }this will be a component/volume in that hierarchy. The SourceGlobalTransform is the transform of {
    }SourceVolume at the start of the sweep. For single volume inputs like a sphere, the volume and {
    }transform are just the inputs to the sweep, and the component is false.")
    SourceComponent<public><native>:?component

    @doc("The source volume (query input).")
    SourceVolume<public><native>:collision_volume

    @doc("The source volume transform at the start of the sweep.")
    SourceStartGlobalTransform<public><native>:transform

    @doc("{
    }The world-space translation (relative to SourceStartGlobalTransform) {
    }of SourceVolume when it touches TargetVolume.")
    SourceHitTranslation<public><native>:vector3 = vector3{}

    @doc("The component that was hit by SourceVolume.")
    TargetComponent<public><native>:component

    @doc("The volume that was hit by SourceVolume.")
    TargetVolume<public><native>:collision_element

    @available{MinUploadedAtFNVersion := 4300}
    @doc("Hit data for the source mesh")

    @available{MinUploadedAtFNVersion := 4300}
    @doc("Hit data for the target mesh")

    @doc("The Distance along the sweep at which SourceVolume touches TargetVolume.")
    SourceHitDistance<public><native>:float = 0.0

    @doc("The point of contact between SourceVolume and TargetVolume.")
    ContactPosition<public><native>:vector3 = vector3{}

    @doc("The normal on TargetVolume at the HitPosition.")
    ContactNormal<public><native>:vector3 = vector3{}

    @doc("{
    }If TargetVolume is a polygonal object (mesh, convex hull, etc.) {
    }and the contact point is on an edge or vertex, this is the {
    }most-opposing face normal of the faces that share that edge {
    }or vertex. Otherwise it is the same as HitNormal.")
    ContactFaceNormal<public><native>:vector3 = vector3{}

