# Copyright Epic Games, Inc. All Rights Reserved.

using { /Verse.org/Native }
using { /Verse.org/Simulation }
using { /Verse.org/SpatialMath }
using { /Verse.org/Temporary/EpicGamesRestricted/Network }

@doc("{
}Collision Volumes represent the collision shapes of meshes. They can be detected {
}by Overlap and Sweep queries and generate collisions in the physics simulation.")
collision_volume<public><native> := class<abstract><predicts>():

    @doc("Enable/disable collision on this volume.")
    var Collidable<public><getter(RedactedInternal)><setter(RedactedInternal)>:logic = external{}

    @doc("Enable/disable spatial queries against this volume.")
    var Queryable<public><getter(GetQueryable)><setter(SetQueryable)>:logic = external{}

    @doc("Get the transform of this volume in the space of its owner (usually a component on an entity)")
    GetLocalTransform<public>()<transacts>:transform = 
        LocalTransformInternal

    @doc("Set the transform of this volume in the space of its owner (usually a component on an entity)")
    SetLocalTransform<public>(NewLocalTransform:transform)<transacts>:void = 
        set LocalTransformInternal = NewLocalTransform
        OnLocalTransformModified()

    # [LATER]
    #var TagContainer:tag_container

    # Begin Collidable implementation
    #CollidableToolTip<private><localizes>:message = "Enable/disable collision on this volume."
    @editable:
    #    ToolTip := CollidableToolTip
    @display_name("Collidable")
    @replicated("RepNotify")
    var CollidableInternal<native>:logic = true
    OnRep_CollidableInternal()<transacts>:void = 
        OnCollidableModified()
    OnCollidableModified<native>()<transacts>:void
    # End Collidable implementation

    # Begin Queryable implementation
    #QueryableToolTip<private><localizes>:message = "Enable/disable spatial queries against this volume."
    @editable:
    #    ToolTip := QueryableToolTip
    @display_name("Queryable")
    @replicated("RepNotify")
    var QueryableInternal<native>:logic = true
        QueryableInternal
        set QueryableInternal = NewQueryable
        OnQueryableModified()
    OnRep_QueryableInternal()<transacts>:void =
        OnQueryableModified()
    OnQueryableModified<native>()<transacts>:void
    # End Queryable implementation

    # Begin LocalTransform implementation
    #LocalTransformToolTip<private><localizes>:message = "The transform of this volume relative to its owner."
    @editable:
    #    ToolTip := LocalTransformToolTip
    @display_name("RepNotify=OnRep_LocalTransform")
    @replicated("RepNotify")
    var LocalTransformInternal<native>:transform = transform{}
    OnRep_LocalTransformInternal()<transacts>:void = 
        OnLocalTransformModified()
    OnLocalTransformModified<native>()<transacts>:void
    # End LocalTransform implementation


collision_element<public><native> := class<abstract><predicts>(collision_volume):

    @doc("The collision_profile for this volume.")
    var CollisionProfile<public><getter(GetCollisionProfile)><setter(SetCollisionProfile)>:collision_profile = external{}

    # [LATER]

    # Begin CollisionProfile implementation
    #CollisionProfileToolTip<private><localizes>:message = "The collision_profile for this volume."
    @editable:
    #    ToolTip := CollisionProfileToolTip
    @display_name("CollisionProfile")
    @replicated("RepNotify")
    var CollisionProfileInternal<internal><native>:collision_profile = CollisionProfiles.StationaryBlockAll
        CollisionProfileInternal
        set CollisionProfileInternal = NewCollisionProfile
        OnCollisionProfileModified()
    OnRep_CollisionProfileInternal()<transacts>:void =
        OnCollisionProfileModified()
    OnCollisionProfileModified<native>()<transacts>:void
    # End CollisionProfile implementation


@doc("A collision capsule aligned along the Z axis.")
collision_capsule<public><native> := class(collision_element):
    @doc("The radius of the capsule")
    var Radius<public><getter(GetRadius)><setter(SetRadius)>:float = external{}
    
    @doc("The length of the capsule's cylindrical section (distance between the two end cap centers)")
    var Length<public><getter(GetLength)><setter(SetLength)>:float  = external{}

    # Begin Radius implementation
    #RadiusToolTip<private><localizes>:message = "The capsule radius."
    @editable:
    #    ToolTip := RadiusToolTip
    @display_name("Radius")
    @replicated("RepNotify")
    var RadiusInternal<internal><native>:float = 50.0
        GetRadiusInternal()
        SetRadiusInternal(NewRadius)
    GetRadiusInternal()<transacts>:float = 
        RadiusInternal
    SetRadiusInternal(NewRadius:float)<transacts>:void =
        set RadiusInternal = NewRadius
        OnRadiusModified()
    OnRep_RadiusInternal()<transacts>:void =
        OnRadiusModified()
    OnRadiusModified<native>()<transacts>:void
    # End Radius implementation

    # Begin Length implementation
    #LengthToolTip<private><localizes>:message = "The capsule length."
    @editable:
    #    ToolTip := LengthToolTip
    @display_name("Length")
    @replicated("RepNotify")
    var LengthInternal<internal><native>:float = 100.0
        GetLengthInternal()
        SetLengthInternal(NewLength)
    GetLengthInternal()<transacts>:float = 
        LengthInternal
    SetLengthInternal(NewLength:float)<transacts>:void =
        set LengthInternal = NewLength
        OnLengthModified()
    OnRep_LengthInternal()<transacts>:void = 
        OnLengthModified()
    OnLengthModified<native>()<transacts>:void
    # End Length implementation


@doc("A collision sphere.")
collision_sphere<public><native> := class<predicts>(collision_element):
    @doc("The radius of the sphere")
    var Radius<public><getter(GetRadius)><setter(SetRadius)>:float = external{}

    # Begin Radius implementation
    #RadiusToolTip<private><localizes>:message = "The sphere radius."
    @editable:
    #    ToolTip := RadiusToolTip
    @display_name("Radius")
    @replicated("RepNotify")
    var RadiusInternal<internal><native>:float = 50.0
        GetRadiusInternal()
        SetRadiusInternal(NewRadius)
    GetRadiusInternal()<transacts>:float = 
        RadiusInternal
    SetRadiusInternal(NewRadius:float)<transacts>:void =
        set RadiusInternal = NewRadius
        OnRadiusModified()
    OnRep_RadiusInternal()<transacts>:void =
        OnRadiusModified()
    OnRadiusModified<native>()<transacts>:void
    # End Radius implementation


@doc("A collision point.")
collision_point<public><native> := class<concrete>(collision_element) {}


@doc("An axis-aligned collision box.")
collision_box<public><native> := class(collision_element):
    var Extents<public><getter(GetExtents)><setter(SetExtents)>:vector3

    # Begin Extents implementation
    #ExtentsToolTip<private><localizes>:message = "The box extents (half the size)."
    @editable:
    #    ToolTip := ExtentsToolTip
    @display_name("Extents")
    @replicated("RepNotify")
    var ExtentsInternal<native>:vector3 = vector3{Forward:=100.0, Left:=100.0, Up:=100.0}
        ExtentsInternal
        case(MemberName):
            "Forward" => return ExtentsInternal.Forward
            "Left" => return ExtentsInternal.Left
            "Up" => return ExtentsInternal.Up
            _ => return 0.0
        set ExtentsInternal = NewExtents
        OnExtentsModified()
        case(MemberName):
            "Forward" => set ExtentsInternal = vector3{Forward:=NewExtent, Left:=ExtentsInternal.Left, Up:=ExtentsInternal.Up}
            "Left" => set ExtentsInternal = vector3{Forward:=ExtentsInternal.Forward, Left:=NewExtent, Up:=ExtentsInternal.Up}
            "Up" => set ExtentsInternal = vector3{Forward:=ExtentsInternal.Forward, Left:=ExtentsInternal.Left, Up:=NewExtent}
            _ => {}
        OnExtentsModified()
    OnRep_ExtentsInternal()<transacts>:void = 
        OnExtentsModified()
    OnExtentsModified<native>()<transacts>:void
    # End Extents implementation
