# Copyright Epic Games, Inc. All Rights Reserved.

using { /Verse.org/Colors }
using { /Verse.org/Native }
using { /Verse.org/Simulation }
using { /Verse.org/Temporary/EpicGamesRestricted/Network }

@doc("A `directional_light_component` simulates light that is being emitted from a source that is infinitely {
     }far away. This means that all shadows cast by this light will be parallel, making this the ideal choice {
     }for simulating sunlight.")
directional_light_component<native><public> := class<final>(light_component):

    @doc("Intensity of the light hitting the surface. In Lux (Lumen per square meter).")
    var Illuminance<public><getter(GetIlluminance)><setter(SetIlluminance)>:float = external{}

    @doc("Angle subtended by light source in degrees (also known as angular diameter). Defaults to 0.5357 which is the angle for our sun.")
    var SourceAngleDegrees<public><getter(GetSourceAngleDegrees)><setter(SetSourceAngleDegrees)>:float = external{}

    # Illuminance Internals


    @editable
    @tool_tip("Intensity of the light hitting the surface. In Lux (Lumen per square meter).")
    @display_name("Illuminance")
    @replicated("RepNotify")
    var _Illuminance<private><native>:float = 10.0

    OnRep__Illuminance<native>():void

    # SourceAngleDegrees Internals


    @editable
    @ui_min("0.0")
    @ui_max("5.0")
    @tool_tip("Angle subtended by light source in degrees (also known as angular diameter).")
    @display_name("Source Angle Degrees")
    @replicated("RepNotify")
    var _SourceAngleDegrees<private><native>:float = 0.5357

    OnRep__SourceAngleDegrees<native>():void

    ###########################################################################
    # FN1 Internal Extensions BEGIN
    # TODO: Once categories are working and can be compiled, use that instead of the display name

    ####
    # Cascade Shadow Map settings
    ####


    @editable_slider(int):
        MinValue := option{0}
        MaxValue := option{4}

    @editable_slider(float):
        MinValue := option{1.0}
        MaxValue := option{4.0}

    @editable_slider(float):
        MinValue := option{0.0}
        MaxValue := option{0.3}

    @editable_slider(float):
        MinValue := option{0.0}
        MaxValue := option{1.0}

    @editable_slider(int):
        MinValue := option{0}
        MaxValue := option{4}


    ####
    # Light Shaft settings
    ####




    # FN1 Internal Extensions END
    ###########################################################################
