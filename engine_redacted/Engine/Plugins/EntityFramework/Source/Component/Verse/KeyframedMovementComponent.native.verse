# Copyright Epic Games, Inc. All Rights Reserved.

using { /Verse.org/Native }
using { /Verse.org/SceneGraph }
using { /Verse.org/Simulation }
using { /Verse.org/SpatialMath }
using { /Verse.org/Temporary/EpicGamesRestricted/Network }
using { /Verse.org/Verse }

@doc("Animate Scene Graph entities with keyframes.")
KeyframedMovement<public> := module:

    @doc("Base class for an animation easing function.")
    easing_function<native><public> := class<abstract><computes>:
        Evaluate<native><public>(Input:float):float

    @doc("Cubic bezier easing function. See CubicBezierEasingFunctions for some basic easing values.")
    cubic_bezier_easing_function<native><public> := class<concrete><computes>(easing_function):
        X0<native><public>:float = 0.0
        Y0<native><public>:float = 0.0
        X1<native><public>:float = 1.0
        Y1<native><public>:float = 1.0
        Evaluate<native><override>(Input:float):float

    @doc("`Linear` animations move at a constant speed.")
    linear_easing_function<native><public> := class<final><concrete><computes>(cubic_bezier_easing_function):
        X0<override>:float = 0.0
        Y0<override>:float = 0.0
        X1<override>:float = 1.0
        Y1<override>:float = 1.0

    @doc("`Ease` animations start slowly, speed up, then end slowly. The speed of the animation is slightly slower at the end than the start.")
    ease_cubic_bezier_easing_function<native><public> := class<final><concrete><computes>(cubic_bezier_easing_function):
        X0<override>:float = 0.25
        Y0<override>:float = 0.1
        X1<override>:float = 0.25
        Y1<override>:float = 1.0

    @doc("`EaseIn` animations start slow, then speed up towards the end.")
    ease_in_cubic_bezier_easing_function<native><public> := class<final><concrete><computes>(cubic_bezier_easing_function):
        X0<override>:float = 0.42
        Y0<override>:float = 0.0
        X1<override>:float = 1.0
        Y1<override>:float = 1.0

    @doc("`EaseOut` animations start fast, then slow down towards the end.")
    ease_out_cubic_bezier_easing_function<native><public> := class<final><concrete><computes>(cubic_bezier_easing_function):
        X0<override>:float = 0.0
        Y0<override>:float = 0.0
        X1<override>:float = 0.58
        Y1<override>:float = 1.0

    @doc("`EaseInOut` animations are similar to `Ease` but the start and end animation speed is symmetric.")
    ease_in_out_cubic_bezier_easing_function<native><public> := class<final><concrete><computes>(cubic_bezier_easing_function):
        X0<override>:float = 0.42
        Y0<override>:float = 0.0
        X1<override>:float = 0.58
        Y1<override>:float = 1.0

    @doc("Controls how the animation plays back.")
    keyframed_movement_playback_mode<native><public> := class<abstract><computes>{}

    @doc("Play once and stop.")
    oneshot_keyframed_movement_playback_mode<native><public> := class<final><concrete><computes>(keyframed_movement_playback_mode){}

    @doc("Play once and repeat indefinitely.")
    loop_keyframed_movement_playback_mode<native><public> := class<final><concrete><computes>(keyframed_movement_playback_mode){}

    @doc("Play continuously reversing direction at each end.")
    pingpong_keyframed_movement_playback_mode<native><public> := class<final><concrete><computes>(keyframed_movement_playback_mode){} 

    @doc("Represents a change in pose and scale over a duration.")
    keyframed_movement_delta<native><public> := class<final><concrete>:

        @doc("Represents a change in the transform relative to the previous keyframe or initial animation position. Translation and Scale are interpreted additively.")
        @editable
        Transform<native><public>:transform = transform{ Scale := vector3{ Forward := 0.0, Left := 0.0, Up := 0.0}}

        @doc("Duration of this keyframe in seconds.")
        @editable_number(float):
            MinValue:=option{0.0}
        Duration<native><public>:float = 0.0

        @doc("Easing function to use for playback.")
        @editable
        (keyframed_movement_delta:)Easing<native><public>:easing_function = linear_easing_function{}

    @doc("Provides teleportation and simple keyframe-based animation for an entity. Animations play back in the Pre-Physics tick phase.{
         } When animating an entity with a parent_constraint, animation will be relative to the parent entity.")
    keyframed_movement_component<native><public> := class<final><final_super>(component):

        #########################
        # Playable
        #########################

        @doc("Pause movement. Subsequently calling Play() will resume from the point in the animation when it was paused.")
        Pause<native><public>():void

        @doc("Begin or resume playback.")
        Play<native><public>():void

        @doc("Stop and reset transform to the initial state. Subsequently calling Play() will begin the animation anew.")
        Stop<public>():void = Stop(0.0)

        @doc("Stop and reset transform to the initial state. Subsequently calling Play() will begin the animation anew.")
        Stop<native><public>(BlendOutTime:float):void

        @doc("Get the event that fires when the animation is Stopped.")
        StoppedEvent<native><public>:listenable() = redacted_internal(){}

        @doc("Get the event that fires when the animation begins or resumes Playing.")
        PlayedEvent<native><public>:listenable()=redacted_internal(){}

        @doc("Get the event that fires when the animation is Paused.")
        PausedEvent<native><public>:listenable()=redacted_internal(){}

        @doc("Is the animation currently playing?")
        IsPlaying<native><public>()<reads><decides>:void

        @doc("Is the animation paused?")
        IsPaused<native><public>()<reads><decides>:void

        @doc("Gets the duration in seconds this keyframed movement will take. Fails if a fixed duration is not known (ex: looping animations).")
        # Inf if known and infinite, false if unknown or not knowable on the server
        var<private> Duration<native><public>:?float = false

        @doc("Get the event that fires when the animations ends, failing if the animation is of infinite duration")
        # "Signaled when the entire animation is complete. This will only fire for `OneShot` animations."
        FinishedEvent<native><public>:listenable() = redacted_internal(){}

        ##########################

        @doc("Stop any ongoing animation, sets the animation path and rebases it relative to the actor's current transform.{
             } Does not start playing until you call Play().")
        SetKeyframes<native><public>(RelativeKeyframes:[]keyframed_movement_delta, PlaybackMode:keyframed_movement_playback_mode):void

        @doc("Is there a valid set of playable keyframes?")
        HasValidAnimation<native><public>()<reads><decides>:void

        @doc("Signaled when any keyframe is reached. (Keyframe:int, IsReversed:logic).")
        KeyframeReachedEvent<native><public>:listenable(tuple(int, logic))=redacted_internal(tuple(int, logic)){}


