# Copyright Epic Games, Inc. All Rights Reserved.

using { /Verse.org/Assets }
using { /Verse.org/Colors }
using { /Verse.org/Native }
using { /Verse.org/Simulation }
using { /Verse.org/SpatialMath }
using { /Verse.org/Temporary/EpicGamesRestricted/Network }

@doc("Base class for light components in the SceneGraph.\n{
     }\n{
     }Dependencies:\n{
     }  * `transform_component` on the entity positions the light.\n{
     }\n{
     }Examples of components implementing `light_component`:\n{
     }  * `directional_light_component`\n{
     }  * `capsule_light_component`\n{
     }  * `sphere_light_component`\n{
     }  * `rect_light_component`\n{
     }  * `spot_light_component`")
light_component<native><public> := class<abstract><final_super>(component, enableable):

    # Public Properties

    @doc("Whether the light should cast any shadows.")
    var CastShadows<public><getter(GetCastShadows)><setter(SetCastShadows)>:logic = external{}

    @doc("Set the filter color of the light. This acts as a colored filter in front of the light source. Note that this can change the light's effective intensity. In normalized range 0-1.")
    var ColorFilter<public><getter(GetColorFilter)><setter(SetColorFilter)>:color = external{}

    @doc("Multiplier on specular highlights. Can be used to artistically remove highlights mimicking polarizing filters or photo touch up. Any value besides 1.0 is not physical. 0.0 means no specular contribution from this light.")
    var SpecularScale<public><getter(GetSpecularScale)><setter(SetSpecularScale)>:float = external{}

    @doc("Multiplier on diffuse lighting. Any value besides 1.0 is not physical. 0.0 means no diffuse contribution from this light.")
    var DiffuseScale<public><getter(GetDiffuseScale)><setter(SetDiffuseScale)>:float = external{}

    # Implements enableable interface

    @doc("Enables rendering of this light.")
    Enable<native><override>():void

    @doc("Disables rendering of this light.")
    Disable<native><override>():void

    @doc("Succeeds if the component is enabled, fails if it's disabled.")
    IsEnabled<native><override>()<decides><transacts>:void

    # enableable internals

    @editable
    @tool_tip("Whether or not the light should render.")
    @display_priority("1")
    @replicated("RepNotify")
    var Enabled<private><native>:logic = true 

    OnRep_Enabled<private><native>():void

    # Cast Shadows Internals


    @editable
    @tool_tip("Whether the light should cast any shadows.")
    @display_priority("2")
    @display_name("CastShadows")
    @replicated("RepNotify")
    var _CastShadows<private><native>:logic = true

    OnRep__CastShadows<native>():void

    # ColorFilter Internals


        case(MemberName):
            "R" => return _ColorFilter.R
            "G" => return _ColorFilter.G
            "B" => return _ColorFilter.B
            _ => return 0.0

        var NewColor:color = _ColorFilter;
        case(MemberName):
            "R" => set NewColor.R = MemberValue
            "G" => set NewColor.G = MemberValue
            "B" => set NewColor.B = MemberValue
            _ =>
        set ColorFilter = NewColor

    @editable
    @tool_tip("Set the filter color of the light. This acts as a colored filter in front of the light source.")
    @display_priority("3")
    @display_name("ColorFilter")
    @replicated("RepNotify")
    var _ColorFilter<private><native>:color = NamedColors.White

    OnRep__ColorFilter<native>():void

    # SpecularScale Internals


    @editable_slider(float):
        MinValue := option{ 0.0 }
        MaxValue := option{ 1.0 }
    @tool_tip("Multiplier on specular highlights. Can be used to artistically remove highlights mimicking polarizing filters or photo touch up. Any value besides 1.0 is not physical.")
    @display_priority("4")
    @display_name("SpecularScale")
    @replicated("RepNotify")
    var _SpecularScale<private><native>:float = 1.0

    OnRep__SpecularScale<native>():void

    # DiffuseScale Internals


    @editable_slider(float):
        MinValue := option{ 0.0 }
        MaxValue := option{ 1.0 }
    @tool_tip("Multiplier on diffuse lighting. Any value besides 1.0 is not physical.")
    @display_priority("5")
    @display_name("DiffuseScale")
    @replicated("RepNotify")
    var _DiffuseScale<private><native>:float = 1.0

    OnRep__DiffuseScale<native>():void

    ###########################################################################
    # FN1 Internal Extensions BEGIN





    @editable_slider(float):
        MinValue := option{0.0}
        MaxValue := option{6.0}

    @editable_slider(float):
        MinValue := option{0.0}
        MaxValue := option{4.0}


    ####
    # Light Shaft settings
    ####






    ####
    # Light Function settings
    ####




    @editable_slider(float):
        MinValue := option{0.0}
        MaxValue := option{1.0}

    ####
    # Light Profile settings
    ####



    @editable_slider(float):
        MinValue := option{0.0}
        MaxValue := option{10.0}

    # FN1 Internal Extensions END
    ###########################################################################
