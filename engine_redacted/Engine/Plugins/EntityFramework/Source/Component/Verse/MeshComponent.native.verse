# Copyright Epic Games, Inc. All Rights Reserved.

using { /Verse.org/Assets }
using { /Verse.org/Native }
using { /Verse.org/Simulation }
using { /Verse.org/SpatialMath }
using { /Verse.org/Temporary/EpicGamesRestricted/Network }

CollidableMeshToolTip<localizes>:message = "Enable/disable collision on this mesh. If enabled, meshes may collide in the physics simulation."
QueryableMeshToolTip<localizes>:message = "Enable/disable spatial queries against this mesh. Disabling this field will also disable EntityEnteredEvent/EntityExitedEvent."
VisibleMeshToolTip<localizes>:message = "Enable/disable visibility of this mesh."
CanAffectNavigationToolTip<localizes>:message = "When enabled, this mesh's collision can contribute to AI navigation. Still requires collision that blocks pawns. Use with care: NPCs will not route around meshes that do not affect navigation."

@doc("Used to render a `mesh` at the location of this entity. A mesh is a set of polygons that can be used to represent shapes {
     }in the world such as:\n{
     }  * foliage and terrain decorations\n{
     }  * movers (for example, doors and lifts)\n{
     }  * procedurally created buildings\n{
     }\n{
     }Dependencies:\n{
     }  * `transform_component` on the entity positions the mesh.")

mesh_component<native><public> := class<final_super>(component, enableable, redacted_internal, redacted_internal):

    @doc("Enables rendering of this mesh.")
    Enable<native><override>():void

    @doc("Disables rendering of this mesh.")
    Disable<native><override>():void

    @doc("Succeeds if the component is enabled, fails if it's disabled.")
    IsEnabled<native><override>()<decides><transacts>:void
    
    # Need a replicated property to indicate to the client whether this is a standalone component or wrapping actor based functionality
    @replicated("")
    var bWantsImplementation<private><native>:logic=true

    @doc("If the mesh should be rendered.")
    @editable
    # Need a replicated property to indicate to the client whether this is enabled
    @replicated("RepNotify")
    var Enabled<private><native>:logic=true

    @doc("Enable/disable collision on this mesh. If enabled, meshes may collide in the physics simulation.")
    var Collidable<public><getter(RedactedInternal)><setter(RedactedInternal)>:logic = external{}

    @doc("Enable/disable spatial queries against this mesh. Disabling this field will also disable EntityEnteredEvent/EntityExitedEvent.")
    var Queryable<public><getter(GetQueryable)><setter(SetQueryable)>:logic = external{}

    @doc("Enable/disable visibility of this mesh.")
    var Visible<public><getter(GetVisible)><setter(RedactedInternal)>:logic = external{}

    @doc("When enabled, this mesh's collision can contribute to AI navigation. Still requires collision that blocks pawns. Use with care: NPCs will not route around meshes that do not affect navigation. Server-authoritative: this value is not replicated.")
    var CanAffectNavigation<public><getter(GetCanAffectNavigation)><setter(SetCanAffectNavigation)>:logic = external{}

    @doc("Triggered at the beginning of each tick when another entity first overlaps this entity.")
    @transient
    EntityEnteredEvent<public><native><predicts>:listenable(entity) = redacted_internal(entity){}

    @doc("Triggered at the beginning of each tick when another entity is no longer overlapping this entity")
    @transient
    EntityExitedEvent<public><native><predicts>:listenable(entity) = redacted_internal(entity){}

    # Begin Collidable implementation
    @editable:
        ToolTip := CollidableMeshToolTip
    @display_name("Collidable")
    @replicated("RepNotify")
    var CollidableInternal<private><native>:logic = true
    OnRep_CollidableInternal<private><native>()<transacts>:void
    # End Collidable implementation

    # Begin Queryable implementation
    @editable:
        ToolTip := QueryableMeshToolTip
    @display_name("Queryable")
    @replicated("RepNotify")
    var<private> QueryableInternal<private><native>:logic = true
    OnRep_QueryableInternal<private><native>()<transacts>:void
    # End Queryable implementation

    # Begin Visible implementation
    @editable:
        ToolTip := VisibleMeshToolTip
    @display_name("Visible")
    @replicated("RepNotify")
    var<private> VisibleInternal<private><native>:logic = true

        VisibleInternal


    OnRep_VisibleInternal<private><native>()<transacts>:void
    # End Visible implementation

    # Begin CanAffectNavigation implementation
    @editable:
        ToolTip := CanAffectNavigationToolTip
    @display_name("CanAffectNavigation")
    var CanAffectNavigationInternal<private><native>:logic = true
    # End CanAffectNavigation implementation

    OnRep_Enabled<private><native>():void

    ###########################################################################
    # FN1 Internal Extensions BEGIN




















