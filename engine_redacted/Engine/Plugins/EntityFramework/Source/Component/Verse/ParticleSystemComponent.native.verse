# Copyright Epic Games, Inc. All Rights Reserved.

using { /Verse.org/Assets }
using { /Verse.org/Colors }
using { /Verse.org/Native }
using { /Verse.org/Simulation }
using { /Verse.org/SpatialMath }
using { /Verse.org/Temporary/EpicGamesRestricted/Network }
using { /UnrealEngine.com/Temporary/SpatialMath }

@doc("Used to spawn a `particle_system` at the location of this entity. The `particle_system` will simulate while {
     }the `particle_system_component` is in the scene.\n{
     }\n{
     }Dependencies:\n{
     }  * `transform_component` on the entity positions the `particle_system`.")
particle_system_component<native><public> := class<final_super>(component, enableable, redacted_internal):

    # Begin Enableable implementation
    @doc("Enables the simulation and rendering of this `particle_system`.")
    Enable<override><native>():void

    @doc("Disables the simulation and rendering of this `particle_system`.")
    Disable<override><native>():void

    @doc("Succeeds if the component is enabled, fails if it’s disabled.")
    IsEnabled<override><native>()<decides><transacts>:void

    @doc("Controls if the `particle_system_component` should start enabled.")
    @editable
    @replicated("RepNotify")
    var<private> Enabled<public><native>:logic = true
    # End Enableable implementation

    # Begin Playable implementation
    Play<public><native>()<predicts>:void

    Stop<public><native>()<predicts>:void

    @doc("Controls if the `particle_system_component` should play the simulation automatically when added to the scene, or when enabled from a disabled state.")
    @editable
    @replicated("")

    @editable
    AutoPlayInEditor<private><native>:logic = true
    # End Playable implementation


    # Need a replicated property to indicate to the client whether this is a standalone component or wrapping actor based functionality
    @replicated("")
    var bWantsImplementation<private><native>:logic = true

    @doc("Should the niagara component tick pre (default) or post physics?")

    # Counter to track playing or not playing and number of calls.  Incremented after each Play/Stop call, negative means not playing, positive means playing
    @replicated("RepNotify")
    var StartCounter<private><native>:int = -1

    # Set functions called from the digest

