# Copyright Epic Games, Inc. All Rights Reserved.

using { /Verse.org/Native }
using { /Verse.org/SceneGraph }
using { /Verse.org/Simulation }
using { /Verse.org/SpatialMath }



@doc("Find all objects in the scene that currently overlap this entity.")
(Entity:entity).FindOverlapHits<public>()<transacts>:generator(overlap_hit) =
    return Entity.FindOverlapHits(Entity.GetGlobalTransform())

@doc("Find all objects in the scene that would overlap this entity if it were placed at GlobalTransform.")
(Entity:entity).FindOverlapHits<public>(GlobalTransform:transform)<transacts>:generator(overlap_hit) =
        return PhysicsScene.FindEntityOverlaps(Entity, GlobalTransform)

@doc("{
}Find all objects in the scene that would overlap Volume if they were placed at GlobalTransform. {
}NOTE: This entity defines the context(scene) for the query but does not otherwise take part in the sweep.")
(Entity:entity).FindOverlapHits<public>(GlobalTransform:transform, Volume:collision_volume)<transacts>:generator(overlap_hit) =

@doc("{
}Find objects in the scene that would intersect this entity if it were swept from its location along the Displacement vector. {
}Returns the first object interacting as collision_interaction.Block, and all objects interacting as collision_interaction.Overlap {
}encountered before the first block. Hits are sorted by hit distance, so the blocking hit will be last.")
(Entity:entity).FindSweepHits<public>(Displacement:vector3)<transacts>:generator(sweep_hit) = 
    return Entity.FindSweepHits(Displacement, Entity.GetGlobalTransform())

@doc("{
}Find objects in the scene that would intersect this entity if it were swept from GlobalTransform along the Displacement vector. {
}Returns the first object interacting as collision_interaction.Block, and all objects interacting as collision_interaction.Overlap {
}encountered before the first block. Hits are sorted by hit distance, so the blocking hit will be last.")
(Entity:entity).FindSweepHits<public>(Displacement:vector3, StartGlobalTransform:transform)<transacts>:generator(sweep_hit) =
        return PhysicsScene.FindEntitySweepHits(Entity, StartGlobalTransform, Displacement)

@doc("{
}Find all objects in the scene that would intersect Volume if they were swept from GlobalTransform along the Displacement vector {
}Hits are sorted by hit distance. {
}NOTE: This entity defines the context(scene) for the query but does not otherwise take part in the sweep.")
(Entity:entity).FindSweepHits<public>(Displacement:vector3, StartGlobalTransform:transform, Volume:collision_volume)<transacts>:generator(sweep_hit) = 



