# Copyright Epic Games, Inc. All Rights Reserved.

using { /Verse.org/Native }
using { /Verse.org/SceneGraph }
using { /Verse.org/Simulation }

@doc("Marks an entity that can be possessed by an agent.")
@hide_in_editor
possessable_component<native><public> := class<final_super>(component):
    @doc("Which agent is this entity currently possessed by.")
    var<private> Agent<native><public><final>:?agent = false
    
    OnEndSimulation<override>():void=
        (super:)OnEndSimulation()
