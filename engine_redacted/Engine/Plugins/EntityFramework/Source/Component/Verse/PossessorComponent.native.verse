# Copyright Epic Games, Inc. All Rights Reserved.

using { /Verse.org/Native }
using { /Verse.org/SceneGraph }
using { /Verse.org/Simulation }

@doc("Agent component to manage possessed entities.")
@hide_in_editor
possessor_component<native><internal> := class<final_super>(component):
    
    @doc("All entities possessed by this agent.")
    var<private> PossessedEntities<native><internal> : []possessable_component = array{}

    @doc("Event that will fire when a new entity is possessed.")

    @doc("Event that will fire when an entity is unpossessed.\n{
    } Entities are automatically unpossessed when removed from the scene")
    
    @doc("Checks if an entity is possessed.")
    IsEntityPossessed<native_callable><internal>(PossessableComponent:possessable_component)<decides><reads>:void =
        if (not PossessedEntities.Find[PossessableComponent]):
            false?

    @doc("Signals the BeginPossessEvent.")
    SignalBeginPossessEvent<native><internal>(PossessableComponent:possessable_component)<transacts>:void

    @doc("Signals the EndPossessEvent.")
    SignalEndPossessEvent<native><internal>(PossessableComponent:possessable_component)<transacts>:void
    
# Extension functions for agents    
    
@doc("All entities possessed by this agent.")
(InAgent:agent).GetPossessedEntities<public>()<reads>:[]possessable_component =
    if (PossessorComponent:=InAgent.GetComponent[possessor_component]):
        return PossessorComponent.PossessedEntities
    else:
        return array{}

@doc("Checks if an entity is possessed.")
(InAgent:agent).IsEntityPossessed<public>(PossessableComponent:possessable_component)<decides><reads>:void =
    InAgent.GetComponent[possessor_component].IsEntityPossessed[PossessableComponent]
