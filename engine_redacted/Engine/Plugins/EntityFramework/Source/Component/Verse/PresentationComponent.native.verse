# Copyright Epic Games, Inc. All Rights Reserved.

using { /Verse.org/Simulation }
using { /Verse.org/Native }
using { /Verse.org/SceneGraph }
using { /Verse.org/Temporary/EpicGamesRestricted/Network }



@doc("Assign the players that this entity will be presented to. False = presentable to everyone, array{} = presentable to no one.")
@available{MinUploadedAtFNVersion := 3800}
(Entity:entity).SetPresentableToPlayers<native><public>(Players:?[]player):void

@doc("Get the players that this entity is currently be presented to. False = presentable to everyone, array{} = presentable to no one.")
@available{MinUploadedAtFNVersion := 3800}
(Entity:entity).GetPresentableToPlayers<native><public>()<transacts>:?[]player