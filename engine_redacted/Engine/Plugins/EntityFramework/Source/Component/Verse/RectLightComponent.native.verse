# Copyright Epic Games, Inc. All Rights Reserved.

using { /Verse.org/Assets }
using { /Verse.org/Colors }
using { /Verse.org/Native }
using { /Verse.org/Simulation }
using { /Verse.org/Temporary/EpicGamesRestricted/Network }

@doc("A `rect_light_component` emits light into the scene from a rectangular plane with a specified width and height. You can use these\n{
     }to simulate any kind of light sources that have rectangular areas, such as televisions or monitor screens, overhead lighting\n{
     }fixtures, or wall sconces.")
rect_light_component<native><public> := class<final>(light_component, redacted_internal):

    @doc("Set the visible light intensity emitted in SI unit Candela.\n{
         }Specified before ColorFilter (which multiplies each color component after the intensity calculation and can change the effective intensity of the light).")
    var Intensity<public><getter(GetIntensity)><setter(SetIntensity)>:float = external{}

    @doc("The bounds of the light's visible influence, in centimeters. This clamping of the light's influence is not physically correct but very important for performance,\n{
         }larger lights cost more. The light falloff is based on Inverse Square law. Towards the tail end of the AttenuationRadius,\n{
         }there is an additional smoothing factor to fade out the light contribution to 0 to avoid a hard cutoff.")
    var AttenuationRadius<public><getter(GetAttenuationRadius)><setter(SetAttenuationRadius)>:?float = external{}

    @doc("The width of the light source rect, in centimeters. Note that light source shapes which intersect shadow casting geometry can cause shadowing artifacts.")
    var SourceWidth<public><getter(GetSourceWidth)><setter(SetSourceWidth)>:float = external{}

    @doc("The height of the light source rect, in centimeters. Note that light source's shapes which intersect shadow casting geometry can cause shadowing artifacts.")
    var SourceHeight<public><getter(GetSourceHeight)><setter(SetSourceHeight)>:float = external{}

    @doc("The angle of the barn door in degrees attached to the light source rect. Clamped between 0.0 and 90.0 degrees.")
    var BarnDoorAngleDegrees<public><getter(GetBarnDoorAngleDegrees)><setter(SetBarnDoorAngleDegrees)>:float = external{}
   
    @doc("The length of the barn door attached to the light source rect, in centimeters.")
    var BarnDoorLength<public><getter(GetBarnDoorLength)><setter(SetBarnDoorLength)>:float = external{}

    #@doc("Texture mapped to the light source rectangle.")
    #var SourceTexture<public><getter(GetSourceTexture)><setter(SetSourceTexture)>:?texture = external{}

    # Intensity Internals


    @editable
    @tool_tip("Set the visible light intensity emitted in SI unit Candela.")
    @display_name("Intensity")
    @replicated("RepNotify")
    var _Intensity<private><native>:float = 8.0

    OnRep__Intensity<native>():void


    # Attenuation Radius Internals


    @editable
    @tool_tip("The bounds of the light's visible influence, in centimeters.")
    @display_name("AttenuationRadius")
    @replicated("RepNotify")
    var _AttenuationRadius<private><native>:?float = option{1000.0}

    OnRep__AttenuationRadius<native>():void


    # Source Width Internals


    @editable
    @tool_tip("The width of the light source rect, in centimeters.")
    @display_name("SourceWidth")
    @replicated("RepNotify")
    var _SourceWidth<private><native>:float = 64.0

    OnRep__SourceWidth<native>():void

    # Source Height Internals


    @editable
    @tool_tip("The height of the light source rect, in centimeters.")
    @display_name("SourceHeight")
    @replicated("RepNotify")
    var _SourceHeight<private><native>:float = 64.0

    OnRep__SourceHeight<native>():void


    # Barn Door Angle Degrees Internals
 

    @editable_slider(float):
        MinValue := option{ 0.0 }
        MaxValue := option{ 90.0 }
    @tool_tip("The angle of the barn door in degrees attached to the light source rect.")
    @display_name("BarnDoorAngleDegrees")
    @replicated("RepNotify")
    var _BarnDoorAngleDegrees<private><native>:float = 88.0

    OnRep__BarnDoorAngleDegrees<native>():void


    # Barn Door Length Internals


    @editable
    @tool_tip("The length of the barn door attached to the light source rect, in centimeters.")
    @display_name("BarnDoorLength")
    @replicated("RepNotify")
    var _BarnDoorLength<private><native>:float = 20.0

    OnRep__BarnDoorLength<native>():void

    # Source Texture

    #GetSourceTexture<private><native>(:accessor)<transacts>:?texture
    #SetSourceTexture<private><native>(:accessor, InSourceTexture:?texture)<transacts>:void

    #@replicated("RepNotify")
    #var _SourceTexture<private><native>:?texture = false

    #OnRep__SourceTexture<native>():void