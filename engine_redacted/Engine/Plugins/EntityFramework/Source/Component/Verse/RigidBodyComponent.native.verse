# Copyright Epic Games, Inc. All Rights Reserved.

using { /Verse.org/Native }
using { /Verse.org/Simulation }
using { /Verse.org/SpatialMath }
using { /Verse.org/Temporary/EpicGamesRestricted/Network }

DynamicBodyToolTip<localizes>:message = "Enable/disable simulation of this body."
MassToolTip<localizes>:message = "The mass to use for component."
GravityGroupIndexToolTip<localizes>:message = "The gravity group index to use for component."
MaxAngularVelocityToolTip<localizes>:message = "The max angular velocity to use for component."
    # End Max Angular Velocity implementation
