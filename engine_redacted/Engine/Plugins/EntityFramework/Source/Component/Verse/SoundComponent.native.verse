# Copyright Epic Games, Inc. All Rights Reserved.

using { /Verse.org/Assets }
using { /Verse.org/Native }
using { /Verse.org/Simulation }
using { /Verse.org/SpatialMath }
using { /Verse.org/Temporary/EpicGamesRestricted/Network }

sound_component<native><public> := class<abstract><final_super>(component, redacted_internal, enableable):
    
    # Begin Playable implementation
    @editable
    @replicated("")

    @doc("Play the sound asset")
    Play<public><native>()<predicts>:void

    @doc("Stop the sound asset")
    Stop<public><native>()<predicts>:void
    # End Playable implementation

    # Begin Enableable implementation
    @doc("Enable the sound component.")
    Enable<override><native>():void

    @doc("Disable the sound component.")
    Disable<override><native>():void

    @doc("Succeeds if the sound component is enabled, fails if it is disabled.")
    IsEnabled<override><native>()<decides><transacts>:void

    @editable
    @replicated("RepNotify")
    var<private> Enabled<public><native>:logic = true
    
    # Need a replicated property to indicate to the client whether this is
    # a standalone component or wrapping actor based functionality
    @replicated("")
    var bWantsImplementation<private><native>:logic=true

    @replicated("RepNotify")
    var StartCounter<private><native>:int = -1

    OnRep_StartCounter<private><native>():void
    OnRep_Enabled<private><native>():void
    OnRep_Sound<private><native>():void
    
