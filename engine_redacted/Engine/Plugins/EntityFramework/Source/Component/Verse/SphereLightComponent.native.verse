# Copyright Epic Games, Inc. All Rights Reserved.

using { /Verse.org/Colors }
using { /Verse.org/Native }
using { /Verse.org/Simulation }
using { /Verse.org/Temporary/EpicGamesRestricted/Network }

@doc("A `sphere_light_component` emits light in all directions into the scene from a spherical source shape with a specified radius. A radius of 0 makes it a point light. You can use these\n{
     }to simulate any kind of light sources that emit in all directions, such as a light bulb.")
sphere_light_component<native><public> := class<final>(light_component, redacted_internal):

    @doc("Set the visible light intensity emitted in SI unit Candela.\n{
         }Specified before ColorFilter (which multiplies each color component after the intensity calculation and can change the effective intensity of the light).")
    var Intensity<public><getter(GetIntensity)><setter(SetIntensity)>:float = external{}

    @doc("The bounds of the light's visible influence, in centimeters. This clamping of the light's influence is not physically correct but very important for performance,\n{
         }larger lights cost more. The light falloff is based on Inverse Square law. Towards the tail end of the AttenuationRadius,\n{
         }there is an additional smoothing factor to fade out the light contribution to 0 to avoid a hard cutoff.")
    var AttenuationRadius<public><getter(GetAttenuationRadius)><setter(SetAttenuationRadius)>:?float = external{}

    @doc("Radius of the source shape, in centimeters. Note that light shapes which intersect shadow casting geometry can cause shadowing artifacts.")
    var SourceRadius<public><getter(GetSourceRadius)><setter(SetSourceRadius)>:float = external{}


    # Intensity Internals
 

    @editable
    @tool_tip("Set the visible light intensity emitted in SI unit Candela.")
    @display_name("Intensity")
    @replicated("RepNotify")
    var _Intensity<private><native>:float = 8.0

    OnRep__Intensity<native>():void


    # Attenuation Radius Internals


    @editable
    @tool_tip("The bounds of the light's visible influence, in centimeters.")
    @display_name("AttenuationRadius")
    @replicated("RepNotify")
    var _AttenuationRadius<private><native>:?float = option{ 1000.0 }

    OnRep__AttenuationRadius<native>():void
    

    # Source Radius Internals


    @editable
    @tool_tip("Radius of the source shape, in centimeters.")
    @display_name("SourceRadius")
    @replicated("RepNotify")
    var _SourceRadius<private><native>:float = 10.0

    OnRep__SourceRadius<native>():void

    ###########################################################################
    # FN1 Internal Extensions BEGIN
    # TODO: Once categories are working and can be compiled, use that instead of the display name


    @editable_slider(float):
        MinValue := option{ 2.0 }
        MaxValue := option{ 16.0 }

    # FN1 Internal Extensions END
    ###########################################################################
