# Copyright Epic Games, Inc. All Rights Reserved.

using { /Verse.org/Colors }
using { /Verse.org/Native }
using { /Verse.org/Simulation }
using { /Verse.org/Temporary/EpicGamesRestricted/Network }

@doc("A `spot_light_component` emits light from a single point in a cone shape. The shape of the light is defined by two cones: the `InnerConeAngleDegrees`\n{
     }and `OuterConeAngleDegrees`. Within the `InnerConeAngleDegrees` the light achieves full brightness. As you go from the extent of the inner radius to the\n{
     }extents of the `OuterConeAngleDegrees` a falloff takes place, creating a penumbra, or softening around the `spot_light_component`'s disc of\n{
     }illumination. The Radius of the light defines the length of the cones. More simply, this will work like a flash light or stage can light.")
spot_light_component<native><public> := class<final>(light_component, redacted_internal):

    @doc("Set the visible light intensity emitted in SI unit Candela.\n{
         }Specified before ColorFilter (which multiplies each color component after the intensity calculation and can change the effective intensity of the light).")
     var Intensity<public><getter(GetIntensity)><setter(SetIntensity)>:float = external{}
     
     @doc("The bounds of the light's visible influence, in centimeters. This clamping of the light's influence is not physically correct but very important for performance,\n{
          }larger lights cost more. The light falloff is based on Inverse Square law. Towards the tail end of the AttenuationRadius,\n{
          }there is an additional smoothing factor to fade out the light contribution to 0 to avoid a hard cutoff.")
     var AttenuationRadius<public><getter(GetAttenuationRadius)><setter(SetAttenuationRadius)>:?float = external{}
    
    @doc("Radius of the source shape, in centimeters. Note that light shapes which intersect shadow casting geometry can cause shadowing artifacts.")
    var SourceRadius<public><getter(GetSourceRadius)><setter(SetSourceRadius)>:float = external{}

    @doc("The light's inner cone shaped angle in degrees. Clamped between 0.0 and 80.0.")
    var InnerConeAngleDegrees<public><getter(GetInnerConeAngleDegrees)><setter(SetInnerConeAngleDegrees)>:float = external{}

    @doc("The light's outer cone shaped angle in degrees. Clamped between 1.0 and 80.0.")
    var OuterConeAngleDegrees<public><getter(GetOuterConeAngleDegrees)><setter(SetOuterConeAngleDegrees)>:float = external{}

    # Intensity Internals


    @editable
    @tool_tip("Set the visible light intensity emitted in SI unit Candela.")
    @display_name("Intensity")
    @replicated("RepNotify")
    var _Intensity<private><native>:float = 8.0

    OnRep__Intensity<native>():void


    # Attenuation Radius Internals


    @editable
    @tool_tip("The bounds of the light's visible influence, in centimeters.")
    @display_name("AttenuationRadius")
    @replicated("RepNotify")
    var _AttenuationRadius<private><native>:?float = option{1000.0}

    OnRep__AttenuationRadius<native>():void
    

     # Source Radius Internals


    @editable
    @tool_tip("Radius of the source shape, in centimeters.")
    @display_name("SourceRadius")
    @replicated("RepNotify")
    var _SourceRadius<private><native>:float = 10.0

    OnRep__SourceRadius<native>():void


    # Inner Cone Angle Internals


    @editable_slider(float):
        MinValue := option{ 0.0 }
        MaxValue := option{ 80.0 }
    @tool_tip("The light's inner cone shaped angle in degrees.")
    @display_name("InnerConeAngleDegrees")
    @replicated("RepNotify")
    var _InnerConeAngleDegrees<private><native>:float = 0.0

    OnRep__InnerConeAngleDegrees<native>():void


    # Outer Cone Angle Internals


    @editable_slider(float):
        MinValue := option{ 1.0 }
        MaxValue := option{ 80.0 }
    @tool_tip("The light's outer cone shaped angle in degrees.")
    @display_name("OuterConeAngleDegrees")
    @replicated("RepNotify")
    var _OuterConeAngleDegrees<private><native>:float = 44.0

    OnRep__OuterConeAngleDegrees<native>():void

    ###########################################################################
    # FN1 Internal Extensions BEGIN
    # TODO: Once categories are working and can be compiled, use that instead of the display name


    @editable_slider(float):
        MinValue := option{ 2.0 }
        MaxValue := option{ 16.0 }

    # FN1 Internal Extensions END
    ###########################################################################
