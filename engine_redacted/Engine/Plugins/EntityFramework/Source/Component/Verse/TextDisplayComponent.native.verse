# Copyright Epic Games, Inc. All Rights Reserved.
 
using { /Verse.org/Assets }
using { /Verse.org/Colors }
using { /Verse.org/Native }
using { /Verse.org/Simulation }
using { /Verse.org/SpatialMath }
using { /Verse.org/Temporary/EpicGamesRestricted/Network }
 
@doc("A range with a minimum and maximum value. For a value to fall inside of this range, the min value must be less than or equal to the max value.")
@experimental
@replicatable
float_range<native><public> := struct<concrete>:
 
    # The minimum value of the range. Must be less than or equal to Max for values to fall inside the range.
    Minimum<public><native>:float = 0.0
 
    # The maximum value of the range. Must be greater than or equal to Min for values to fall inside the range.
    @editable
    Maximum<public><native>:float = 0.0
