# Copyright Epic Games, Inc. All Rights Reserved.

using { /Verse.org/Native }
using { /Verse.org/Simulation }

@import_as("/Script/Entity.UBaseComponent")
base_component<native><internal> := class<abstract>{}

@doc("Base class for authoring logic and data in the SceneGraph. Using components you\n{
     }can author re-usable building blocks of logic and data which can then be added to\n{
     }entities in the scene.\n{
     }\n{
     }Components are a very low level building block which can be used in many ways.\n{
     }For example:\n{
     }  * Exposing engine level concepts like mesh or sound\n{
     }  * Adding gameplay capabilities like damage or interaction\n{
     }  * Storing an inventory for a character in the game\n{
     }\n{
     }As components are generic there is no specific way that they must be used.\n{
     }It is up to the needs of your experience if you use one big game component\n{
     }or if you break up logic into many small components.\n{
     }\n{
     }Classes deriving from component must also specify `<final_super>` to be added\n{
     }to entities. This ensures the class will always derive directly from `component`.\n{
     }Further subclassing of the initial derived component is allowed and does not require\n{
     }specifying `<final_super>` on the derived classes.\n{
     }\n{
     }Only one instance of a component from each subclass group can be added to an entity\n{
     }at a time. For example, given this group of components, only one light_component can\n{
     }exist on a single entity. To create multiple lights you should use multiple entities.\n{
     }\n{
     }  light_component             := class<final_super>(component)\{\}\n{
     }  capsule_light_component     := class<final>(light_component)\{\}\n{
     }  directional_light_component := class<final>(light_component)\{\}\n{
     }  spot_light_component        := class<final>(light_component)\{\}\n{
     }  sphere_light_component      := class<final>(light_component)\{\}\n{
     }  rect_light_component        := class<final>(light_component)\{\}\n{
     }\n{
     }==============================================================================\n{
     }Component Lifetime\n{
     }\n{
     }  Components move through a series of lifetime functions as they are added\n{
     }  to entities, added to the scene, and begin running in the simulation. Components\n{
     }  should override these methods to perform setup and run their simulation.\n{
     }\n{
     }  As a component shuts down it will then move through shutdown version of these\n{
     }  functions, giving users the opportunity to clean up any retained state on the\n{
     }  component before it is disposed\n.{
     }\n{
     }  Lifetime Methods:\n{
     }     OnAddedToScene\n{
     }       OnBeginSimulation -> OnSimulate<suspends>\n{
     }       OnEndSimulation\n{
     }     OnRemovingFromScene\n{
     }==============================================================================")
component<native><public> := class<abstract><unique><castable><final_super_base>(base_component, redacted_internal):

    ############################
    # Entity Hierarchy Methods
    ############################

    @doc("The parent entity of this component.\n{
         }  * Components must have a parent entity pointer provided when being constructed.\n{
         }  * Components cannot be moved between parents.")
    @allow_self_reference
    @experimental_always_overridden
    @predicts_extern
    Entity<native><public><predicts>:entity

    @doc("Called when the component is added to the scene by parenting it under the simulation entity or another entity already in the scene.\n{
         }  * Querying for components in the scene is valid after this phase completes.")
    OnAddedToScene<protected><native><native_callable>():void

    @doc("Called when the component begins simulating within the scene.\n{
         }  * Use this to set up TickEvent callbacks or other setup that must be guaranteed to complete immediately.\n{
         }  * `OnAddedToScene` is guaranteed to run before `OnBeginSimulation`.")
    OnBeginSimulation<protected><native><native_callable>():void

    @doc("Called when the component begins simulating within the scene.\n{
         }  * Use this to add asynchronous/suspends update logic for a component.\n{
         }  * `OnBeginSimulation` is guaranteed to run before `OnSimulate`.\n{
         }  * `OnSimulate` will be cancelled before `OnEndSimulation`")
    OnSimulate<protected><native_callable>()<suspends>:void = {}

    @doc("Called when the component ends simulation within the scene.\n{
         }  * Simulation ends on a component when the experience resets, the parent entity is removed from the scene.\n{
         }  * Cached TickEvents cancelables should be canceled in `OnEndSimulation`.\n{
         }  * `OnSimulate` task will be canceled before `OnEndSimulation` is called.\n{
         }  * `OnEndSimulation` is only called on components that have already had `OnBeginSimulation` called.")
    OnEndSimulation<protected><native><native_callable>():void

    @doc("Called when the component is about to be removed from the scene.\n{
         }  * Components are removed from a scene when the parent entity is removed from the scene.\n{
         }  * `OnRemovingFromScene` is only called on components that have already had `OnAddedToScene` called.")
    OnRemovingFromScene<protected><native><native_callable>():void

    @doc("Removes the component from the entity.\n{
          }  * Removed components are removed from the scene and can only be added back to the same entity.\n{
          }  * Flows through `OnEndSimulation`-> `OnRemovingFromScene`.")
    RemoveFromEntity<native><public><final>():void

    @doc("Succeeds if the component is currently in the scene.\n{
         }  * After `OnAddedToScene` is called this call succeeds.\n{
         }  * After `OnRemovingFromScene` is called this call fails.")
    IsInScene<native><public><final>()<decides><reads>:void

    @doc("Succeeds if the component is currently simulating.\n{
         }  * After `OnBeginSimulation` is called this call succeeds.\n{
         }  * After `OnEndSimulation` is called this call fails.")
    IsSimulating<native><public><final>()<decides><reads>:void

    ############################
    # Ticking
    ############################
    @doc("Set callbacks to `TickEvents.PrePhysics` and `TickEvents.PostPhysics` to receive per-frame updates {
         }before and after physics is updated on your object.")
    var<private> TickEvents<protected><getter(GetTickEvents)><setter(SetTickEvents)>:tick_events = external{}

    ############################
    # Scene Events
    ############################

    @doc("Send a scene event to this component, invoking OnReceive. {
         } Returns true if any participant consumed the event.")
    SendDown<public><native><native_callable><final>(SceneEvent:scene_event):logic

    @doc("Respond to a scene event. Return true to consume the event and halt propagation to the next entity.")
    @available{MinUploadedAtFNVersion := 4000}
    OnReceive<protected><native_callable>(SceneEvent:scene_event):logic = false

    ############################
    # diagnosable overrides
    ############################


