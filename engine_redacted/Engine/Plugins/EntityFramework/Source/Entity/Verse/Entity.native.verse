# Copyright Epic Games, Inc. All Rights Reserved.

using { /Verse.org/Verse }
using { /Verse.org/Native }
using { /Verse.org/Simulation }
using { /Verse.org/SpatialMath }


@doc("An event which can be sent through the scene graph.")
scene_event<public><native> := interface{}

@doc("Entities are the base object in the SceneGraph.\n{
     }  * Objects in experiences are constructed of one or more entities.\n{
     }  * Entities are hierarchical. You can query your parent using `GetParent` and add child entities using `AddEntities`.\n{
     }  * Behavior is added to entities through `component`s. You can add new components using `AddComponents`.\n{
     }  * The structure and content of entities is dynamic and be changed at any time through your experience\n.{
     }\n{
     }==============================================================================\n{
     }Deriving from entity\n{
     }\n{
     }  In the SceneGraph system a class that derives from `entity` is also known as a prefab. Prefabs are useful when you\n{
     }  want to spawn/re-use a collection of entities and components many times within your game. Primarily prefabs are\n{
     }  authored through the editor, with their Verse classes generated as part of the build into the projects\n{
     }  Assets.digest.verse file.\n{
     }\n{
     }  While you can create base prefabs for common game object types like a vehicle or character, we highly recommended\n{
     }  that you do not add code directly to the entity class, and instead keep logic in components. Keeping logic and data in\n{
     }  components allows you to restructure your prefabs throughout production of your experience, without needing to massively\n{
     }  refactor your class structure.")
entity<native><public> := class<concrete><unique><transacts><castable>(redacted_internal, redacted_internal, has_tags):

    ############################
    # Entity Hierarchy Methods
    ############################
    
    @doc("Returns the parent entity of this entity.\n{
         }  * The parent entity controls the lifetime of its child entities and components. When an entity\n{
         }    is removed from the scene, all its child entities and components will be removed as well.\n{
         }  * Method fails if there is currently no parent entity.")
    GetParent<native><public><final>()<reads><decides>:entity
    
    @doc("Removes this entity from its parent. This is used to remove entities from the scene.\n{
         }  * Components on this entity and its children will run through `OnEndSimulation` -> `OnRemovingFromScene`.\n{
         }  * Entity can be added back later by using `NewParent.AddEntities`.")
    RemoveFromParent<native><public><final>():void

    @doc("Adds the provided entities as children of this entity.\n{
         }  * If child entity already has a parent, removes the entity from its current parent and adds it to the new one.\n{
         }  * Added child entities will move through their lifetime methods until they match the state of the new parent.")
    AddEntities<native><public><final>(NewChildEntities:[]entity):void

    @doc("Returns the child entities belonging to this entity which are accessible from the calling context.\n{
         }  * This method only gets the direct entity children. To query multiple levels down the entity structure use\n{
         }    the Find* query methods instead.")
    GetEntities<native><public><final>()<reads>:[]entity

    ############################
    # Component Methods
    ############################

    @doc("Succeeds and returns the child component of type `component_type` if it exists and is accessible from the calling context.\n{
        }  Note: When called during the AddedToScene or BeginSimulation phase, it will make sure the returned component has achieved the corresponding phase.\n{
        }  Fails if no component of `component_type` exists or can be accessed.")
    GetComponent<native><public><final>(component_type:castable_subtype(component))<reads><decides><predicts>:component_type

    @doc("Returns the child components belonging to this entity which are accessible from the calling context.")
    GetComponents<native><public><final>()<reads><predicts>:[]component

    @doc("Adds the provided components to the entity.\n{
         }  * If a component is not allowed to be added to this entity it is skipped.\n{
         }  Note: When called during the AddedToScene or BeginSimulation phase, it will make sure the added component has achieved the corresponding phase.\n{
         }  * Components are added following these rules:\n{
         }      1. All components are added to the entity child list.\n{
         }      2. All components have `OnAddedToScene` called (if this entity is in the scene).\n{
         }      3. All components have `OnBeginSimulation` called (if this entity is simulating).")
    AddComponents<native><public><final>(Components:[]component):void

    ############################
    # diagnosable overrides
    ############################


    ############################
    # Scene Event Methods
    ############################

    @doc("Send a scene event to this entity and then up the hierarchy.{
         } First, SendDown/OnReceive will be invoked on each component on this entity.{
         } Next, SendUp will be invoked on this entity's parent.{
         } Consuming the event at any point will halt propagation.{
         } Returns true if any participant consumed the event.")
    SendUp<public><native><native_callable><final>(SceneEvent:scene_event):logic

    @doc("Send a scene event to this entity and then down the hierarchy.{
         } First, SendDown/OnReceive will be invoked on each component on this entity.{
         } Next, SendDown will be invoked on each child entity. {
         } Consuming the event at any point will halt propagation.{
         } Returns true if any participant consumed the event.")
    SendDown<public><native><native_callable><final>(SceneEvent:scene_event):logic

    ############################
    # Has Tags Methods
    ############################

    @doc("Adds a `tag` instance to this entity. Returns a `tag_key` that is uniquely associated with the added instance.")
    AddTag<override><native><final>(Tag:tag)<transacts>:tag_key
    
    @doc("Removes the tag instance associated with the `tag_key`, succeeds if an instance was removed, fails otherwise.")
    RemoveTag<override><native><final>(Key:tag_key)<transacts><decides>:void

    @doc("Removes all tag instances of type `tag_type`, succeeds if at least one instance was removed, fails otherwise.")
    RemoveAllTags<override><native><final>(tag_type:castable_subtype(tag))<transacts><decides>:void

    @doc("Removes all tag instances that are not of type `tag_type`, succeeds if at least one instance was removed, fails otherwise.")
    RemoveAllTagsExcept<override><native><final>(tag_type:castable_subtype(tag))<transacts><decides>:void

    @doc("Removes all tag instances that are not of any of the types in `tag_types`, succeeds if at least one instance was removed, fails otherwise.")
    RemoveAllTagsExcept<override><native><final>(tag_types:[]castable_subtype(tag))<transacts><decides>:void

    @doc("Succeeds if at least one tag of type `tag_type` is found in this container, fails otherwise.")
    ContainsTag<override><native><final>(tag_type:castable_subtype(tag))<decides><reads>:void

    @doc("Fails if at least one type in `tag_types` cannot be found in this container, succeeds otherwise. Note that this means that if `tag_types` is empty this call succeeds.")
    ContainsAllTags<override><native><final>(tag_types:[]castable_subtype(tag))<decides><reads>:void

    @doc("Succeeds if at least of the types in `tag_types` is found in this container, fails otherwise. Note that this means that if `tag_types` is empty this call fails.")
    ContainsAnyTag<override><native><final>(tag_types:[]castable_subtype(tag))<decides><reads>:void
