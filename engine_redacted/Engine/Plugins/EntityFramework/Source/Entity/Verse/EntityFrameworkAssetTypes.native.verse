# Copyright Epic Games, Inc. All Rights Reserved.

using { /Verse.org/Assets }
using { /Verse.org/Native }

@doc("Reference type to editor defined prefab. Only generated digest code should reference this type.")
entity_prefab<native><public> := class<computes>(redacted_internal) {}
