# Copyright Epic Games, Inc. All Rights Reserved.

using { /Verse.org/Native }
using { /Verse.org/SceneGraph }
using { /Verse.org/Simulation }

@doc("Finds all descendant entities including `InEntity` that has any tags of type `tag_type`.\n{
    }When querying from the simulation entity, the simulation entity itself is not included in the results.\n{
    }The order of the returned entities is unspecified and subject to change.")
(InEntity:entity).FindDescendantEntitiesWithTag<native><public>(tag_type:castable_subtype(tag))<transacts>:generator(entity)

@doc("Finds all ancestor entities to `InEntity` that has any tags of type `tag_type`.\n{
    }The order of the returned entities is unspecified and subject to change.")
(InEntity:entity).FindAncestorEntitiesWithTag<native><public>(tag_type:castable_subtype(tag))<transacts>:generator(entity)