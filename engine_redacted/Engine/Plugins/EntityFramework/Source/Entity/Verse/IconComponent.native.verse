# Copyright Epic Games, Inc. All Rights Reserved.

using { /Verse.org/Assets }
using { /Verse.org/Native }
using { /Verse.org/Simulation }
using { /Verse.org/Temporary/EpicGamesRestricted/Network }

@doc("Component that holds an icon for an entity.")
icon_component<public><native> := class<final_super>(component, has_icon):

    var Icon<override><getter(GetIcon)><setter(SetIcon)>:texture = external {}

    @editable
    @display_name("Icon")
    @replicated("")
    var IconPrivate<private><native>:texture = texture {}

    # Internal access by native code.
        IconPrivate



    OnBeginSimulation<override>():void=
        (super:)OnBeginSimulation()
        
    OnEndSimulation<override>():void=
        (super:)OnEndSimulation()
