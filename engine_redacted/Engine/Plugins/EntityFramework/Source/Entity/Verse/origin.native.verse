# Copyright Epic Games, Inc. All Rights Reserved.

using { /Verse.org/Native }
using { /Verse.org/Simulation }
using { /Verse.org/SpatialMath }
using { /Verse.org/Temporary/EpicGamesRestricted/Network }

@doc("Interface to provide alternative origin to an entity which is defaulted to its parent. See `transform_component`")
origin<native><public> := interface:
    
    GetTransform<public><native_callable>()<transacts>:transform

@doc("class to provide alternative origin to the 'transform_component' as an entity")
entity_origin<native><public> := class<final>(origin):
    @editable
    @replicated("RepNotify")
    Entity<native><public>:entity



    GetTransform<override>()<transacts>:transform = Entity.GetGlobalTransform()
