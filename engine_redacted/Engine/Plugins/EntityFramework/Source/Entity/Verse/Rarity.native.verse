# Copyright Epic Games, Inc. All Rights Reserved.

using { /Verse.org/Colors }
using { /Verse.org/Colors/NamedColors }
using { /Verse.org/Native }
using { /Verse.org/Simulation }

@doc("Rarity may be used by gameplay and presentation systems to classify and rank things.")
@available{MinUploadedAtFNVersion := 4000} # rarity introduced at this version
rarity<native><public> := class<castable><concrete><unique><allocates>():   # this should use has_description - <redacted-ticket>
    @editable
    Color<native><public>:color = NamedColors.Grey

@available{MinUploadedAtFNVersion := 4000} # rarity introduced at this version
common_rarity<public> := class<allocates>(rarity):

@available{MinUploadedAtFNVersion := 4000} # rarity introduced at this version
uncommon_rarity<public> := class<allocates>(common_rarity):
    Color<override>:color = NamedColors.Green

@available{MinUploadedAtFNVersion := 4000} # rarity introduced at this version
rare_rarity<public> := class<allocates>(uncommon_rarity):
    Color<override>:color = NamedColors.Blue

@available{MinUploadedAtFNVersion := 4000} # rarity introduced at this version
epic_rarity<public> := class<allocates>(rare_rarity):
    Color<override>:color = NamedColors.Purple

@available{MinUploadedAtFNVersion := 4000} # rarity introduced at this version
legendary_rarity<public> := class<allocates>(epic_rarity):
    Color<override>:color = NamedColors.Orange
