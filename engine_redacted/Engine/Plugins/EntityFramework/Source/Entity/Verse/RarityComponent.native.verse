# Copyright Epic Games, Inc. All Rights Reserved.

using { /Verse.org }
using { /Verse.org/Native }
using { /Verse.org/Colors }
using { /Verse.org/SceneGraph }
using { /Verse.org/Simulation }
using { /Verse.org/Temporary/EpicGamesRestricted/Network }

@doc("Component that denotes the rarity of an entity.")
@available{MinUploadedAtFNVersion := 4000} # rarity introduced at this version
rarity_component<public><native> := class<final_super>(component):

    @doc("The rarity of this entity. An entity can have only one rarity.")
    var Rarity<public><getter(RedactedInternal)><setter(RedactedInternal)>:rarity = external{}
    
    @editable
    @display_name("Rarity")
    var RarityPrivate<private><native>:rarity = rarity{}
    
    @replicated("RepNotify")
    var RarityValue<private><native>:int = -1
    OnRep_RarityValue<private><native>():void

    @replicated("RepNotify")
    var RarityColor<private><native>:color = color{}
    OnRep_RarityColor<private><native>():void

    @replicated("RepNotify")
    var IsCustomRarity<private><native>:logic = false
    OnRep_IsCustomRarity<private><native>():void


    OnBeginSimulation<override>():void=
        (super:)OnBeginSimulation()
        
    OnEndSimulation<override>():void=
        (super:)OnEndSimulation()
