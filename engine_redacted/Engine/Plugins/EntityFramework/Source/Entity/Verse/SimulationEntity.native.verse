# Copyright Epic Games, Inc. All Rights Reserved.

using { /Verse.org/Verse }
using { /Verse.org/Native }
using { /Verse.org/Simulation }
using { /Verse.org/Temporary/EpicGamesRestricted/Network }

@doc("Returns the simulation entity parent for this entity.\n{
     }  * The simulation entity is the rootmost entity in an experience.\n{
     }  * Fails if this entity is not currently in the scene.")
(InEntity:entity).GetSimulationEntity<native><public>()<transacts><decides>:entity