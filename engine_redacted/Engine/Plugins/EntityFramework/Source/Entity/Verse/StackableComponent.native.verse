# Copyright Epic Games, Inc. All Rights Reserved.

using { /Verse.org/Native }
using { /Verse.org/SceneGraph }
using { /Verse.org/Simulation }
using { /Verse.org/Temporary/EpicGamesRestricted/Network }

@doc("A component that when attached to an entity allows for it to merge or 'stack' with other entities with compatible components.")
@available{MinUploadedAtFNVersion := 4000}
stackable_component<public><native> := class<abstract><final_super>(component):
    @doc("The current amount of this entity held in the stack.")
    @replicated("RepNotify")
    @editable_number(int):
        MinValue := option{1}
    var<private> StackSize<public><native>:int = 1
    OnRep_StackSize<private><native>():void

    @doc("The maximum amount this component can hold. If unset, it holds an unlimited amount.")
    @replicated("RepNotify")
    @editable
    var<private> MaxStackSize<public><native>:?int = option{1}
    OnRep_MaxStackSize<private><native>():void


    @doc("Sets the stack size of this component. If the provided stack size is invalid, e.g. negative or exceeding MaxStackSize, stack size will not change.")
    SetStackSize<public><final><native>(NewStackSize:int)<transacts>:void

    @doc("Sets the maximum stack size for this component. If NewMaxStackSize is false, stack size is unlimited. If ClampStackSize is true, StackSize will be clamped to NewMaxStackSize.")
    SetMaxStackSize<public><final><native>(NewMaxStackSize:?int, ?ClampStackSize:logic = false)<transacts>:void

    # Splits the specified amount out of this component’s stack, reducing this component’s stack size in the process.
    # Returns an instance of split_prefab_type with a stack size equal to the amount it was able to take from this entity.
    # If taking the exact (full) amount, the entity itself will be returned instead.
    Split<public><native_callable>(Amount:int)<transacts><decides>:entity

    @doc("Succeeds if this entity can be merged into the target entity. Merging an entity with itself will always fail. Note that merging should be checked in both directions!")
    CanMergeInto<public><native_callable>(TargetEntity:entity)<decides><reads>:void=
        Entity <> TargetEntity
        TargetStackableComponent := TargetEntity.GetComponent[stackable_component]

        # Allow native overrides
        if (AllowMerge := OverrideResult?):
            AllowMerge?

        # Check all the components for any vetoes on this merge.
        for:
            Component : Entity.GetComponents()
            MergeableInterface := has_merge_rules[Component]
        do:
            MergeableInterface.AllowMergeInto[TargetEntity]

    @doc("Attempts to merge this entity into the specified entity. Fails if entities cannot be merged.\n{
    } If TargetAmount is specified, only that amount will try to be merged into this entity. By default, the entire stack will try to merge.\n{
    } If the specified amount is invalid, the merge will fail.")
    MergeInto<public><native_callable>(TargetEntity:entity, ?TargetAmount:?int = false)<transacts><decides>:void=
        # Limit to the amount available. If none was specified, use the whole stack
        var Amount:int = StackSize
        if:
            ProvidedAmount := TargetAmount?
        then:
            set Amount = ProvidedAmount

        # Check the amount is valid.
        Amount > 0
        Amount <= StackSize


        for:
            Component : Entity.GetComponents()
            MergeableInterface := has_merge_rules[Component]
        do:
            MergeableInterface.OnMergeInto(TargetEntity, MergeAmount)

    ChangeStackSizeEvent<public><native><final>:listenable(change_stack_size_result) = redacted_internal(change_stack_size_result) {}
    ChangeMaxStackSizeEvent<public><native><final>:listenable(change_max_stack_size_result) = redacted_internal(change_max_stack_size_result) {}

    OnBeginSimulation<override>():void=
        (super:)OnBeginSimulation()
        
    OnEndSimulation<override>():void=
        (super:)OnEndSimulation()


@available{MinUploadedAtFNVersion := 4000}
change_stack_size_result<public><native> := class:
    StackableComponent<public><native>:stackable_component
    PreviousStackSize<public><native>:int
    CurrentStackSize<public><native>:int

@available{MinUploadedAtFNVersion := 4000}
change_max_stack_size_result<public><native> := class:
    StackableComponent<public><native>:stackable_component
    PreviousMaxStackSize<public><native>:?int
    CurrentMaxStackSize<public><native>:?int


@doc("Merge Rules Interface. Implemented by components that want to participate in entity mergeability checks.")
has_merge_rules<public> := interface<unique>:
    # Called when merge is attempted. To stop a merge with the target entity, this call must fail.
    AllowMergeInto<public>(TargetEntity:entity)<decides><reads>:void= {}

    # Called after a merge occurs. MergeAmount is the amount added due to this merge.
    OnMergeInto<public>(TargetEntity:entity, MergeAmount:int)<transacts>:void= {}


# Stackable component which both merges and splits based on a prefab field.
@available{MinUploadedAtFNVersion := 4000}
basic_stackable_component<public><native> := class(stackable_component):

    @doc("The prefab used when this entity merges with another or when it is split into a new instance.")
    @editable
    split_prefab_type<public><native>:castable_concrete_subtype(entity) = entity # default can be removed when component fields can be fully written in Verse

    # Splits the specified amount out of this component’s stack, reducing this components’s stack size in the process.
    # Returns an instance of split_prefab_type with a stack size equal to the amount it was able to take from this entity.
    # If taking the exact (full) amount, the entity itself will be returned instead.
    Split<override>(Amount:int)<transacts><decides>:entity =
        # This will fail if we try and Split() too much
        StackSize >= Amount

        # If we take the whole amount, just return this entity
        if (StackSize = Amount):
            return Entity
        else:
            # Reduce this component's stack size by Amount
            SetStackSize(StackSize - Amount)

            # Set the stack size in the new entity to Amount
            NewEntity := split_prefab_type {}
            NewStackingComponent := NewEntity.GetComponent[stackable_component]
            NewStackingComponent.SetStackSize(Amount)
            return NewEntity

    @doc("Succeeds if this entity can be merged into the target entity. Merging an entity with itself will always fail.")
    CanMergeInto<override>(TargetEntity:entity)<decides><reads>:void=
        (super:)CanMergeInto[TargetEntity]
        split_prefab_type[TargetEntity]
        TargetEntity.GetComponent[basic_stackable_component].split_prefab_type[Entity]
