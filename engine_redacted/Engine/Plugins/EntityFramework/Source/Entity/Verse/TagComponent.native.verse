# Copyright Epic Games, Inc. All Rights Reserved.

using { /Verse.org/Native }
using { /Verse.org/SceneGraph }
using { /Verse.org/Simulation }

# TODO: make tag_component implement diagnosable
@doc("Internal component used to add tags to an entity. Tagged entities can then be queried using `entity.FindDescendantEntitiesWithTag` and `entity.FindAncestorEntitiesWithTag`.\n{
     }  * Finding entities from the simulation entity for tagged entities only considers any `tag_component` that has run `OnAddedToScene`.")
@hide_in_editor #Tags are currently edited from the top-level entity
tag_component<native><internal> := class<final><final_super>(component, has_tags):
    @doc("Adds a `tag` instance to this container. Returns a `tag_key` that is uniquely associated with the added instance.")
    AddTag<native><override>(Tag:tag)<transacts>:tag_key
    
    @doc("Removes the tag instance associated with the `tag_key`, succeeds if an instance was removed, fails otherwise.")
    RemoveTag<native><override>(Key:tag_key)<transacts><decides>:void
    
    @doc("Removes all tag instances of type `tag_type`, succeeds if at least one instance was removed, fails otherwise.")
    RemoveAllTags<native><override>(tag_type:castable_subtype(tag))<transacts><decides>:void
    
    @doc("Removes all tag instances that are not of type `tag_type`, succeeds if at least one instance was removed, fails otherwise.")
    RemoveAllTagsExcept<native><override>(tag_type:castable_subtype(tag))<transacts><decides>:void
    
    @doc("Removes all tag instances that are not of any of the types in `tag_types`, succeeds if at least one instance was removed, fails otherwise.")
    RemoveAllTagsExcept<native><override>(tag_types:[]castable_subtype(tag))<transacts><decides>:void

    @doc("Succeeds if at least one type of `tag_type` is found in this container, fails otherwise.")
    ContainsTag<native><override>(tag_type:castable_subtype(tag))<decides><reads>:void
    
    @doc("Succeeds if all types in `tag_types` are found in this container, fails otherwise.")
    ContainsAllTags<native><override>(tag_types:[]castable_subtype(tag))<decides><reads>:void
    
    @doc("Succeeds if any of the types in `tag_types` is found in this container, fails otherwise.")
    ContainsAnyTag<native><override>(tag_types:[]castable_subtype(tag))<decides><reads>:void

    @editable
    @display_name("Tags")
    @doc("Internal container for the actual `tag` instances.")
