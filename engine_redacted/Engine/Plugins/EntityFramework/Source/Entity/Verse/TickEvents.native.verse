# Copyright Epic Games, Inc. All Rights Reserved.

using { /Verse.org/Native }

@doc("Describes discrete phases of a frame update. Subscribe to members of the tick_events object to run code {
     }before or after the physics system has updated your object, allowing you to affect or react to these updates.")
tick_events<native><public> := class:

    @doc("Listen `PrePhysics` to run your code before the physics system has updated your object this frame.")
    PrePhysics<public><native>:execution_listenable = execution_listenable{}

    @doc("Listen `PostPhysics` to run your code after the physics system has updated your object this frame.")
    PostPhysics<public><native>:execution_listenable = execution_listenable{}
