# Copyright Epic Games, Inc. All Rights Reserved.

using { /Verse.org/Native }
using { /Verse.org/Simulation }
using { /Verse.org/SpatialMath }
using { /Verse.org/Temporary/EpicGamesRestricted/Network }

# TODO uncomment when <redacted-ticket> is resolved
#LocalTransformToolTip<internal><localizes>:message = "Local transform to its origin which defaults to the parent entity."
#OriginToolTip<internal><localizes>:message = "Use this property to provide an alternate origin than the default parent entity."

@doc("Stores the transforms for an entity, which are used to position the entity.")
transform_component<native><public> := class<final><final_super>(component):
    
    # Global transform
    @doc("Current transform of the entity, globally referenced. Any value set at construction will be overrridden with the calculation from the Local Transform")
    var GlobalTransform<public><getter(GetGlobalTransform)><setter(SetGlobalTransform)>:transform = external{}

    var<private> GlobalTransformInternal<private><native>:transform = transform{}

    GetGlobalTransform<private><native>(:redacted_internal)<transacts>:transform
        # Using GetGlobalTransform() here instead of GlobalTransformInternal as the method will make sure it is fully updated
        case(MemberName):
            "Translation" => GetGlobalTransform(Accessor).Translation
            "Rotation" => GetGlobalTransform(Accessor).Rotation
            "Scale" => GetGlobalTransform(Accessor).Scale
            _ => Err("shouldn't get here. unknown member name")
        # Using GetGlobalTransform() here instead of GlobalTransformInternal as the method will make sure it is fully updated
        case(MemberName1):
            "Translation" => 
                case(MemberName2):
                    "Forward" => GetGlobalTransform(Accessor).Translation.Forward
                    "Left" => GetGlobalTransform(Accessor).Translation.Left
                    "Up" => GetGlobalTransform(Accessor).Translation.Up
                    _ => Err("shouldn't get here. unknown member name")
            "Scale" =>
                case(MemberName2):
                    "Forward" => GetGlobalTransform(Accessor).Scale.Forward
                    "Left" => GetGlobalTransform(Accessor).Scale.Left
                    "Up" => GetGlobalTransform(Accessor).Scale.Up
                    _ => Err("shouldn't get here. unknown member name")
            _ => Err("shouldn't get here. unknown member name")
    SetGlobalTransform<private><native>(:redacted_internal, NewGlobalTransform:transform)<transacts>:void
        # Using GetGlobalTransform() here instead of GlobalTransformInternal as the method will make sure it is fully updated
        var Transform:transform=GetGlobalTransform(Accessor)
        case(MemberName):
            "Translation" => set Transform.Translation = UnsafeCast(NewValue, vector3)
            "Rotation" => set Transform.Rotation = UnsafeCast(NewValue, rotation)
            "Scale" => set Transform.Scale = UnsafeCast(NewValue, vector3)
            _ => Err("shouldn't get here. unknown member name")
        set GlobalTransform = Transform
        # Using GetGlobalTransform() here instead of GlobalTransformInternal as the method will make sure it is fully updated
        var Transform:transform=GetGlobalTransform(Accessor)
        case(MemberName1):
            "Translation" => 
                case(MemberName2):
                    "Forward" => set Transform.Translation.Forward = NewValue
                    "Left" => set Transform.Translation.Left = NewValue
                    "Up" => set Transform.Translation.Up = NewValue
                    _ => Err("shouldn't get here. unknown member name")
            "Scale" =>
                case(MemberName2):
                    "Forward" => set Transform.Scale.Forward = NewValue
                    "Left" => set Transform.Scale.Left = NewValue
                    "Up" => set Transform.Scale.Up = NewValue
                    _ => Err("shouldn't get here. unknown member name")
            _ => Err("shouldn't get here. unknown member name")
        set GlobalTransform = Transform

    # Local transform
    @doc("LocalTransform to its parent/origin")
    var LocalTransform<public><getter(GetLocalTransform)><setter(SetLocalTransform)>:transform = external{}

    @editable:
    # TODO uncomment when <redacted-ticket> is resolved
    #    ToolTip := LocalTransformToolTip
    @display_name("LocalTransform")
    @replicated("RepNotify")
    var<private> LocalTransformInternal<private><native>:transform = transform{}

    GetLocalTransform<private><native>(:redacted_internal)<transacts>:transform
        # Using GetLocalTransform() here instead of LocalTransformInternal as the method will make sure it is fully updated
        case(MemberName):
            "Translation" => GetLocalTransform(Accessor).Translation
            "Rotation" => GetLocalTransform(Accessor).Rotation
            "Scale" => GetLocalTransform(Accessor).Scale
            _ => Err("shouldn't get here. unknown member name")
        # Using GetLocalTransform() here instead of LocalTransformInternal as the method will make sure it is fully updated
        case(MemberName1):
            "Translation" => 
                case(MemberName2):
                    "Forward" => GetLocalTransform(Accessor).Translation.Forward
                    "Left" => GetLocalTransform(Accessor).Translation.Left
                    "Up" => GetLocalTransform(Accessor).Translation.Up
                    _ => Err("shouldn't get here. unknown member name")
            "Scale" =>
                case(MemberName2):
                    "Forward" => GetLocalTransform(Accessor).Scale.Forward
                    "Left" => GetLocalTransform(Accessor).Scale.Left
                    "Up" => GetLocalTransform(Accessor).Scale.Up
                    _ => Err("shouldn't get here. unknown member name")
            _ => Err("shouldn't get here. unknown member name")

    SetLocalTransform<private><native>(:redacted_internal, NewLocalTransform:transform)<transacts>:void
        # Using GetLocalTransform() here instead of LocalTransformInternal as the method will make sure it is fully updated
        var Transform:transform=GetLocalTransform(Accessor)
        case(MemberName):
            "Translation" => set Transform.Translation = UnsafeCast(NewValue, vector3)
            "Rotation" => set Transform.Rotation = UnsafeCast(NewValue, rotation)
            "Scale" => set Transform.Scale = UnsafeCast(NewValue, vector3)
            _ => Err("shouldn't get here. unknown member name")
        set LocalTransform = Transform
        # Using GetLocalTransform() here instead of LocalTransformInternal as the method will make sure it is fully updated
        var Transform:transform=GetLocalTransform(Accessor)
        case(MemberName1):
            "Translation" => 
                case(MemberName2):
                    "Forward" => set Transform.Translation.Forward = NewValue
                    "Left" => set Transform.Translation.Left = NewValue
                    "Up" => set Transform.Translation.Up = NewValue
                    _ => Err("shouldn't get here. unknown member name")
            "Scale" =>
                case(MemberName2):
                    "Forward" => set Transform.Scale.Forward = NewValue
                    "Left" => set Transform.Scale.Left = NewValue
                    "Up" => set Transform.Scale.Up = NewValue
                    _ => Err("shouldn't get here. unknown member name")
            _ => Err("shouldn't get here. unknown member name")
        set LocalTransform = Transform
    OnRep_LocalTransformInternal<private><native>():void

    # Origin
    @doc("alternate origin than the default parent entity")
    var Origin<public><getter(GetOrigin)><setter(SetOrigin)>:?origin = external{}

    @editable:
    # TODO uncomment when <redacted-ticket> is resolved
    #    ToolTip := OriginToolTip
    @display_name("Origin")
    @replicated("RepNotify")
    var<private> OriginInternal<private><native>:?origin = false

    SetOrigin<private><native>(:redacted_internal, NewOrigin:?origin)<transacts>:void
    OnRep_OriginInternal<private><native>()<transacts>:void

@doc("Returns the global transform of this entity, in the case the entity does not have a transform_component it will return the transform of the first parent that has a transform")
(InEntity:entity).GetGlobalTransform<native><public>()<transacts>:transform


@doc("Returns the local transform of this entity, in the case the entity does not have a transform_component it will return identity")
(InEntity:entity).GetLocalTransform<native><public>()<transacts>:transform

@doc("Returns the origin of the entity in any, in the case the entity does not have a transform_component the method will fail")
(InEntity:entity).GetOrigin<native><public>()<transacts><decides>:origin

@doc("Sets the global transform of this entity, in the case the entity does not have a transform_component it will create one and set its global transform")
(InEntity:entity).SetGlobalTransform<native><public>(NewGlobalTransform:transform)<transacts>:void

@doc("Sets the local transform of this entity, in the case the entity does not have a transform_component it will create one and set its local transform")
(InEntity:entity).SetLocalTransform<native><public>(NewLocalTransform:transform)<transacts>:void

@doc("Sets the origin of this entity, in the case the entity does not have a transform_component it will create one and set its origin. This method fails if it recognized dependency recursion.")
(InEntity:entity).SetOrigin<native><public>(NewOrigin:origin)<transacts>:void

@doc("Resets the origin of this entity, which will now default to its parent")
(InEntity:entity).ResetOrigin<native><public>()<transacts>:void