# Copyright Epic Games, Inc. All Rights Reserved.

using { /Verse.org/Verse }
using { /Verse.org/Verse/Easing }
using { /Verse.org/Native }
using { /Verse.org/Simulation }
using { /Verse.org/Temporary/EpicGamesRestricted/Network }

@doc("An easing function combined with a relative time window")
@experimental
@available{MinUploadedAtFNVersion := 4100}
easing_window<native><public> := struct<concrete>:
    @doc("Easing function to apply")
    (easing_window:)Easing<native>:type{_(:float)<reads>:float} = (/Verse.org/Verse/Easing:)EaseInOut
    
    @doc("Duration of the easing window in seconds")
    @editable
    Duration<native>:float = 0.3
    
    @doc("Offset of the easing window in seconds\n{
         }The offset allows the window to be moved later or earlier relative to the easing modifier’s internal timeline.\n{
         }For example, moving the ease-in window later (ie +Offset) would cause the blend to happen later - ie delay it.\n{
         }For -ve offsets to ease-in and +ve offsets to ease-out, this has the effect of allowing ‘frozen’ states of targets\n{ 
         }to be observed (for time-varying targets).")
    @editable
    Offset<native>:float = 0.0

@doc("Interface used to trigger easing behaviors")
@experimental
@available{MinUploadedAtFNVersion := 4100}
easeable<public> := interface(cancelable):

    @doc("Trigger an ease-in.\n{
         }Triggering an ease-in again when already easing-in or when fully eased-in has no effect unless a different target weight is specified.")
    EaseIn()<transacts>:void

    @doc("Trigger an ease-in providing a new ease-in window, overriding any previously provided window.\n{
         }Triggering an ease-in again when already easing-in or when fully eased-in has no effect unless a different target weight is specified.")
    EaseIn(EaseInWindow:easing_window)<transacts>:void

    @doc("Trigger an ease-out.\n{
         }Triggering an ease-out again when already easing-out, or when fully eased-out has no effect")
    EaseOut<public>()<transacts>:void

    @doc("Trigger an ease-out. If an easing_window is provided here, then it will override any previously provided window.\n{
         }Triggering an ease-out again when already easing-out, or when fully eased-out has no effect")
    EaseOut<public>(EaseOutWindow:easing_window)<transacts>:void
