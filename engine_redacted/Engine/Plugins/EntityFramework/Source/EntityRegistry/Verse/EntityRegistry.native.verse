# Copyright Epic Games, Inc. All Rights Reserved.

using { /Verse.org/Native }
using { /Verse.org/SceneGraph }

####################################################
# Search down entity tree
####################################################


@doc("Finds all descendant entities including `InEntity` containing a component of type `component_type`.\n{
      }The order of the returned entities is unspecified and subject to change.")
(InEntity:entity).FindDescendantEntitiesWithComponent<native><public>(component_type:castable_subtype(component))<transacts>:generator(entity)

@doc("Finds all components attached to descendant entities to and including `InEntity` of type `component_type`.\n{
      }The order of the returned components is unspecified and subject to change.")
(InEntity:entity).FindDescendantComponents<native><public>(component_type:castable_subtype(component))<predicts><transacts>:generator(component_type)

####################################################
# Search up entity tree
####################################################


@doc("Finds all ancestor entities to `InEntity` containing a component of type `component_type`.\n{
      }The order of the returned entities is unspecified and subject to change.")
(InEntity:entity).FindAncestorEntitiesWithComponent<native><public>(component_type:castable_subtype(component))<transacts>:generator(entity)


@doc("Finds all components attached to ancestor entities to `InEntity` of type `component_type`.\n{
      }The order of the returned components is unspecified and subject to change.")
(InEntity:entity).FindAncestorComponents<native><public>(component_type:castable_subtype(component))<transacts><predicts>:generator(component_type)
