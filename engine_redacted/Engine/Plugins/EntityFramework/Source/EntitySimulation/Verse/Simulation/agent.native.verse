# Copyright Epic Games, Inc. All Rights Reserved.

using { /Verse.org/Native }
using { /Verse.org/SceneGraph }
using { /Verse.org/Temporary/EpicGamesRestricted/Network }

@replicatable
@import_as("/Script/Engine.FUniqueNetIdRepl")
unique_net_id_internal<native> := struct<concrete>{}

@hide_in_editor
agent<native><public> := class<unique><internal>(entity):


    @replicated("RepNotify")
    ReplicatedUniqueNetId<native><private>:unique_net_id_internal = unique_net_id_internal{}
    OnRep_ReplicatedUniqueNetId<private><native>():void


MakeLocalizableValue(Agent:agent):redacted_internal = redacted_internal{Value:=Agent}

