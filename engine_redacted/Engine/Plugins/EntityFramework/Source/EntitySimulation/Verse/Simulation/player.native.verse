# Copyright Epic Games, Inc. All Rights Reserved.

using { /Verse.org/Native }

@hide_in_editor
player<native><public> := class<unique><persistent><module_scoped_var_weak_map_key>(agent):
    @doc("{
    }Succeeds when this `player` may be used as a module-scoped `var` `weak_map` key. {
    }This coincides with the corresponding player having joined the game and not yet {
    }left. Using a `player` as a module-scope `var` `weak_map` key when this method {
    }fails results in a runtime error.")
    IsActive<native><public>()<reads><decides>:void
