# Copyright Epic Games, Inc. All Rights Reserved.

using { /Verse.org/Verse }
using { /Verse.org/Native }
using { /Verse.org/Simulation }
using { /Verse.org/Assets }
using { /Verse.org/Temporary/EpicGamesRestricted/Network }
using { /UnrealEngine.com/Temporary }

@doc("A modifier of skeletons, used to animate meshes using skeletal animation")
@experimental
@available{MinUploadedAtFNVersion := 4100}
skeletal_animation<public><native> := class<castable>(redacted_internal, modifier(skeleton)):

    Evaluate<native><override>(InValue:skeleton)<reads>:skeleton

    # playable interface
    Play<native>()<transacts>:void
    Pause<native>()<transacts>:void
    Stop<native>()<transacts>:void

    @doc("Get this animation's lifetime, in seconds.{
         }Fails if an animation has an indefinite duration (i.e. it is not possible at the point of the call to tell how long the animation is){
         }Animations can be of infinite length (such as a looping animation)")
    GetDuration<native_callable>()<reads><decides>:float =
        false
        Inf

    # The time that this animation begins
