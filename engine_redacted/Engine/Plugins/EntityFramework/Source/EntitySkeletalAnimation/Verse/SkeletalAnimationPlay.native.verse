# Copyright Epic Games, Inc. All Rights Reserved.

using { /Verse.org/Verse }
using { /Verse.org/Native }
using { /Verse.org/Simulation }

@experimental
@available{MinUploadedAtFNVersion := 4100}
play_skeletal_animation_result<public> := struct:
    
    @doc("The skeletal_animation that was played")
    Animation:skeletal_animation
    
    @doc("An easeable interface used to control blending")
    Easeable:easeable

@doc("Play a skeletal_animation on an entity")
@experimental
@available{MinUploadedAtFNVersion := 4100}
(Entity:entity).PlaySkeletalAnimation<public>(Animation:skeletal_animation, ?EaseInWindow:easing_window = easing_window{}, ?EaseOutWindow:easing_window = easing_window{})<decides><transacts> : play_skeletal_animation_result =

    # Subscribe to ease out when animation is nearly done

    # Subscribe to stop the animation on ease-out

    # Subscribe to remove from the stack once eased-out

    # Start playback
    Animation.Play()

    # Ease in the blend

    play_skeletal_animation_result:
        Animation := Animation
