# Copyright Epic Games, Inc. All Rights Reserved.

using { /Verse.org/Verse }
using { /Verse.org/Native }
using { /Verse.org/Simulation }

@doc("Skeletons are collections of bones & sets/chains")
@experimental
@available{MinUploadedAtFNVersion := 4100}
skeleton<public><native> := class{}