# Copyright Epic Games, Inc. All Rights Reserved.

using { /Verse.org/Assets }
using { /Verse.org/Native }
using { /Verse.org/Simulation }

@doc("Entity client streaming modes")
entity_streaming_policy<native><public> := enum:
    @doc("Entity will be spatially loaded.")
    Spatial,

    @doc("Entity will be non-spatially loaded.")
    NonSpatial,

    @doc("Entity will be always loaded.")
    Persistent

@doc("Child entities client streaming modes")
children_streaming_policy<native><public> := enum:
    @doc("Child entities are loaded atomically with their parent.")
    Atomic,
    
    @doc("Child entities are loaded independently from their parent.")
    Discrete
