# Copyright Epic Games, Inc. All Rights Reserved.

using { /Verse.org/Native }
using { /Verse.org/Concurrency }


@doc("Users to subscribe to, or await on, a DeltaTime based callback from one of the phases in a component's `TickEvents` object.")
execution_listenable<public><native> := class(listenable(float)):

    @doc("Suspends the current task until resumed by a matching call to `signalable.Signal`. Returns the event `payload`.")
    Await<native><override>()<suspends>:float

    @doc("Subscribe a callback function to this TickEvent phase. The input parameter to your function (DeltaTime) is the time that has {
         }passed between the last update and the current update.")
    Subscribe<native><override>(Callback(:float):void)<transacts>:cancelable

