# Copyright Epic Games, Inc. All Rights Reserved.

new_class_template := class:
    #TODO: Add members and methods here
