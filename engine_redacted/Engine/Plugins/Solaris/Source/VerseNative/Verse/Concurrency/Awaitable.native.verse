# Copyright Epic Games, Inc. All Rights Reserved.

using{ /Verse.org/Native }

# Cannot import /Verse.org/Verse, so cannot use embedded { } as newlines in docstrings in this file.

@doc("A parametric interface implemented by events with a `payload` that can be waited on. Matched with `signalable.`")
awaitable<public>(payload:type) := interface:

    @doc("Suspends the current task until resumed by a matching call to `signalable.Signal`. Returns the event `payload`.")
    Await<public>()<suspends>:payload

awaitable<public>() := awaitable(void)
