# Copyright Epic Games, Inc. All Rights Reserved.

using{ /Verse.org/Native }

#----------------------------------------------------------------------------------------
# A stateful, asynchronous future used to represent the invoked context for an async
# expression (such as an invoked coroutine) executing concurrently in a cooperatively
# multitasked environment over time - having a lifespan of one or more updates / ticks /
# frames of the Verse runtime system before it completes.
# A task is usually obtained from the result of a unstructured concurrency `spawn` expression.
@cpp_name("verse::task")
task<native><public>(t:type) := class<abstract><final>(awaitable(t)):

    # Phase query methods

    # Wait until the current task has completed - this essentially anchors this task and adds a
    # caller for it to return to at this call point.
    # Notes:
    #   - Multiple `Await()` calls can be made on this one same task - essentially giving it
    #     multiple callers to return to.
    #   - The order that `Await()` calls are accumulated is important - they are woken in first in
    #     first out (FIFO) order.
    #   - If this task has already completed, then this coroutine completes immediately.
    #   - If this task is canceled, this `Await()` coroutine will not be notified and will appear
    #     to take forever.
    #   - If this `Await()` task is canceled then it will automatically unregister itself to be
    #     woken up from this task.
    #   - This task is not registered as a subtask to the `Await()` task so if the `Await()` task
    #     or any of its calling tasks are canceled, this task will *not* also be canceled as with
    #     a standard subtask of a caller.
    Await<native><override>()<suspends>:t

    # Cancel any subtasks and then this task itself.
    # Notes:
    #   - Once `Cancel()` is called, the task will change phase from "active" to "canceling" (if
    #     there is any associated deferred code) and then "canceled" once it is fully canceled.
    #   - The current phase of a task can be determined with the various task query methods.
    #   - Subtasks are canceled in last in first out (LIFO) order giving newest subtasks a chance
    #     to reference info managed in older subtasks during canceling.
    #   - If new subtasks are accumulated in the middle of canceling then they will also be canceled.
    #   - If this task has a caller or calls to `Await()`, they will not be notified (or canceled)
    #     and will appear to take forever.
    #   - Calling `Cancel()` on a task that is currently running (either the current task or a task
    #     that started or resumed it directly or indirectly) will suspend the caller, until the
    #     target task reaches a cancellation point (i.e. it tries to suspend).
    #     ***Warning: This is not implemented on the BPVM, where this kind of cancellation will have
    #     unexpected results.
    Cancel<native>()<suspends>:void
