# Copyright Epic Games, Inc. All Rights Reserved.

using{ /Verse.org/Native }

#----------------------------------------------------------------------------------------
# Manages a set of coroutine tasks associated with it
bp_vm_only:
    @cpp_name("verse::task_group")
    task_group<internal><native> := class<final> {}
