# Copyright Epic Games, Inc. All Rights Reserved.

@attribscope_class @attribscope_struct @attribscope_enum @attribscope_interface @attribscope_module
import_as_attribute := class<internal><computes>(attribute):
    ImportName:string

import_as<constructor>(ImportName:string)<computes> := import_as_attribute:
    ImportName := ImportName
