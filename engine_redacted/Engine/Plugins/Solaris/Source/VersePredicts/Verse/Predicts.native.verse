# Copyright Epic Games, Inc. All Rights Reserved.

using { /Verse.org/Native }

### registering Verse objects (ie. class instances) with the <predicts> runtime
PredictsServerRegisterObject<native>(:any)<transacts>:any

### internals
PredictsIsRunningOnServer<native>()<transacts><decides><predicts>:void








