# Copyright Epic Games, Inc. All Rights Reserved.

using { /Verse.org/Native }

@import_as("/Script/Engine.UCurveFloat")
curve_float_base<native> := class{}


editable_curve<native><public> := class<final><concrete>(curve_float_base):

    @doc("Evaluates this float curve at the specified time and returns the result as a float")
    Evaluate<native><public>(Time:float):float
