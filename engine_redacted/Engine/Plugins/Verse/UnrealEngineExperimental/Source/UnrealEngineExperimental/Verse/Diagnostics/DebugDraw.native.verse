# Copyright Epic Games, Inc. All Rights Reserved.

using { /Verse.org/Native }
using { /Verse.org/SpatialMath }
using { /Verse.org/Colors }

using { /UnrealEngine.com/Temporary/SpatialMath }

# Range of debug draw utilities. 
# Users need only provide the minimal subset of optional parameters they care about, and sensible defaults will be provided for those which they do not specify.
#    
# Examples:
#
#     - Draws a sphere at the named location. Default draw parameters for color, duration, style. Default radius/segment count.
# DrawSphere(Pos)
#
#     - Draws a Red sphere at the named location. Duration/Style uses defaults. Default radius/segment count.
# DrawSphere(Pos, ?Color:=Colors.Red} 
#
#     - Draws a Red sphere at the named location. Draws a 2m radius red sphere of 1cm thickness. Duration uses defaults.
# DrawSphere(Pos, ?Color:=Colors.Red, ?Radius:= 200f, ?Thickness:= 1f} 
#
#     - Draws a sphere at the named location for an infinite duration.
# DrawSphere(Pos, ?DrawDurationPolicy:=debug_draw_duration_policy.Persistent) 
#
#     - Draws a sphere at the named location for a user specified finite duration.
# DrawSphere(Pos, ?DrawDurationPolicy:=debug_draw_duration_policy.FiniteDuration, ?Duration:=7.5f) 
#
#     - Draws a sphere at the named location for one frame.
# DrawSphere(Pos, ?DrawDurationPolicy:=debug_draw_duration_policy.SingleFrame) 


@doc("Enumerated presets for policies describing a desired draw duration.")
debug_draw_duration_policy<native><public> := enum:
    SingleFrame
    FiniteDuration
    Persistent

@doc("debug_draw_channel is the base class used to define debug draw channels.")
debug_draw_channel<native><public> := class<abstract>(){}

DefaultDebugDrawColor<internal> : color = NamedColors.Yellow
DefaultDebugDrawDurationPolicy<internal> : debug_draw_duration_policy = debug_draw_duration_policy.FiniteDuration
DefaultDebugDrawDuration<internal> : float = 5.0
DefaultDebugDrawSize<internal> : float = 10.0
DefaultDebugDrawThickness<internal> : float = 0.0
DefaultDebugTextFontScale<internal> : float = 1.0

@doc("debug draw class to draw debug shapes on screen.")
debug_draw<native><public> := class:

    @doc("Channel will be used to clear specific debug draw.")
    Channel<native><public>:subtype(debug_draw_channel) := debug_draw_channel 
    
    @doc("Show Debug Draw for the channel for all users.")
    ShowChannel<native><public>()<transacts>:void

    @doc("Hide Debug Draw for the channel for all users.")
    HideChannel<native><public>()<transacts>:void

    @doc("Clears all debug draw for the channel.")
    ClearChannel<native><public>()<transacts>:void

    @doc("Clears all debug draw from this debug_draw instance.")
    Clear<native><public>()<transacts>:void

    @doc("Draws a sphere at the named location, and using the provided draw parameters.")
    DrawSphere<native><public>(Center:(/Verse.org/SpatialMath:)vector3, ?Radius:float = DefaultDebugDrawSize, ?Color:color = DefaultDebugDrawColor, ?NumSegments:int = 12, ?Thickness:float = DefaultDebugDrawThickness, ?DrawDurationPolicy:debug_draw_duration_policy = DefaultDebugDrawDurationPolicy, ?Duration:float = DefaultDebugDrawDuration)<transacts>:void
    @doc("Draws a sphere at the named location, and using the provided draw parameters.")
    DrawSphere<public>(Center:(/UnrealEngine.com/Temporary/SpatialMath:)vector3, ?Radius:float = DefaultDebugDrawSize, ?Color:color = DefaultDebugDrawColor, ?NumSegments:int = 12, ?Thickness:float = DefaultDebugDrawThickness, ?DrawDurationPolicy:debug_draw_duration_policy = DefaultDebugDrawDurationPolicy, ?Duration:float = DefaultDebugDrawDuration)<transacts>:void = 

    @doc("Draws a box at the named location, and using the provided draw parameters")
    DrawBox<native><public>(Center:(/Verse.org/SpatialMath:)vector3, Rotation:(/Verse.org/SpatialMath:)rotation, ?Extent:(/Verse.org/SpatialMath:)vector3 = (/Verse.org/SpatialMath:)vector3{Forward := DefaultDebugDrawSize, Left := DefaultDebugDrawSize, Up := DefaultDebugDrawSize}, ?Color:color = DefaultDebugDrawColor, ?Thickness:float = DefaultDebugDrawThickness, ?DrawDurationPolicy:debug_draw_duration_policy = DefaultDebugDrawDurationPolicy, ?Duration:float = DefaultDebugDrawDuration)<transacts>:void
    @doc("Draws a box at the named location, and using the provided draw parameters")
    DrawBox<public>(Center:(/UnrealEngine.com/Temporary/SpatialMath:)vector3, Rotation:(/UnrealEngine.com/Temporary/SpatialMath:)rotation, ?Extent:(/UnrealEngine.com/Temporary/SpatialMath:)vector3 = (/UnrealEngine.com/Temporary/SpatialMath:)vector3{X := DefaultDebugDrawSize, Y := DefaultDebugDrawSize, Z := DefaultDebugDrawSize}, ?Color:color = DefaultDebugDrawColor, ?Thickness:float = DefaultDebugDrawThickness, ?DrawDurationPolicy:debug_draw_duration_policy = DefaultDebugDrawDurationPolicy, ?Duration:float = DefaultDebugDrawDuration)<transacts>:void =

    @doc("Draws a capsule at the named location, and using the provided draw parameters.")
    DrawCapsule<native><public>(Center:(/Verse.org/SpatialMath:)vector3, Rotation:(/Verse.org/SpatialMath:)rotation, ?Height:float = DefaultDebugDrawSize, ?Radius:float = 25.0, ?Color:color = DefaultDebugDrawColor, ?Thickness:float = DefaultDebugDrawThickness, ?DrawDurationPolicy:debug_draw_duration_policy = DefaultDebugDrawDurationPolicy, ?Duration:float = DefaultDebugDrawDuration)<transacts>:void
    @doc("Draws a capsule at the named location, and using the provided draw parameters.")
    DrawCapsule<public>(Center:(/UnrealEngine.com/Temporary/SpatialMath:)vector3, Rotation:(/UnrealEngine.com/Temporary/SpatialMath:)rotation, ?Height:float = DefaultDebugDrawSize, ?Radius:float = 25.0, ?Color:color = DefaultDebugDrawColor, ?Thickness:float = DefaultDebugDrawThickness, ?DrawDurationPolicy:debug_draw_duration_policy = DefaultDebugDrawDurationPolicy, ?Duration:float = DefaultDebugDrawDuration)<transacts>:void =

    @doc("Draws a cone at the named location, and using the provided draw parameters.")
    DrawCone<native><public>(Origin:(/Verse.org/SpatialMath:)vector3, Direction:(/Verse.org/SpatialMath:)vector3, ?Height:float = DefaultDebugDrawSize, ?NumSides:int = 12, ?AngleWidthRadians:float = PiFloat / 4.0, ?AngleHeightRadians:float = PiFloat / 4.0, ?Color:color = DefaultDebugDrawColor, ?Thickness:float = DefaultDebugDrawThickness, ?DrawDurationPolicy:debug_draw_duration_policy = DefaultDebugDrawDurationPolicy, ?Duration:float = DefaultDebugDrawDuration)<transacts>:void
    @doc("Draws a cone at the named location, and using the provided draw parameters.")
    DrawCone<public>(Origin:(/UnrealEngine.com/Temporary/SpatialMath:)vector3, Direction:(/UnrealEngine.com/Temporary/SpatialMath:)vector3, ?Height:float = DefaultDebugDrawSize, ?NumSides:int = 12, ?AngleWidthRadians:float = PiFloat / 4.0, ?AngleHeightRadians:float = PiFloat / 4.0, ?Color:color = DefaultDebugDrawColor, ?Thickness:float = DefaultDebugDrawThickness, ?DrawDurationPolicy:debug_draw_duration_policy = DefaultDebugDrawDurationPolicy, ?Duration:float = DefaultDebugDrawDuration)<transacts>:void =

    @doc("Draws a cylinder at the named location, and using the provided draw parameters.")
    DrawCylinder<native><public>(Start:(/Verse.org/SpatialMath:)vector3, End:(/Verse.org/SpatialMath:)vector3, ?NumSegments:int = 12, ?Radius:float = DefaultDebugDrawSize, ?Color:color = DefaultDebugDrawColor, ?Thickness:float = DefaultDebugDrawThickness, ?DrawDurationPolicy:debug_draw_duration_policy = DefaultDebugDrawDurationPolicy, ?Duration:float = DefaultDebugDrawDuration)<transacts>:void
    @doc("Draws a cylinder at the named location, and using the provided draw parameters.")
    DrawCylinder<public>(Start:(/UnrealEngine.com/Temporary/SpatialMath:)vector3, End:(/UnrealEngine.com/Temporary/SpatialMath:)vector3, ?NumSegments:int = 12, ?Radius:float = DefaultDebugDrawSize, ?Color:color = DefaultDebugDrawColor, ?Thickness:float = DefaultDebugDrawThickness, ?DrawDurationPolicy:debug_draw_duration_policy = DefaultDebugDrawDurationPolicy, ?Duration:float = DefaultDebugDrawDuration)<transacts>:void =
        DrawCylinder(FromVector3(Start), FromVector3(End), ?NumSegments:=NumSegments, ?Radius:=Radius, ?Color:=Color, ?Thickness:=Thickness, ?DrawDurationPolicy:=DrawDurationPolicy, ?Duration:=Duration)

    @doc("Draws a line from Start to End locations, and using the provided draw parameters.")
    DrawLine<native><public>(Start:(/Verse.org/SpatialMath:)vector3, End:(/Verse.org/SpatialMath:)vector3, ?Color:color = DefaultDebugDrawColor, ?Thickness:float = DefaultDebugDrawThickness, ?DrawDurationPolicy:debug_draw_duration_policy = DefaultDebugDrawDurationPolicy, ?Duration:float = DefaultDebugDrawDuration)<transacts>:void
    @doc("Draws a line from Start to End locations, and using the provided draw parameters.")
    DrawLine<public>(Start:(/UnrealEngine.com/Temporary/SpatialMath:)vector3, End:(/UnrealEngine.com/Temporary/SpatialMath:)vector3, ?Color:color = DefaultDebugDrawColor, ?Thickness:float = DefaultDebugDrawThickness, ?DrawDurationPolicy:debug_draw_duration_policy = DefaultDebugDrawDurationPolicy, ?Duration:float = DefaultDebugDrawDuration)<transacts>:void =
        DrawLine(FromVector3(Start), FromVector3(End), ?Color:=Color, ?Thickness:=Thickness, ?DrawDurationPolicy:=DrawDurationPolicy, ?Duration:=Duration)

    @doc("Draws a point at the named location, and using the provided draw parameters.")
    DrawPoint<native><public>(Position:(/Verse.org/SpatialMath:)vector3, ?Color:color = DefaultDebugDrawColor, ?Thickness:float = DefaultDebugDrawThickness, ?DrawDurationPolicy:debug_draw_duration_policy = DefaultDebugDrawDurationPolicy, ?Duration:float = DefaultDebugDrawDuration)<transacts>:void
    @doc("Draws a point at the named location, and using the provided draw parameters.")
    DrawPoint<public>(Position:(/UnrealEngine.com/Temporary/SpatialMath:)vector3, ?Color:color = DefaultDebugDrawColor, ?Thickness:float = DefaultDebugDrawThickness, ?DrawDurationPolicy:debug_draw_duration_policy = DefaultDebugDrawDurationPolicy, ?Duration:float = DefaultDebugDrawDuration)<transacts>:void =

    @doc("Draws an arrow pointing from Start to End locations, and using the provided draw parameters.")
    DrawArrow<native><public>(Start:(/Verse.org/SpatialMath:)vector3, End:(/Verse.org/SpatialMath:)vector3, ?ArrowSize:float = 25.0, ?Color:color = DefaultDebugDrawColor, ?Thickness:float = DefaultDebugDrawThickness, ?DrawDurationPolicy:debug_draw_duration_policy = DefaultDebugDrawDurationPolicy, ?Duration:float = DefaultDebugDrawDuration)<transacts>:void
    @doc("Draws an arrow pointing from Start to End locations, and using the provided draw parameters.")
    DrawArrow<public>(Start:(/UnrealEngine.com/Temporary/SpatialMath:)vector3, End:(/UnrealEngine.com/Temporary/SpatialMath:)vector3, ?ArrowSize:float = 25.0, ?Color:color = DefaultDebugDrawColor, ?Thickness:float = DefaultDebugDrawThickness, ?DrawDurationPolicy:debug_draw_duration_policy = DefaultDebugDrawDurationPolicy, ?Duration:float = DefaultDebugDrawDuration)<transacts>:void =
        DrawArrow(FromVector3(Start), FromVector3(End), ?ArrowSize:=ArrowSize, ?Color:=Color, ?Thickness:=Thickness, ?DrawDurationPolicy:=DrawDurationPolicy, ?Duration:=Duration)

    @doc("Draws a 3D text using the provided draw parameters.")
    DrawText<native><public>(Text:string, Position:(/Verse.org/SpatialMath:)vector3, ?Color:color = DefaultDebugDrawColor, ?DrawDurationPolicy:debug_draw_duration_policy = DefaultDebugDrawDurationPolicy, ?Duration:float = DefaultDebugDrawDuration, ?FontScale:float = DefaultDebugTextFontScale, ?DrawDropShadow:logic = false)<transacts>:void
    
    @doc("Draws a 3D text using the provided draw parameters.")
    DrawText<public>(Text:string, Position:(/UnrealEngine.com/Temporary/SpatialMath:)vector3, ?Color:color = DefaultDebugDrawColor, ?DrawDurationPolicy:debug_draw_duration_policy = DefaultDebugDrawDurationPolicy, ?Duration:float = DefaultDebugDrawDuration, ?FontScale:float = DefaultDebugTextFontScale, ?DrawDropShadow:logic = false)<transacts>:void =
