# Copyright Epic Games, Inc. All Rights Reserved.

using { /Verse.org/Native }
using { /Verse.org/Colors }

<#
Logging example:

    using { /UnrealEngine.com/Temporary/Diagnostics }
    
    log_foo := class(log_channel){}
    log_bar := class(log_channel){}
    
    foo := class():
        # Can have two (or more) loggers tied to the same log_channel
        LogFoo<private>:log = log{Channel := log_foo}
        LogFooWarning<private>:log = log{Channel:=log_foo, DefaultLevel:== log_level.Warning}
    
        DoFoo():void=
            # Send "LogVerse: log_foo: Hello world!" to UE LOG
            LogFoo.Print("Hello world!")
    
        DoBar():void=
            # We can instantiate a new default log on the fly with a different channel setup
            LogBar := log{Channel:=log_bar, DefaultLevel:=log_level.Verbose}
    
            # Sends "LogVerse: log_bar: Hello Bar" to UE LOG
            LogBar.Print("Hello Bar")
    
            # Sends "LogVerse: Warning: log_bar: Warning Bar" to UE LOG
            LogBar.Print("Warning Bar", ?Level:=log_level.Warning)
#>

@doc("log levels available for various log commands")
log_level<native><public> := enum:
    Debug
    Verbose
    Error

@doc("log_channel is the base class used to define log channels. When printing a message to a log, the log channel class name will be prefixed to the output message.")
log_channel<native><public> := class<abstract>(){}

@doc("log class to send messages to the default log")
log<native><public> := class:

    @doc("Channel class name will be added as a prefix used when printing the message e.g. '[log_channel]: #Message")
    Channel<native><public>:subtype(log_channel)

    @doc("Sets the default log level of the displayed message. See log_level enum for more info on log levels. Defaults to log_level.Normal.")
    DefaultLevel<native><public><predicts>:log_level = log_level.RedactedInternal

    @doc("Sets the current debug diagnostic display `Color` that will be applied to all future Print calls. By default, `Color` is `NamedColors.White`, Only used if Duration is greater than `0.0`.")
    @available{MinUploadedAtFNVersion := 4000}
    Color<native>:color = NamedColors.White

    @doc("If set to greater than 0.0, displays the log on the screen in debug environments.")
    @available{MinUploadedAtFNVersion := 4000}
    Duration<native>:float = -1.0

    @doc("Print `Message` using the given log level.")
    (log:)Print<public>(Message:string, ?Level:log_level = DefaultLevel)<computes><predicts>:void=PrivatePrintImpl(Message, Level)

    @doc("Print `Message` diagnostic using the given log level.")

    @doc("Prints the current script call stack using the given log level.")
    PrintCallStack<native><public>(?Level:log_level = DefaultLevel)<computes>:void

    # This implements both the string and diagnostic Print methods.
    # Since it will receive the contents of the opaque diagnostic messages as a string, it needs to be private/final to ensure user-defined subclasses of log
    # can't inspect the contents of diagnostic messages as a string.
    PrivatePrintImpl<native><private><final>(Message:string, Level:log_level)<computes><predicts>:void
