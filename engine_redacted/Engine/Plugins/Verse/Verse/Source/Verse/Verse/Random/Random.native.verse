# Copyright Epic Games, Inc. All Rights Reserved.
using { /Verse.org/Native }

@doc("Returns a uniformly distributed, cryptographically-secure random `float` between `Low` and `High`, inclusive. (`Low` and `High` can be out of order.)")
GetRandomFloat<native><public>(Low:float, High:float)<transacts>:float

@doc("Returns a uniformly distributed, cryptographically-secure random `int` between `Low` and `High`, inclusive. (`Low` and `High` can be out of order.)")
GetRandomInt<native><public>(Low:int, High:int)<transacts>:int

@doc("Makes an `array` with the same elements as `Input` shuffled in a random order.")
Shuffle<public>(Input:[]t where t:type)<transacts>:[]t=
    for(i:=0..Length-2):
        j := GetRandomInt(i,Length-1)
        if: # should be assert
            Temp1 := Result[j]
            Temp2 := Result[i]
            set Result[j] = Temp2
            set Result[i] = Temp1
    return Result

#TODO: Add a psuedo random stream that can start with a specified seed
