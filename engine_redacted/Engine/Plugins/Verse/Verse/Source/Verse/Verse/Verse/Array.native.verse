# Copyright Epic Games, Inc. All Rights Reserved.

using { /Verse.org/Native }

@doc("Makes a flattened `array` by concatenating the elements of `Arrays`.")
Concatenate<public>(Arrays:[][]t where t:type)<computes>:[]t=
    if (Arrays.Length=1; Result:=Arrays[0]) then Result
    else if (Arrays.Length=2; A0:=Arrays[0]; A1:=Arrays[1]) then A0+A1
    else if (Arrays.Length=3; A0:=Arrays[0]; A1:=Arrays[1]; A2:=Arrays[2]) then A0+A1+A2
    else if (Arrays.Length>3; Midpoint:=Floor(Arrays.Length/2); A0:=Arrays.Slice[0, Midpoint]; A1:=Arrays.Slice[Midpoint, Arrays.Length]) then Concatenate(A0)+Concatenate(A1)
    else:
        #assert{Array.Length=0}
        array{}

@doc("Makes an `array` containing `Input`'s elements from `StartIndex` to `StopIndex-1`.\n{
    }Fails unless `0 <= StartIndex <= StopIndex <= Input.Length`.")
(Input:[]t where t:type).Slice<public>(StartIndex:int, StopIndex:int)<computes><decides>:[]t=

@doc("Makes an `array` containing `Input`'s elements from `StartIndex` to `Input.Length-1`.\n{
     }Succeeds if `0 <= StartIndex <= Input.Length`.")
(Input:[]t where t:type).Slice<public>(StartIndex:int)<computes><decides>:[]t=

@doc("Makes an `array` by inserting `ElementsToInsert` into `Input` such that the first element of `ElementsToInsert` is at `InsertionIndex`.")
(Input:[]t where t:type).Insert<public>(InsertionIndex:int, ElementsToInsert:[]t)<computes><decides>:[]t=

@doc("Makes an `array` by removing the element at `IndexToRemove` from `Input`.\n{
     }Succeeds if `0 <= IndexToRemove <= Input.Length-1`.")
(Input:[]t where t:type).RemoveElement<public>(IndexToRemove:int)<computes><decides>:[]t=

@doc("Makes an `array` by removing the element at the lowest index that equals `ElementToRemove` from `Input`.\n{
     }Fails if `Input` did not contain any instances of `ElementToRemove`.")
(Input:[]t where t:subtype(comparable)).RemoveFirstElement<public>(ElementToRemove:t)<computes><decides>:[]t=

@doc("Makes an `array` by removing all elements that equal `ElementToRemove` from `Input`.")
(Input:[]t where t:subtype(comparable)).RemoveAllElements<public>(ElementToRemove:t)<computes>:[]t=

@doc("Makes an `array` by replacing the element at `IndexToReplace` with `ElementToReplaceWith` in `Input`.\n{
     }Succeeds if `0 <= IndexToReplace <= Input.Length-1`.")
(Input:[]t where t:type).ReplaceElement<public>(IndexToReplace:int, ElementToReplaceWith:t)<computes><decides>:[]t=

@doc("Makes an `array` by replacing the element at the lowest index that equals `ElementToReplace` with `ElementToReplaceWith` in `Input`.\n{
     }Fails if `Input` did not contain any instances of `ElementToReplace`.")
(Input:[]t where t:subtype(comparable)).ReplaceFirstElement<public>(ElementToReplace:t, ElementToReplaceWith:t)<computes><decides>:[]t=

@doc("Makes an `array` by replacing all elements that equal `ElementToReplace` with `ElementToReplaceWith` in `Input`.")
(Input:[]t where t:subtype(comparable)).ReplaceAllElements<public>(ElementToReplace:t, ElementToReplaceWith:t)<computes>:[]t=
        if(InputElement=ElementToReplace) then ElementToReplaceWith else InputElement

@doc("Makes an `array` by replacing all ranges of elements that equal `ElementsToReplace` with `Replacement` in `Input`.\n{
     }When there are multiple overlapping instances of `ElementsToReplace` in `Input`, only the position with the lowest {
     }index is replaced.")
(Input:[]t where t:subtype(comparable)).ReplaceAll<public>(ElementsToReplace:[]t, Replacement:[]t)<transacts>:[]t =
    ToReplaceLength := ElementsToReplace.Length

    var Result:[]t = array{}
    var I:int = 0
    var SliceBeginIndex:int = 0

    loop:
        if (ToReplaceLength > Length - I):
            # There isn't enough length left to match against, so we're done
            return Result

        var Matches:logic = true
        var J:int = 0
        loop:
            if (J >= ToReplaceLength):
                break
                set Matches = false
                break
            set J += 1

        if (Matches?):
            set I += ToReplaceLength
            set SliceBeginIndex = I
        else:
            set I += 1

@doc("Makes an `array` by removing `Input`'s elements from `StartIndex` to `StopIndex-1`.\n{
    }Succeeds if `0 <= StartIndex <= StopIndex <= Input.Length`.")
(Input:[]t where t:type).Remove<public>(StartIndex:int, StopIndex:int)<computes><decides>:[]t =

@doc("Returns the first index whose element in `Input` equals `ElementToFind`.\n{
     }Fails if ElementToFind does not exist in the array.")
(Input:[]t where t:subtype(comparable)).Find<public>(ElementToFind:t)<computes><decides>:int=
    Results[0]

@available{MinUploadedAtFNVersion := 4400}
@doc("Returns the last element in `Input` if one exists.\n{
     }Fails if `Input.Length = 0`.")
(Input:[]t where t:type).Last<public>()<computes><decides>:t=

@available{MinUploadedAtFNVersion := 4400}
@doc("Returns the element at the position `OffsetFromLast` before the last element of `Input`.\n{
     }Fails if Input.Length <= OffsetFromLast, or if OffsetFromLast is negative.")
(Input:[]t where t:type).Last<public>(OffsetFromLast : int)<computes><decides>:t=

<#>
    For future implementation:
    Replace<public>(Input:[]t, StartIndex:int, StopIndex:int, ElementsToReplaceWith:[]t where t:type)<computes><decides>:[]t
