# Copyright Epic Games, Inc. All Rights Reserved.

using { /Verse.org/Native }

@doc("Implemented by classes that allow users to cancel an operation. For example, calling {
     }`subscribable.Subscribe` with a callback returns a `cancelable` object. Calling `Cancel` {
     }on the return object unsubscribes the callback.")
cancelable<public><native> := interface:
    @doc("Prevents any current or future work from completing.")
    Cancel<public><native_callable>()<transacts>:void
