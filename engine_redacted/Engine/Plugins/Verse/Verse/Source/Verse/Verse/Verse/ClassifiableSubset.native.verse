# Copyright Epic Games, Inc. All Rights Reserved.

using { /Verse.org/Native }

##################################################
# - Note that element type does NOT need to be comparable, although in practice it needs to be castable since all extension methods operate on castable subtypes. 
##################################################
@doc("A `classifiable_subset` is a container that holds a set of elements. {
      }A classifiable_subset can hold multiple elements of the same type.")
@experimental
classifiable_subset<native><public>(element_type:type) := class<final><computes><internal>(redacted_internal):
    @doc("Temporary function for constraining variance correctly.")
    CovarianceConstraint<protected>():?element_type = false
    
    

##################################################
# Internal helper method
##################################################

# TEMP: util code for casting to "castable_subtype(t) where t:type" since that does not compile in verse today, <redacted-url>
@experimental
@available{MinUploadedAtFNVersion := 3800}
(InInstance:t where t:castable_subtype(any)).IsOfType<native><internal>(query_type:castable_subtype(any))<reads><decides>:void

##################################################
# Public methods
##################################################

@doc("Constructs a `classifiable_subset` containing the `InElements`.")
@experimental
MakeClassifiableSubset<native><public>(InElements:[]t where t:type)<converges>:classifiable_subset(t)

@doc("Succeeds if `element_type` is not present in `InSet`.")
@experimental
@available{MinUploadedAtFNVersion := 3800}
(InSet:classifiable_subset(t) where t:castable_subtype(k), k:type).NotContains<internal>(element_type:castable_subtype(k))<decides><reads>:void=
    do: 
        not Element.IsOfType[element_type]

@doc("Succeeds if `element_type` is present in `InSet`.")
@experimental
@available{MinUploadedAtFNVersion := 3800}
(InSet:classifiable_subset(t) where t:castable_subtype(k), k:type).Contains<public>(element_type:castable_subtype(k))<decides><reads>:void=
    not InSet.NotContains[element_type]

@doc("Succeeds if all the `element_types` are present in `InSet`.")
@experimental
@available{MinUploadedAtFNVersion := 3800}
(InSet:classifiable_subset(t) where t:castable_subtype(k), k:type).ContainsAll<public>(element_types:[]castable_subtype(k))<decides><reads>:void=
    for (element_type : element_types):
        InSet.Contains[element_type]

@doc("Succeeds if none of the `element_types` are present in `InSet`.")
@experimental
@available{MinUploadedAtFNVersion := 3800}
(InSet:classifiable_subset(t) where t:castable_subtype(k), k:type).ContainsNone<internal>(element_types:[]castable_subtype(k))<decides><reads>:void=
    for:
        element_type : element_types
    do:
        not Element.IsOfType[element_type]

@doc("Succeeds if any of the `element_types` are present in `InSet`.")
@experimental
@available{MinUploadedAtFNVersion := 3800}
(InSet:classifiable_subset(t) where t:castable_subtype(k), k:type).ContainsAny<public>(element_types:[]castable_subtype(k))<decides><reads>:void=
    not InSet.ContainsNone[element_types]

@doc("Returns a new set that is the union of all elements in `InSetL` set and `InSetR`.")
@experimental
operator'+'<native><public>(InSetL:classifiable_subset(t), InSetR:classifiable_subset(t) where t:type)<transacts>:classifiable_subset(t)

@doc("Returns a new set that contains all the elements in `InSet` that are of type `element_type`.")
@experimental
@available{MinUploadedAtFNVersion := 3800}
(InSet:classifiable_subset(t) where t:castable_subtype(k), k:type).FilterByType<native><public>(element_type:castable_subtype(k))<transacts>:classifiable_subset(t)
