# Copyright Epic Games, Inc. All Rights Reserved.

using { /Verse.org/Native }

@doc("An opaque diagnostic message that only shows up in diagnostic logs. The format of the diagnostic may change at any time without warning and may not be inspected by Verse code.")
diagnostic<public><native> := class<computes>(redacted_internal):

@doc("Concatenates two diagnostic messages.")
operator'+'<public>(Lhs:diagnostic, Rhs:diagnostic)<computes>:diagnostic=

@doc("Concatenates a diagnostic message with a normal string, yielding a diagnostic message.")
operator'+'<public>(Lhs:diagnostic, Rhs:string)<computes>:diagnostic=

@doc("Concatenates a normal string with a diagnostic message, yielding a diagnostic message.")
operator'+'<public>(Lhs:string, Rhs:diagnostic)<computes>:diagnostic=

@doc("Converts any Verse value into an opaque diagnostic message.")
ToDiagnostic<public><native>(Value:any)<predicts><reads>:diagnostic
