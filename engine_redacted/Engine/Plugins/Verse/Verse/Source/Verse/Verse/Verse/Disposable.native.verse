# Copyright Epic Games, Inc. All Rights Reserved.

using { /Verse.org/Native }

@doc("Implemented by classes whose instances have limited lifetimes.")
disposable<native><public> := interface:

    @doc("Cleans up this object.")
    Dispose<public>():void
