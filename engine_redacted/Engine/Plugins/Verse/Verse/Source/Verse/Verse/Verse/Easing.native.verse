# Copyright Epic Games, Inc. All Rights Reserved.

using { /Verse.org/Native } # For the doc attribute

@available{MinUploadedAtFNVersion := 4100}
Easing<public> := module:
    # Internal CubicBezier function used for the various easing variants
    CubicBezierInterpInternal<internal><native>(Value:float, X1:float, Y1:float, X2:float, Y2:float)<reads>:float
    
    # Additional manual capture class as BPVM does not support local functions
    cubic_bezier_capture_internal<internal><native> := class<internal>:
        X1<internal><native>:float
        Y1<internal><native>:float
        X2<internal><native>:float
        Y2<internal><native>:float
        Evaluate<internal><native_callable>(Value:float)<reads>:float =
    
    @doc("Make a cubic bezier interpolation function.")
    @experimental
    @available{MinUploadedAtFNVersion := 4100}
    CubicBezier<public><native>(X1:float, Y1:float, X2:float, Y2:float)<converges>:type{_(:float)<reads>:float}
    
    @doc("Linear interpolation function. `Linear` animations move at a constant speed.")
    @experimental
    @available{MinUploadedAtFNVersion := 4100}
    Linear<public><native_callable>(Value:float)<reads>:float =
    
    @doc("Ease interpolation function. `Ease` animations start slowly, speed up, then end slowly. The speed of the animation is slightly slower at the end than the start.")
    @experimental
    @available{MinUploadedAtFNVersion := 4100}
    Ease<public><native_callable>(Value:float)<reads>:float =
    
    @doc("Ease-in interpolation function. `EaseIn` animations start slow, then speed up towards the end.")
    @experimental
    @available{MinUploadedAtFNVersion := 4100}
    EaseIn<public><native_callable>(Value:float)<reads>:float =
    
    @doc("Ease-out interpolation function. `EaseOut` animations start fast, then slow down towards the end.")
    @experimental
    @available{MinUploadedAtFNVersion := 4100}
    EaseOut<public><native_callable>(Value:float)<reads>:float =
    
    @doc("Ease-in-out interpolation function. `EaseInOut` animations are similar to `Ease` but the start and end animation speed is symmetric.")
    @experimental
    @available{MinUploadedAtFNVersion := 4100}
    EaseInOut<public><native_callable>(Value:float)<reads>:float =
