# Copyright Epic Games, Inc. All Rights Reserved.

using { /Verse.org/Native }

@doc("Implemented by classes whose instances can be enabled and disabled.")
enableable<native><public> := interface<public>:

    @doc("Enable this object.")
    Enable<public>():void

    @doc("Disable this object.")
    Disable<public>():void

    @doc("Succeeds if the object is enabled, fails if it’s disabled.")
    IsEnabled<public>()<decides><transacts>:void
