# Copyright Epic Games, Inc. All Rights Reserved.

using { /Verse.org/Native }

@doc("Halts the Verse runtime with error `Message`.")
Err<public><native>(Message:string)<computes><predicts>:false
