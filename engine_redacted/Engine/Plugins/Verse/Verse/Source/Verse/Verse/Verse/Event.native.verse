# Copyright Epic Games, Inc. All Rights Reserved.

using { /Verse.org/Concurrency }
using { /Verse.org/Native }

@doc("A *recurring*, successively signaled parametric `event` with a `payload` allowing a {
      }simple mechanism to coordinate between concurrent tasks.")
event<native><public>(t:type) := class(redacted_internal, signalable(t), awaitable(t)):

    @doc("Suspends the current task until another task calls `Signal`.\n{
         }If called during another invocation of `Signal`, the the task will still suspend {
         }and resume during the next call to `Signal`.")
    Await<native><override>()<suspends>:t

    @doc("Concurrently resumes the tasks that were suspended by `Await` calls before this {
         }call to `Signal`.\n{
         }\n{
         }Tasks are resumed in the order they were suspended. Each task will perform as {
         }much work as it can until it encounters a blocking call, whereupon it will {
         }transfer control to the next suspended task.")
    Signal<native><override>(Val:t):void

@doc("A *recurring*, successively signaled event allowing a simple mechanism to coordinate {
     }between concurrent tasks.")
event<public>() := event(tuple())



event_subscription<native><internal> := class(cancelable):
     Cancel<override><native>()<transacts>:void
     
