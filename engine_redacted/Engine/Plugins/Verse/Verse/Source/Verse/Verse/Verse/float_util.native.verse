# Copyright Epic Games, Inc. All Rights Reserved.

using { /Verse.org/Native } # For the doc attribute

@doc("Returns the smallest `int` that is greater than or equal to `Val`.\n{
     }Fails if `not IsFinite(Val)`.")
Ceil<native><public>(Val:float)<reads><decides>:int

@doc("Returns the largest `int` that is less than or equal to `Val`.\n{
     }Fails if `not IsFinite(Val)`.")
Floor<native><public>(Val:float)<reads><decides>:int

@doc("Returns `Val` rounded to the nearest `int`. When the fractional part of `Val` is `0.5`, {
     }rounds to the nearest *even* `int` (per the IEEE-754 default rounding mode).\n{
     }Fails if `not IsFinite(Val)`.")
Round<native><public>(Val:float)<reads><decides>:int

@doc("Returns the `int` that equals `Val` without the fractional part.\n{
     }Fails if `not IsFinite(val)`.")
Int<native><public>(Val:float)<reads><decides>:int

# TODO: Add format specifiers
@doc("Makes a `string` representation of `Val`.")
ToString<native><public>(Val:float)<predicts><reads>:string
