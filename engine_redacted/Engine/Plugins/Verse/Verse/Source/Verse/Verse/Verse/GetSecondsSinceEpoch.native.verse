# Copyright Epic Games, Inc. All Rights Reserved.

using { /Verse.org/Native }

@doc("Returns the number of seconds since January 1, 1970 UTC, ignoring leap seconds. I.e, this function implements Unix time. This function always returns the same value within the same transaction.")
GetSecondsSinceEpoch<public><native>()<reads>:float
