# Copyright Epic Games, Inc. All Rights Reserved.

using { /Verse.org/Native } # For the doc attribute

# TODO: Add format specifiers for things like hex
@doc("Makes a printable `string` representation of `Val`.")
ToString<native><public>(Val:int)<computes><predicts>:string
