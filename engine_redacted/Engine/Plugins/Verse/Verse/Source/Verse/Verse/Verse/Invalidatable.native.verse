# Copyright Epic Games, Inc. All Rights Reserved.

using { /Verse.org/Native }

@doc("Implemented by classes whose instances can become invalid at runtime.")
invalidatable<native><public> := interface(disposable):

    @doc("Succeeds if this object is still valid.")
    IsValid<public>()<transacts><decides>:void
