# Copyright Epic Games, Inc. All Rights Reserved.

using { /Verse.org/Concurrency }
using { /Verse.org/Native }

@doc("A parametric interface combining `awaitable` and `subscribable`.")
listenable<public><native>(payload:type) := interface(awaitable(payload), subscribable(payload)) {}

@doc("A parameterless interface combining `awaitable` and `subscribable`.")
listenable<public>() := listenable(tuple())
