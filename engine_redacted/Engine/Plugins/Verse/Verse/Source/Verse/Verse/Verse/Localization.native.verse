# Copyright Epic Games, Inc. All Rights Reserved.

using { /Verse.org/Native }

@doc("Used for message localization.")
locale<public><native> := struct {}

@doc("A localizable text message.")
message<public><native> := class:
    DefaultText<native> : string = ""

@doc("Makes a `string` by localizing `Message` based on the current `locale`.")
Localize<public><native>(Message:message)<reads>: string


@available{MinUploadedAtFNVersion := 4100} # Guarding against name collision in already published projects.
@doc("Makes a `message` by concatenating `Separator` between the elements of `Messages`.")
Join<native><public>(Messages:[]message, Separator:message)<transacts>:message

# localizable_value is a temporary implementation detail that is required to
# hold the captured values. eventually this will get removed and replaced
# by capturing lambdas
localizable_value<native> := class:

localizable_string<native> := class<internal>(localizable_value):

MakeLocalizableValue(V:string):localizable_string = localizable_string{Value:=V}

localizable_int<native> := class<internal>(localizable_value):

MakeLocalizableValue(V:int):localizable_int = localizable_int{Value:=V}

localizable_float<native> := class<internal>(localizable_value):

MakeLocalizableValue(V:float):localizable_float = localizable_float{Value:=V}

localizable_message<native> := class<internal>(localizable_value):

MakeLocalizableValue(V:message):localizable_message = localizable_message{Value:=V}

# internals

MakeMessageInternal<native>(K:string, D:string, S:[string]localizable_value)<converges>: message
