# Copyright Epic Games, Inc. All Rights Reserved.

using { /Verse.org/Native } # For the doc attribute

# Math utility functions. Names for functions included in ISO 80000-2:2019 follow the conventions specified by that standard.
@doc("Pi, the ratio of the circumference of a circle to its diameter.")
PiFloat<public>:float := 3.141592653589793238462643383279502884

# TODO: include Euler's constant (i.e. `EFloat`) since it would be helpful to reference in `Exp` and `Ln`.
#       This constant appears as `ValueOfE` in "Math.versetest".

@doc("Constrains the value of `Val` between `A` and `B`. Robustly handles different argument orderings.\n{
     }Returns the median of `Val`, `A`, and `B`, such that comparisons with `NaN` operate as if `NaN > +Inf`.")
Clamp<public>(Val:float, A:float, B:float)<computes>:float :=
	Lo := Min(A, B)
	Hi := Max(A, B)
	return Min(Max(Val, Lo), Hi)

@rtfm_always_open
@doc("Constrains the value of `Val` between `A` and `B`. Robustly handles different argument orderings.\n{
     }Returns the median of `Val`, `A`, and `B`.")
Clamp<native><public>(Val:int, A:int, B:int)<computes>:int

@doc("Returns the minimum of `X` and `Y`.")
Min<public>(X:int, Y:int)<computes>:int = (X < Y) or Y

@doc("Returns the maximum of `X` and `Y`.")
Max<public>(X:int, Y:int)<computes>:int = (X > Y) or Y

@doc("Returns the minimum of `X` and `Y` unless either are `NaN`.\n{
     }Returns `NaN` if either `X` or `Y` are `NaN`.")
Min<public>(X:float, Y:float)<computes>:float := (X <= Y) or (X = NaN) or Y

@doc("Returns the maximum of `X` and `Y` unless either are `NaN`.\n{
     }Returns `NaN` if either `X` or `Y` are `NaN`.")
Max<public>(X:float, Y:float)<computes>:float := (X > Y) or (X = NaN) or Y

@vm_no_effect_token
@rtfm_always_open
@doc("Returns the square root of `X` if `X >= 0.0`.\n{
     }Returns `NaN` if `X < 0.0`.")
Sqrt<native><public>(X:float)<predicts><reads>:float

@vm_no_effect_token
@rtfm_always_open
@doc("Returns the sine of `X`, where `X` is interpreted as a value in radians, if `IsFinite[X]`.\n{
     }Returns `NaN` if `not IsFinite[X]`.")
Sin<native><public>(X:float)<reads>:float

@vm_no_effect_token
@rtfm_always_open
@doc("Returns the cosine of `X`, where `X` is interpreted as a value in radians, if `IsFinite[X]`.\n{
     }Returns `NaN` if `not IsFinite[X]`.")
Cos<native><public>(X:float)<reads>:float

@vm_no_effect_token
@rtfm_always_open
@doc("Returns the tangent of `X`, where `X` is interpreted as a value in radians, if `IsFinite[X]`.\n{
     }Returns `NaN` if `not IsFinite[X]`.")
Tan<native><public>(X:float)<reads>:float

# JIRA: <redacted-ticket>
# TODO: ArcSin and ArcCos call into FMath, whose implementations clamp the
# domain to [-1, 1] instead of generating/propagating NaNs. For now, will
# leave the behavior unspecified.

@vm_no_effect_token
@rtfm_always_open
@doc("Returns the inverse sine (arcsine) of `X` as a value in radians if `-1.0 <= X <= 1.0`.")
ArcSin<native><public>(X:float)<reads>:float

@vm_no_effect_token
@rtfm_always_open
@doc("Returns the inverse cosine (arccosine) of `X` as a value in radians if `-1.0 <= X <= 1.0`.")
ArcCos<native><public>(X:float)<reads>:float

@vm_no_effect_token
@rtfm_always_open
@doc("Returns the inverse tangent (arctangent) of `X` as a value in radians such that:{
     }`-PiFloat/2.0 <= ArcTan(x) <= PiFloat/2.0`.")
ArcTan<native><public>(X:float)<reads>:float

@vm_no_effect_token
@rtfm_always_open
@doc("Returns the angle in radians at the origin between a ray pointing to `(X, Y)` {
     }and the positive `X` axis such that `-PiFloat < ArcTan(Y, X) <= PiFloat`.\n{
     }Returns `0.0` if `X=0.0 and Y=0.0`.")
ArcTan<native><public>(Y:float, X:float)<reads>:float

@vm_no_effect_token
@rtfm_always_open
@doc("Returns the hyperbolic sine of `X`.")
Sinh<native><public>(X:float)<reads>:float

@vm_no_effect_token
@rtfm_always_open
@doc("Returns the hyperbolic cosine of `X`.")
Cosh<native><public>(X:float)<reads>:float

@vm_no_effect_token
@rtfm_always_open
@doc("Returns the hyperbolic tangent of `X`.")
Tanh<native><public>(X:float)<reads>:float

# JIRA: <redacted-ticket>
# TODO: ArSinh returns NaN at infinity, which does not seem correct. Are any of these used?
@vm_no_effect_token
@rtfm_always_open
@doc("Returns the inverse hyperbolic sine of `X` if `IsFinite(X)`.")
ArSinh<native><public>(X:float)<reads>:float

@vm_no_effect_token
@rtfm_always_open
@doc("Returns the inverse hyperbolic cosine of `X` if `1.0 <= X`.")
ArCosh<native><public>(X:float)<reads>:float

@vm_no_effect_token
@rtfm_always_open
@doc("Returns the inverse hyperbolic tangent of `X` if `IsFinite(X)`.")
ArTanh<native><public>(X:float)<reads>:float

@vm_no_effect_token
@rtfm_always_open
@doc("Returns `A` to the power of `B`.")
Pow<native><public>(A:float, B:float)<reads>:float

@doc("Returns `X` if `X` is finite.\n{
     }Fails if `X` is one of:`\n{
     } * `+Inf`\n{
     } * `-Inf`\n{
     } * `NaN`")
(X:float).IsFinite<public>()<computes><decides>:float=
    return (X > -Inf) and (X < Inf)

@rtfm_always_open
@doc("Returns the quotient `X/Y` as defined by Euclidean division, i.e.:\n{
     } * `Quotient[X/Y] = Floor[X/Y]` when `Y > 0`\n{
     } * `Quotient[X/Y] = Ceil[X/Y]` when `Y < 0`\n{
     } * `Quotient[X/Y] * Y + Mod[X,Y] = X`\n{
     }Fails if `Y = 0`.")
Quotient<native><public>(X:int, Y:int)<computes><decides>:int

@rtfm_always_open
@doc("Returns the remainder of `X/Y` as defined by Euclidean division, i.e.:\n{
     } * `Mod[X,Y] = X - Quotient(X/Y)*Y`\n{
     } * `0 <= Mod[X,Y] < Abs(Y)`\n{
     }Fails if `Y=0`.")
Mod<native><public>(X:int, Y:int)<computes><decides>:int

@vm_no_effect_token
@rtfm_always_open
@doc("Returns the natural exponent of `X`.")
Exp<native><public>(X:float)<reads>:float

@vm_no_effect_token
@rtfm_always_open
@doc("Returns the natural logarithm of `X`.")
Ln<native><public>(X:float)<reads>:float

@doc("Returns the base `B` logarithm of `X`.")
Log<public>(B:float, X:float)<reads>:float :=
    return Ln(X) / Ln(B)

@vm_no_effect_token
@rtfm_always_open
@doc("Used to linearly interpolate/extrapolate between `From` (when `Parameter = 0.0`) {
     }and `To` (when `Parameter = 1.0`). Expects that all arguments are finite.\n{
     }Returns `From*(1 - Parameter) + To*Parameter`.")
Lerp<native><public>(From:float, To:float, Parameter:float)<predicts><reads>:float

@doc("Returns the sign of `Val`:\n {
     } * `1` if `Val > 0`\n{
     } * `0` if `Val = 0`\n{
     } * `-1` if `Val < 0`")
Sgn<public>(Val:int)<computes>:int:=
    if (Val > 0):
        return 1
    else if (Val < 0):
        return -1
    else:
        return 0

@doc("Returns the sign of `Val`:\n{
     } * `1.0` if `Val > 0.0`\n{
     } * `0.0` if `Val = 0.0`\n{
     } * `-1.0` if `Val < 0.0`\n{
     } * `NaN` if `Val = NaN`")
Sgn<public>(Val:float)<computes>:float:=
    if (Val = NaN):
        return NaN
    else if (Val > 0.0):
        return 1.0
    else if (Val < 0.0):
        return -1.0
    else:
        return 0.0

# JIRA: <redacted-ticket>
#TODO: consider what happens with Inf and NaN.

@doc("Succeeds if `Val` is within `AbsoluteTolerance` of `0.0`.")
(Val:float).IsAlmostZero<public>(AbsoluteTolerance:float)<decides><computes>:void=
    Abs(Val) <= AbsoluteTolerance

@doc("Succeeds if `Val1` and `Val2` are within `AbsoluteTolerance` of each other.")
IsAlmostEqual<public>(Val1:float, Val2:float, AbsoluteTolerance:float)<decides><computes>:void=
    Abs(Val1 - Val2) <= AbsoluteTolerance
