# Copyright Epic Games, Inc. All Rights Reserved.

using { /Verse.org/Native }

@doc("Implemented by classes that provide a result for an operation, which can fail or be successful")
@available{MinUploadedAtFNVersion := 3800} # Guarding against name collision in already published projects.
result<public><native>(success_type:type, error_type:type) := interface<internal><computes>:
    @doc("Returns the success data of the specified type.")
    GetSuccess<public><native_callable>()<computes><decides>:success_type
    @doc("Returns the error data of the specified type.")
    GetError<public><native_callable>()<computes><decides>:error_type

@doc("Result class that holds the success data.")
success_result<native><internal>(success_type:type) := class<computes>(result(success_type, false)):

    GetSuccess<override>()<computes>:success_type =
    GetError<override>()<computes><decides>:false =
        false?
        Err("unreachable")

@doc("Result class that holds the failure data.")
error_result<native><internal>(error_type:type) := class<computes>(result(false, error_type)):
    Error<native>:error_type

    GetSuccess<override>()<computes><decides>:false =
        false?
        Err("unreachable")
    GetError<override>()<computes>:error_type =
        Error

@available{MinUploadedAtFNVersion := 3800} # Guarding against name collision in already published projects.
MakeSuccess<public>(Result:success_type where success_type:type)<computes>:result(success_type, false) =
    success_result(success_type){ Success := Result }

@available{MinUploadedAtFNVersion := 3800} # Guarding against name collision in already published projects.
MakeError<public>(Result:error_type where error_type:type)<computes>:result(false, error_type) =
    error_result(error_type){ Error := Result }