# Copyright Epic Games, Inc. All Rights Reserved.

using { /Verse.org/Native }

@doc("Implemented by classes whose instances can change visibility to be shown or hidden.")
showable<native><public> := interface:
    
    @doc("Set this value to hide or show the class.")
    var Show<public> : logic