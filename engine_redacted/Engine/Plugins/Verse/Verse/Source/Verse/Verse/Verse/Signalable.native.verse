# Copyright Epic Games, Inc. All Rights Reserved.

using { /Verse.org/Native }

@doc("A parametric interface implemented by events with a `payload` that can be signaled.\n{
     }Can be used with `awaitable`, `subscribable`, or both (see: `listenable`).")
signalable<public><native>(payload:type) := interface:

    @doc("Concurrently resumes the tasks waiting for this event in `awaitable.Await` and {
         }synchronously invokes any callbacks added to this event by `subscribable.Subscribe`.")
    Signal<public><native_callable>(Val:payload):void
