# Copyright Epic Games, Inc. All Rights Reserved.

using{/Verse.org/Native}

@doc("Makes a `string` by concatenating `Separator` between the elements of `Strings`.")
Join<native><public>(Strings:[]string, Separator:string)<computes><predicts>:string

@doc("Returns `String` without modification.")
ToString<public>(String:string)<computes><predicts>:string = RedactedInternal

@doc("Makes a `string` from `Character`.")
ToString<native><public>(Character:char)<computes><predicts>:string

@doc("Makes a `string` by converting `Character` to UTF-8 from UTF-32.")
ToString<native>(Character:char32)<computes><predicts>:string
