# Copyright Epic Games, Inc. All Rights Reserved.

using { /Verse.org/Native }

@doc("A parametric interface implemented by events with a `payload` that can be subscribed to.\n{
     }Matched with `signalable.`")
subscribable<public><native>(t:type) := interface:
    @doc("Registers `Callback` to be invoked on matching calls to `signable.Signal`.\n{
         }Returns an unsubscriber object. Call `cancelable.Cancel` on the unsubscriber to unregister `Callback`.")
    Subscribe<public>(Callback(:t):void)<transacts>:cancelable

@doc("A parameterless interface implemented by events that can be subscribed to.")
subscribable<public>() := subscribable(tuple())
