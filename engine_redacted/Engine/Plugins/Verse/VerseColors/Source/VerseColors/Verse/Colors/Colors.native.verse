# Copyright Epic Games, Inc. All Rights Reserved.

using { /Verse.org/Native }
using { /Verse.org/Simulation }
using { /Verse.org/Temporary/EpicGamesRestricted/Network }

@doc("Represents colors as RGB triples in the ACES 2065-1 color space.\n{
     }Component values are linear (i.e. `*gamma* = 1.0`).")
@replicatable
@import_as("/*.FVColor")
color<public><native> := struct<concrete><computes><persistable><uht_comparable>:

    @doc("Red component of this `color`.")
    @editable
    R<public><native>:float = 0.0

    @doc("Green component of this `color`.")
    @editable
    G<public><native>:float = 0.0

    @doc("Blue component of this `color`.")
    @editable
    B<public><native>:float = 0.0

# TODO: When ShipVerse supports this, allow specifying a named color wherever an ACES color is required.
#color<public>(NamedColors:named_color)<constructor>:=color {..}

@doc("Makes an ACES 2065-1 `color` from the component-wise sum of `c0` and `c1`.")
operator'+'<public><native>(c0:color, c1:color)<converges>:color

@doc("Makes an ACES 2065-1 `color` from the component-wise difference of `c0` and `c1`.")
operator'-'<public><native>(c0:color, c1:color)<converges>:color

@doc("Makes an ACES 2065-1 `color` from the component-wise product of `c0` and `c1`.")
operator'*'<public><native>(c0:color, c1:color)<converges>:color

@doc("Makes an ACES 2065-1 `color` from each component of `c` scaled by `factor`.")
operator'*'<public><native>(c:color, factor:float)<converges>:color

@doc("Makes an ACES 2065-1 `color` from each component of `c` scaled by `factor`.")
operator'*'<public><native>(c:color, factor:int)<converges>:color

@doc("Makes an ACES 2065-1 `color` from each component of `c` scaled by `factor`.")
operator'*'<public><native>(factor:float, c:color)<converges>:color

@doc("Makes an ACES 2065-1 `color` from each component of `c` scaled by `factor`.")
operator'*'<public><native>(factor:int, c:color)<converges>:color
        
@doc("Makes an ACES 2065-1 `color` from each component of `c` divided by `factor`.")
operator'/'<public><native>(c:color, factor:float)<converges><decides>:color

@doc("Makes an ACES 2065-1 `color` from each component of `c` divided by `factor`.")
operator'/'<public><native>(c:color, factor:int)<converges><decides>:color

#TODO: explain what happens with negative sRGB component values.
@doc("Makes an ACES 2065-1 `color` from sRGB components `Red`, `Green`, and `Blue`.\n{
     }Normal sRGB component values are between `0.0` and `1.0`, but this can handle {
     }larger values.")
MakeColorFromSRGB<public><native>(Red:float, Green:float, Blue:float)<converges>:color

@doc("Makes an sRGB `tuple` by converting `InColor` from an ACES 2065-1 `color` to sRGB.")
MakeSRGBFromColor<public><native>(InColor:color)<converges>:tuple(float, float, float)

@doc("Makes an ACES 2065-1 `color` from the integer sRGB components `Red`, `Green`, and `Blue`.\n{
     }Valid sRGB component values are between '0' and '255', inclusive.")
MakeColorFromSRGBValues<public><native>(Red:int, Green:int, Blue:int)<converges>:color

# TODO: Ideally the string formats should be quoted, but that hurts readability due to escaping.
# When we switch from doc strings to markup, we can format these more appropriately.
@doc("Makes an ACES 2065-1 `color` from a CSS-style sRGB `hexString`. Supported formats are:\n{
     } * RGB\n{
     } * RRGGBB\n{
     } * RRGGBBAA\n{
     } * #RGB\n{
     } * #RRGGBB\n{
     } * #RRGGBBAA\n{
     }An invalid hex string will return `Black`.")
MakeColorFromHex<public><native>(hexString:string)<converges>:color

@doc("Makes an ACES 2065-1 `color` from `Hue`, `Saturation`, and `Value` components.\n{
     }Components use the HSV color model in the sRGB color space. Expected ranges:\n{
     } * `0.0 <= Hue <= 360.0`\n{
     } * `0.0 <= Saturation <= 1.0`\n{
     } * `0.0 <= Value <= 1.0`\n{
     }Values out of expected ranges will undergo range reduction and conversion.")
MakeColorFromHSV<public><native>(Hue:float, Saturation:float, Value:float)<converges>:color

@doc("Makes an HSV `tuple` by converting `InColor` from an ACES 2065-1 `color` to sRGB and applying the HSV color model.")
MakeHSVFromColor<public><native>(InColor:color):tuple(float, float, float)

@doc("Makes an ACES 2065-1 `color` from the chromaticity of a blackbody radiator at `Temperature` Kelvin.\n{
    }`Temperature` is clamped such that `0 <= Temperature`.")
MakeColorFromTemperature<public><native>(Temperature:float)<converges>:color

@doc("Color presets from CSS Color Module 3 Extended color keywords.")
NamedColors<public> := module:
    AliceBlue<public> : color = MakeColorFromHex("F0F8FF")
    AntiqueWhite<public> : color = MakeColorFromHex("FAEBD7")
    Aqua<public> : color = MakeColorFromHex("00FFFF")
    Aquamarine<public> : color = MakeColorFromHex("7FFFD4")
    Azure<public> : color = MakeColorFromHex("F0FFFF")
    Beige<public> : color = MakeColorFromHex("F5F5DC")
    Bisque<public> : color = MakeColorFromHex("FFE4C4")
    Black<public> : color = MakeColorFromHex("000000")
    BlanchedAlmond<public> : color = MakeColorFromHex("FFEBCD")
    Blue<public> : color = MakeColorFromHex("0000FF")
    BlueViolet<public> : color = MakeColorFromHex("8A2BE2")
    Brown<public> : color = MakeColorFromHex("A52A2A")
    Burlywood<public> : color = MakeColorFromHex("DEB887")
    CadetBlue<public> : color = MakeColorFromHex("5F9EA0")
    Chartreuse<public> : color = MakeColorFromHex("7FFF00")
    Chocolate<public> : color = MakeColorFromHex("D2691E")
    Coral<public> : color = MakeColorFromHex("FF7F50")
    CornflowerBlue<public> : color = MakeColorFromHex("6495ED")
    Cornsilk<public> : color = MakeColorFromHex("FFF8DC")
    Crimson<public> : color = MakeColorFromHex("DC143C")
    Cyan<public> : color = MakeColorFromHex("00FFFF")
    DarkBlue<public> : color = MakeColorFromHex("00008B")
    DarkCyan<public> : color = MakeColorFromHex("008B8B")
    DarkGoldenrod<public> : color = MakeColorFromHex("B8860B")
    DarkGray<public> : color = MakeColorFromHex("A9A9A9")
    DarkGreen<public> : color = MakeColorFromHex("006400")
    DarkGrey<public> : color = MakeColorFromHex("A9A9A9")
    DarkKhaki<public> : color = MakeColorFromHex("BDB76B")
    DarkMagenta<public> : color = MakeColorFromHex("8B008B")
    DarkOliveGreen<public> : color = MakeColorFromHex("556B2F")
    DarkOrange<public> : color = MakeColorFromHex("FF8C00")
    DarkOrchid<public> : color = MakeColorFromHex("9932CC")
    DarkRed<public> : color = MakeColorFromHex("8B0000")
    DarkSalmon<public> : color = MakeColorFromHex("E9967A")
    DarkSeaGreen<public> : color = MakeColorFromHex("8FBC8F")
    DarkSlateBlue<public> : color = MakeColorFromHex("483D8B")
    DarkSlateGray<public> : color = MakeColorFromHex("2F4F4F")
    DarkSlateGrey<public> : color = MakeColorFromHex("2F4F4F")
    DarkTurquoise<public> : color = MakeColorFromHex("00CED1")
    DarkViolet<public> : color = MakeColorFromHex("9400D3")
    DeepPink<public> : color = MakeColorFromHex("FF1493")
    DeepSkyBlue<public> : color = MakeColorFromHex("00BFFF")
    DimGray<public> : color = MakeColorFromHex("696969")
    DimGrey<public> : color = MakeColorFromHex("696969")
    DodgerBlue<public> : color = MakeColorFromHex("1E90FF")
    Firebrick<public> : color = MakeColorFromHex("B22222")
    FloralWhite<public> : color = MakeColorFromHex("FFFAF0")
    ForestGreen<public> : color = MakeColorFromHex("228B22")
    Fuchsia<public> : color = MakeColorFromHex("FF00FF")
    Gainsboro<public> : color = MakeColorFromHex("DCDCDC")
    GhostWhite<public> : color = MakeColorFromHex("F8F8FF")
    Gold<public> : color = MakeColorFromHex("FFD700")
    Goldenrod<public> : color = MakeColorFromHex("DAA520")
    Gray<public> : color = MakeColorFromHex("808080")
    Green<public> : color = MakeColorFromHex("008000")
    GreenYellow<public> : color = MakeColorFromHex("ADFF2F")
    Grey<public> : color = MakeColorFromHex("808080")
    Honeydew<public> : color = MakeColorFromHex("F0FFF0")
    Hotpink<public> : color = MakeColorFromHex("FF69B4")
    IndianRed<public> : color = MakeColorFromHex("CD5C5C")
    Indigo<public> : color = MakeColorFromHex("4B0082")
    Ivory<public> : color = MakeColorFromHex("FFFFF0")
    Khaki<public> : color = MakeColorFromHex("F0E68C")
    Lavender<public> : color = MakeColorFromHex("E6E6FA")
    LavenderBlush<public> : color = MakeColorFromHex("FFF0F5")
    LawnGreen<public> : color = MakeColorFromHex("7CFC00")
    LemonChiffon<public> : color = MakeColorFromHex("FFFACD")
    LightBlue<public> : color = MakeColorFromHex("ADD8E6")
    LightCoral<public> : color = MakeColorFromHex("F08080")
    LightCyan<public> : color = MakeColorFromHex("E0FFFF")
    LightGoldenrodYellow<public> : color = MakeColorFromHex("FAFAD2")
    LightGray<public> : color = MakeColorFromHex("D3D3D3")
    LightGreen<public> : color = MakeColorFromHex("90EE90")
    LightGrey<public> : color = MakeColorFromHex("D3D3D3")
    LightPink<public> : color = MakeColorFromHex("FFB6C1")
    LightSalmon<public> : color = MakeColorFromHex("FFA07A")
    LightSeaGreen<public> : color = MakeColorFromHex("20B2AA")
    LightSkyBlue<public> : color = MakeColorFromHex("87CEFA")
    LightSlateGray<public> : color = MakeColorFromHex("778899")
    LightSlateGrey<public> : color = MakeColorFromHex("778899")
    LightSteelBlue<public> : color = MakeColorFromHex("B0C4DE")
    LightYellow<public> : color = MakeColorFromHex("FFFFE0")
    Lime<public> : color = MakeColorFromHex("00FF00")
    LimeGreen<public> : color = MakeColorFromHex("32CD32")
    Linen<public> : color = MakeColorFromHex("FAF0E6")
    Magenta<public> : color = MakeColorFromHex("FF00FF")
    Maroon<public> : color = MakeColorFromHex("800000")
    MediumAquamarine<public> : color = MakeColorFromHex("66CDAA")
    MediumBlue<public> : color = MakeColorFromHex("0000CD")
    MediumOrchid<public> : color = MakeColorFromHex("BA55D3")
    MediumPurple<public> : color = MakeColorFromHex("9370DB")
    MediumSeaGreen<public> : color = MakeColorFromHex("3CB371")
    MediumSlateBlue<public> : color = MakeColorFromHex("7B68EE")
    MediumSpringGreen<public> : color = MakeColorFromHex("00FA9A")
    MediumTurquoise<public> : color = MakeColorFromHex("48D1CC")
    MediumVioletRed<public> : color = MakeColorFromHex("C71585")
    MidnightBlue<public> : color = MakeColorFromHex("191970")
    MintCream<public> : color = MakeColorFromHex("F5FFFA")
    MistyRose<public> : color = MakeColorFromHex("FFE4E1")
    Moccasin<public> : color = MakeColorFromHex("FFE4B5")
    NavajoWhite<public> : color = MakeColorFromHex("FFDEAD")
    Navy<public> : color = MakeColorFromHex("000080")
    OldLace<public> : color = MakeColorFromHex("FDF5E6")
    Olive<public> : color = MakeColorFromHex("808000")
    OliveDrab<public> : color = MakeColorFromHex("6B8E23")
    Orange<public> : color = MakeColorFromHex("FFA500")
    OrangeRed<public> : color = MakeColorFromHex("FF4500")
    Orchid<public> : color = MakeColorFromHex("DA70D6")
    PaleGoldenrod<public> : color = MakeColorFromHex("EEE8AA")
    PaleGreen<public> : color = MakeColorFromHex("98FB98")
    PaleTurquoise<public> : color = MakeColorFromHex("AFEEEE")
    PaleVioletred<public> : color = MakeColorFromHex("DB7093")
    PapayaWhip<public> : color = MakeColorFromHex("FFEFD5")
    PeachPuff<public> : color = MakeColorFromHex("FFDAB9")
    Peru<public> : color = MakeColorFromHex("CD853F")
    Pink<public> : color = MakeColorFromHex("FFC0CB")
    Plum<public> : color = MakeColorFromHex("DDA0DD")
    PowderBlue<public> : color = MakeColorFromHex("B0E0E6")
    Purple<public> : color = MakeColorFromHex("800080")
    Red<public> : color = MakeColorFromHex("FF0000")
    RosyBrown<public> : color = MakeColorFromHex("BC8F8F")
    RoyalBlue<public> : color = MakeColorFromHex("4169E1")
    SaddleBrown<public> : color = MakeColorFromHex("8B4513")
    Salmon<public> : color = MakeColorFromHex("FA8072")
    SandyBrown<public> : color = MakeColorFromHex("F4A460")
    SeaGreen<public> : color = MakeColorFromHex("2E8B57")
    SeaShell<public> : color = MakeColorFromHex("FFF5EE")
    Sienna<public> : color = MakeColorFromHex("A0522D")
    Silver<public> : color = MakeColorFromHex("C0C0C0")
    SkyBlue<public> : color = MakeColorFromHex("87CEEB")
    SlateBlue<public> : color = MakeColorFromHex("6A5ACD")
    SlateGray<public> : color = MakeColorFromHex("708090")
    SlateGrey<public> : color = MakeColorFromHex("708090")
    Snow<public> : color = MakeColorFromHex("FFFAFA")
    SpringGreen<public> : color = MakeColorFromHex("00FF7F")
    SteelBlue<public> : color = MakeColorFromHex("4682B4")
    (NamedColors:)Tan<public> : color = MakeColorFromHex("D2B48C")
    Teal<public> : color = MakeColorFromHex("008080")
    Thistle<public> : color = MakeColorFromHex("D8BFD8")
    Tomato<public> : color = MakeColorFromHex("FF6347")
    Turquoise<public> : color = MakeColorFromHex("40E0D0")
    Violet<public> : color = MakeColorFromHex("EE82EE")
    Wheat<public> : color = MakeColorFromHex("F5DEB3")
    White<public> : color = MakeColorFromHex("FFFFFF")
    WhiteSmoke<public> : color = MakeColorFromHex("F5F5F5")
    Yellow<public> : color = MakeColorFromHex("FFFF00")
    YellowGreen<public> : color = MakeColorFromHex("9ACD32")
    
@available{MinUploadedAtFNVersion := 3800} # color_alpha API
@doc("Represents colors as RGB triples in the ACES 2065-1 color space with an additional alpha channel A.\n{
     }This reasons about the `Color` and alpha (`A`) as separate concepts instead of as a single concept.\n{
     }All values are stored strictly as unopinionated floats but, when interpreted as a color with alpha,\n{
     }ranges for `A` are `0.0` (transparent) to `1.0` (opaque). Color values are not premultiplied.\n{
     }Component values are linear (i.e. `*gamma* = 1.0`).")
@replicatable
@import_as("/*.FVColorAlpha")
color_alpha<native><public> := struct<concrete><computes><persistable><uht_comparable>:
    @doc("`Color` component of this `color_alpha`.")
    @editable
    Color<native><public>:color = color{}
    
    @doc("Alpha component of this `color_alpha`.")
    @editable
    A<native><public>:float = 1.0
        
@available{MinUploadedAtFNVersion := 3800} # color_alpha API
@doc("Makes a new `color_alpha` from individual `R`, `G`, `B`, `A` component values.")
MakeColorAlpha<public>(R:float, G:float, B:float, ?A:float=1.0)<computes>:color_alpha =
    color_alpha:
        Color := color:
            R := R
            G := G
            B := B
        A := A

@available{MinUploadedAtFNVersion := 3800} # color_alpha API
@doc("Blend colors `CA1` and `CA2` with `CA1` over `CA2` using two non-premultiplied `color_alpha` values.\n{
    }Alpha components are clamped between `0.0` and `1.0`. Fails if the value of both clamped Alpha components are `0.0`.")
Over<public>(CA1:color_alpha, CA2:color_alpha)<reads><decides>:color_alpha =
    ClampedCA1 := color_alpha{Color := CA1.Color, A := Clamp(CA1.A, 0.0, 1.0)}
    ClampedCA2 := color_alpha{Color := CA2.Color, A := Clamp(CA2.A, 0.0, 1.0)}
    
    if (ClampedCA1.A = 0.0 and ClampedCA2.A = 0.0):
        false?
    
