# Copyright Epic Games, Inc. All Rights Reserved.

using { /Verse.org/Assets }
using { /Verse.org/Native }
using { /Verse.org/Restricted }
using { /Verse.org/Simulation }
using { /Verse.org/Colors }
using { /Verse.org/Temporary/EpicGamesRestricted/Network }





animation_sequence<native><public> := class<computes><final>(redacted_internal, redacted_internal) {}

material<native><public> := class(redacted_internal, redacted_internal, redacted_internal):






particle_system<native><public> := class<computes><final>(redacted_internal) {}




mesh<native><public> := class<computes>(redacted_internal) {}

sound_wave<native><public> := class<computes><final>(redacted_internal) {}






texture<native><public> := class<computes><final>(redacted_internal, redacted_internal, redacted_internal):

input_action<native><public>(t:type) := class<allocates><unique><computes>(redacted_internal):

    OnInputActionConstructed<private><native>()<computes>:void
    block:
        OnInputActionConstructed()

input_mapping<native><public> := class<computes>(redacted_internal) {}

