# Copyright Epic Games, Inc. All Rights Reserved.

using { /Verse.org/Assets }
using { /Verse.org/Native }
using { /Verse.org/Simulation }

@doc("Interface that provides an icon.")
has_icon<public><native> := interface():
    @doc("A texture used as the 2D visual representation of this entity (e.g. an icon or portrait).")
    @editable
    var Icon<public>:texture

    GetIcon<private><native_callable>()<reads>:texture=Icon

    SetIcon<private><native_callable>(InIcon:texture)<transacts>:void=
        set Icon = InIcon
