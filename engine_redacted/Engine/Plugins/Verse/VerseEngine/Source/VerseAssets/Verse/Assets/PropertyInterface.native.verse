# Copyright Epic Games, Inc. All Rights Reserved.

using { /Verse.org/Assets }
using { /Verse.org/Native }
using { /Verse.org/Temporary/EpicGamesRestricted/Network }

@import_as("/Script/VerseAssets.UParameterizePropertyInterfaceBase")
parameterized_property_interface_base<native> := class{}
