# Copyright Epic Games, Inc. All Rights Reserved.

using { /Verse.org/Native }

@doc("Enumerates the days of the week in 7-day calendars.")
day_of_week<public><native> := enum:
    Monday
    Tuesday
    Wednesday
    Thursday
    Friday
    Saturday
    Sunday

@doc("Enumerates the months of the year in 12-month calendars.")
month_of_year<public><native> := enum:
    January
    February
    March
    April
    May
    June
    July
    August
    September
    October
    November
    December

date_time<public><native> := struct:

    Ticks<public><native>:int := 0

    # TODO: We really want a constructor or related functionality to call validate and constrain against DateTimeMax() and DateTimeMin()

date_time_ticks<internal> := module:

    @doc("The number of nanoseconds per tick.")
    NanosecondsPerTick<public>:int := 100

    @doc("The number of ticks per day.")
    TicksPerDay<public>:int := 864000000000

    @doc("The number of ticks per hour.")
    TicksPerHour<public>:int := 36000000000

    @doc("The number of timespan ticks per microsecond.")
    TicksPerMicrosecond<public>:int := 10

    @doc("The number of timespan ticks per millisecond.")
    TicksPerMillisecond<public>:int = 10000

    @doc("The number of timespan ticks per minute.")
    TicksPerMinute<public>:int = 600000000

    @doc("The number of timespan ticks per second.")
    TicksPerSecond<public>:int = 10000000

    @doc("The number of timespan ticks per week.")
    TicksPerWeek<public>:int = 6048000000000

    @doc("The number of timespan ticks per year (365 days, not accounting for leap years).")
    TicksPerYear<public>:int = 365 * TicksPerDay

using { /Verse.org/VerseExperimental/date_time_ticks } # TODO: <redacted-ticket> - Currently no way to include these where needed. Must import the full module directly into the current file.

@doc("Creates and initializes a new instance with the specified year, month, day, hour, minute, second and millisecond.")
<#	<redacted-ticket> - Improve @doc comments to support params.
     * @param Year The year.
     * @param Month The month.
     * @param Day The day.
     * @param Hour The hour. # TODO: make optional
     * @param Minute The minute. # TODO: make optional
     * @param Second The second. # TODO: make optional
     * @param Millisecond The millisecond # TODO: make optional
#>
CreateDateTime<public><native>(Year:int, Month:int, Day:int, Hour:int, Minute:int, Second:int, Millisecond:int)<transacts><decides>:date_time

ValidateDateTime<public><native>(Year:int, Month:int, Day:int, Hour:int, Minute:int, Second:int, Millisecond:int):logic

@doc("Return the number of days in the specified month of the specified year")
DaysInMonth<public><native>(Year:int, Month:int):int

@doc("Return the number of days in the specified year")
DaysInYear<public><native>(Year:int):int

@doc("Returns the maximum date value. The maximum date value is December 31, 9999, 23:59:59.9999999. See DateTimeMin")
DateTimeMax<public>()<transacts>:date_time:=
    return date_time{Ticks:=3652059 * TicksPerDay - 1}

@doc("Returns the minimum date value. The minimum date value is January 1, 0001, 00:00:00.0. See DateTimeMax")
DateTimeMin<public>()<transacts>:date_time:=
    return date_time{Ticks:=0}

@doc("Returns a string representation of the datetime in the following format: yyyy.mm.dd-hh.mm.ss. Assumes Datetime is in UTC.")
ToString<public><native>(Val:date_time):string

#@doc("Returns the ISO-8601 string representation of datetime. Assumes Datetime is in UTC.")
#ToIso8601<public><native>(Val:date_time):string

#@doc("Returns the RFC 1123 string representation of datetime. Assumes Datetime is in UTC.")
#ToHttpDate<public><native>(Val:date_time):string

@doc("Returns the current UTC date_time.")
UtcNow<public><native>():date_time

date_parts<public><native> := struct<internal>:
    Year<native>:int
    Month<native>:int
    Day<native>:int

# TODO: Create a time_space struct and related time_span methods - <redacted-ticket>
time_of_day_parts<public><native> := struct<internal>:
    Hours<native>:int
    Minutes<native>:int
    Seconds<native>:int
    Milliseconds<native>:int
    Microseconds<native>:int
    Nanoseconds<native>:int

GetDate<public><native>(Val:date_time):date_parts


GetYear<public><native>(Val:date_time):int

# <redacted-ticket> - We can compile this fine, but calling it isn't currently supported
#postfix'.Year'<public>(Val:date_time):int:=
#    return GetYear(Val)

GetMonth<public><native>(Val:date_time):int

# <redacted-ticket> - Must use internal methods since postfix operators don't work well with VNI
#postfix'.Month'<public>(Val:date_time):int:=
#    return GetMonth(Val)

GetDay<public><native>(Val:date_time):int

# <redacted-ticket>
#postfix'.Day'<public>(Val:date_time):int:=
#    return GetDay(Val)

GetHours<public><native>(Val:date_time):int

# <redacted-ticket>
#postfix'.Hours'<public>(Val:date_time):int:=
#    return GetHours(Val)

GetMinutes<public><native>(Val:date_time):int

# <redacted-ticket>
#postfix'.Minutes'<public>(Val:date_time):int:=
#    return GetMinutes(Val)


# <redacted-ticket>
#postfix'.Seconds'<public>(Val:date_time):int:=

GetMilliseconds<public><native>(Val:date_time):int

# <redacted-ticket>
#postfix'.Milliseconds'<public>(Val:date_time):int:=
#    return GetMilliseconds(Val)

GetMicroseconds<public><native>(Val:date_time):int

# <redacted-ticket>
#postfix'.Microseconds'<public>(Val:date_time):int:=
#    return GetMicroseconds(Val)

GetNanoseconds<public><native>(Val:date_time):int

# <redacted-ticket>
#postfix'.Nanoseconds'<public>(Val:date_time):int:=
#    return GetNanoseconds(Val)

GetDayOfWeek<public><native>(Val:date_time):day_of_week

GetMonthOfYear<public><native>(Val:date_time):month_of_year

Validate<internal>(Val:date_time)<transacts><decides>:void:=
    GreaterEqual[Val, DateTimeMin()]
    LessEqual[Val, DateTimeMax()]

Less<public>(Left:date_time, Right:date_time)<transacts><decides>:date_time=
    date_time{Ticks:=Left.Ticks < Right.Ticks}

Greater<public>(Left:date_time, Right:date_time)<transacts><decides>:date_time=
    date_time{Ticks:=Left.Ticks > Right.Ticks}

GreaterEqual<public>(Left:date_time, Right:date_time)<transacts><decides>:date_time=
    date_time{Ticks:=Left.Ticks >= Right.Ticks}

LessEqual<public>(Left:date_time, Right:date_time)<transacts><decides>:date_time=
    date_time{Ticks:=Left.Ticks <= Right.Ticks}

# date_time/time_span ops returning date_time structures
PerformNativeAdd<internal><native>(Left:date_time, Right:time_span):date_time

operator'+'<public>(Left:date_time, Right:time_span):date_time=
    return PerformNativeAdd(Left, Right)

PerformNativeSubtract<internal><native>(Left:date_time, Right:time_span):date_time

operator'-'<public>(Left:date_time, Right:time_span):date_time=
    return PerformNativeSubtract(Left, Right)

# TODO: Write unit test for this...
