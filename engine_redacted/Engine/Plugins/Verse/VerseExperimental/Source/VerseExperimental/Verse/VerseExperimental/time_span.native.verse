# Copyright Epic Games, Inc. All Rights Reserved.

using { /Verse.org/Native }


time_span<native><public> := struct:

    
    Ticks<public>:int := 0

using { /Verse.org/VerseExperimental/date_time_ticks } # TODO: <redacted-ticket> - Currently no way to include these where needed. Must import the full module directly into the current file.

@doc("Creates and initializes a new instance with the specified days, hours, minutes, seconds and milliseconds.")
<#	<redacted-ticket> - Improve @doc comments to support params.
# TODO make all parameters optional
     * @param Days The days.
     * @param Hours The hours. 
     * @param Minutes The minutes.
     * @param Seconds The seconds.
     * @param Milliseconds The milliseconds
#>
CreateTimeSpan<native><public>(Days:int, Hours:int, Minutes:int, Seconds:int, Milliseconds:int)<transacts>:time_span

# No ValidateTimeSpan because there aren't restrictions on TimeSpan itself

@doc("Returns a string representation of the time span in the following format: yyyy.mm.dd-hh.mm.ss.")
ToString<native><public>(Val:time_span):string

# # <redacted-ticket> - Must use internal methods since postfix operators don't work well with VNI
#postfix<public>'.Month'(Val:time_span):int:=
#    return GetMonth(Val)


GetDays<native><public>(Val:time_span):int

# # <redacted-ticket>
#postfix<public>'.Days'(Val:time_span):int:=
#    return GetDays(Val)


GetHours<native><public>(Val:time_span):int

# # <redacted-ticket>
#postfix<public>'.Hours'(Val:time_span):int:=
#    return GetHours(Val)


GetMinutes<native><public>(Val:time_span):int

# # <redacted-ticket>
#postfix<public>'.Minutes'(Val:time_span):int:=
#    return GetMinutes(Val)



# # <redacted-ticket>
#postfix<public>'.Seconds'(Val:time_span):int:=


GetMilliseconds<native><public>(Val:time_span):int

# # <redacted-ticket>
#postfix<public>'.Milliseconds'(Val:time_span):int:=
#    return GetMilliseconds(Val)


GetMicroseconds<native><public>(Val:time_span):int

# # <redacted-ticket>
#postfix<public>'.Microseconds'(Val:time_span):int:=
#    return GetMicroseconds(Val)


GetNanoseconds<native><public>(Val:time_span):int


GetTotalDays<native><public>(Val:time_span):int

# # <redacted-ticket>
#postfix<public>'.TotalDays'(Val:time_span):int:=
#    return GetTotalDays(Val)


GetTotalHours<native><public>(Val:time_span):int

# # <redacted-ticket>
#postfix<public>'.TotalHours'(Val:time_span):int:=
#    return GetTotalHours(Val)


GetTotalMinutes<native><public>(Val:time_span):int

# # <redacted-ticket>
#postfix<public>'.TotalMinutes'(Val:time_span):int:=
#    return GetTotalMinutes(Val)


GetTotalSeconds<native><public>(Val:time_span):int

# # <redacted-ticket>
#postfix<public>'.TotalSeconds'(Val:time_span):int:=
#    return GetTotalSeconds(Val)


GetTotalMilliseconds<native><public>(Val:time_span):int

# # <redacted-ticket>
#postfix<public>'.TotalMilliseconds'(Val:time_span):int:=
#    return GetTotalMilliseconds(Val)


GetTotalMicroseconds<native><public>(Val:time_span):int

# # <redacted-ticket>
#postfix<public>'.TotalMicroseconds'(Val:time_span):int:=
#    return GetTotalMicroseconds(Val)

# pure time_span operations

PerformNativeAdd<native><internal>(Left:time_span, Right:time_span):time_span

operator'+'<public>(Left:time_span, Right:time_span):time_span=
    return PerformNativeAdd(Left, Right)


PerformNativeSubtract<native><internal>(Left:time_span, Right:time_span):time_span

operator'-'<public>(Left:time_span, Right:time_span):time_span=
    return PerformNativeSubtract(Left, Right)


PerformNativeNegate<native><internal>(Right:time_span):time_span

operator'-'<public>(Right:time_span):time_span=
    return time_span{Ticks:=-Right.Ticks}


PerformNativeMultiply<native><internal>(Left:time_span, Right:float):time_span

operator'*'<public>(Left:time_span, Right:float):time_span=
    return PerformNativeMultiply(Left, Right)

operator'*'<public>(Left:float, Right:time_span):time_span=
    return PerformNativeMultiply(Right, Left)


PerformNativeDivide<native><internal>(Left:time_span, Right:float):time_span

operator'/'<public>(Left:time_span, Right:float):time_span=
    return PerformNativeDivide(Left, Right)

Less<public>(Left:time_span, Right:time_span)<transacts><decides>:time_span=
    time_span{Ticks:=Left.Ticks < Right.Ticks}

Greater<public>(Left:time_span, Right:time_span)<transacts><decides>:time_span=
    time_span{Ticks:=Left.Ticks > Right.Ticks}

GreaterEqual<public>(Left:time_span, Right:time_span)<transacts><decides>:time_span=
    time_span{Ticks:=Left.Ticks >= Right.Ticks}

LessEqual<public>(Left:time_span, Right:time_span)<transacts><decides>:time_span=
    time_span{Ticks:=Left.Ticks <= Right.Ticks}

# date_time/time_span operations returning time_span structures

PerformNativeSubtract<native><internal>(Left:date_time, Right:date_time):time_span

operator'-'<public>(Left:date_time, Right:date_time):time_span=
    return PerformNativeSubtract(Left, Right)

# TODO: Write unit test for this...