# Copyright Epic Games, Inc. All Rights Reserved.
using { /Verse.org/Native }

value<public><native> := class:
    @doc("Retrieve an object value or fail if value is not a json object")
    AsObject<public><native>()<decides><transacts>:[string]value

    @doc("Retrieve an array value or fail if value is not a json array")
    AsArray<public><native>()<decides><transacts>:[]value

    @doc("Retrieve an integer value or fail if value is not a json number")
    AsInt<public><native>()<decides><transacts>:int

    @doc("Retrieve a float value or fail if value is not a json number")
    AsFloat<public><native>()<decides><transacts>:float

    @doc("Retrieve an object value or fail if value is not a string")
    AsString<public><native>()<decides><transacts>:string

    #@doc("Retrieve an object value or fail if value is not a true or false value")
    #AsBool<public><native>()<decides><transacts>:logic

    @doc("Retrieve an object value or fail if value is not null")
    AsNull<public><native>()<decides><transacts>:void

@doc("Parse a JSON string returning a value with its contents")
Parse<public><native>(JSONString:string)<decides><transacts>:value
