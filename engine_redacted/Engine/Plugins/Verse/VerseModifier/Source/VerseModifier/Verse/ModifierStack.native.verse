# Copyright Epic Games, Inc. All Rights Reserved.

using { /Verse.org/Verse }
using { /Verse.org/Native }
using { /Verse.org/Simulation }
using { /Verse.org/Temporary/EpicGamesRestricted/Network }


@doc("Implemented by classes to provide a method for modification evaluation.")
@experimental
@available{MinUploadedAtFNVersion := 4100}
modifier<public><native>(t:type) := interface:

    Evaluate<public><native_callable>(InValue:t)<reads> : t

@doc("Modifier stacks provide an ordered application of modifiers.")
@experimental
@available{MinUploadedAtFNVersion := 4100}
modifier_stack<public><native>(t:type) := class(modifier(t)):

    @doc("Returns a t which is the input evaluated against each modifier in the stack, executed in position order.")
    Evaluate<override>(InValue:t)<reads>:t =
        EvaluateRecursiveInternal(InValue, 0)

    @doc("Insert a Modifier at Position in the stack.\n{
         }If multiple modifiers are added at the same position, they are applied such that the most recently added is the last evaluated.\n{
         }Returns a cancelable which can be used to remove Modifier from the stack.")
    AddModifier<public>(Modifier:modifier(t), Position:rational)<transacts>:cancelable =

    # Recursive evaluation as we dont have a Reduce() function yet
    EvaluateRecursiveInternal<private>(InValue:t, Index:int)<reads>:t =
        
    # Callback from cancellable entry that will remove the entry from the stack
    
    # Native impl backing AddModifier
    AddModifierInternal<internal><native>(Modifier:modifier(t), Position:rational)<transacts>:cancelable

    # Returns the current first position in the stack. If the stack is empty, this is the default position (0).
    var<private> FirstPosition<native><public>:rational = 0
    
    # Returns the current last position in the stack. If the stack is empty, this is the default position (0).

    OnRep_Entries<private><native>():void
