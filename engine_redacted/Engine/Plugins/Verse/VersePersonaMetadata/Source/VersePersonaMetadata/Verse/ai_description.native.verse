# Copyright Epic Games, Inc. All Rights Reserved.

using { /Verse.org/Native }

@available{MinUploadedAtFNVersion := 4100} @attribscope_data @customattribhandler
ai_description_attribute<public> := class<internal><computes>(attribute):
    Description<public>:string

@available{MinUploadedAtFNVersion := 4100}
ai_description<public><constructor>(Description:string)<computes> := ai_description_attribute:
    Description := Description
