# Copyright Epic Games, Inc. All Rights Reserved.

using { /Verse.org/Native }
using { /Verse.org/Colors }

@doc("Writes `Message` to a dedicated `Print` log while displaying it in `Color` on the {
     }client screen for `Duration` seconds. By default, `Color` is `NamedColors.White` and {
     }`Duration` is `2.0` seconds.")
Print<native><public>(Message:string, ?Duration:float = 2.0, ?Color:color = NamedColors.White)<predicts><transacts>:void

@doc("Writes `Message` to a dedicated `Print` log while displaying it in `Color` on the {
     }client screen for `Duration` seconds. By default, `Color` is `NamedColors.White` and {
     }`Duration` is `2.0` seconds.")
Print<native><public>(Message:message, ?Duration:float = 2.0, ?Color:color = NamedColors.White)<predicts><transacts>:void

@doc("Writes `Message` to a dedicated `Print` log while displaying it in `Color` on the {
     }client screen for `Duration` seconds. By default, `Color` is `NamedColors.White` and {
     }`Duration` is `2.0` seconds.")
Print<public>(Message:diagnostic, ?Duration:float = 2.0, ?Color:color = NamedColors.White)<transacts>:void=
