# Copyright Epic Games, Inc. All Rights Reserved.

using { /Verse.org/Native }

@doc("Type for which there is a single instance per round.  Use `GetSession` to {
     }get the current round's `session` instance. May be used with `weak_map` to {
     }implement global variables.\n{
     }Note: may be changed in a future release to a single instance per game. {
     }Round-local behavior should not be relied upon.")
session<native><public> := class<unique><module_scoped_var_weak_map_key> {}

@doc("Returns the `session` corresponding to the current round.  The result can {
     }be used with `weak_map` to implement global variables.\n{
     }Note: may be changed in a future release to return a single instance per {
     }game. Round-local behavior should not be relied upon.")
GetSession<native><public>()<reads>:session

@doc("Specifies what type of environment the current session is in.")
session_environment<public><native> := enum:
    @doc("The current session is in an Edit environment for an experience, such as a session started within UEFN.")
    Edit,

    @doc("The current session is in a Private environment for an experience, such as a playtest.")
    Private,

    @doc("The current session is in a Live environment for an experience.")
    Live

@available{MinUploadedAtFNVersion := 3800}
(Session:session).Environment<public><native>()<transacts>:session_environment 