# Copyright Epic Games, Inc. All Rights Reserved.

using { /Verse.org/Native }

# /Verse.org/Simulation module

@doc("{
}Waits specified number of seconds and then resumes. If `Seconds` = 0.0 then it waits until next tick/frame/update. If `Seconds` = Inf then it waits forever and only calls back if canceled - such as via `race`. If `Seconds` < 0.0 then it completes immediately and does not yield to other aysnc expressions.\n{
}Waiting until the next update (0.0) is especially useful in a loop of a coroutine that needs to do some work every update and this yields to other coroutines so that it doesn't hog a processor's resources.\n{
}Waiting forever (Inf) will have any expression that follows never be evaluated. Occasionally it is desireable to have a task never complete such as the last expression in a `race` subtask where the task must never win the race though it still may be canceled earlier.\n{
}Immediately completing (less than 0) is useful when you want programmatic control over whether an expression yields or not.\n{
}")
Sleep<native><public>(Seconds:float)<suspends><predicts>:void

@doc("Get the seconds that have elapsed since the world began simulating")
GetSimulationElapsedTime<native><public>()<transacts>:float
