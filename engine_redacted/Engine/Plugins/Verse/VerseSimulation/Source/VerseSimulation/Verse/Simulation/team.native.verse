# Copyright Epic Games, Inc. All Rights Reserved.

using { /Verse.org/Native }

@import_as("/Script/VerseSimulation.UVerseTeam")
team_base<native> := class<abstract>:

team<native><public> := class<unique>(team_base){}
