# Copyright Epic Games, Inc. All Rights Reserved.

using { /Verse.org/Native }


# Custom attribute to allow properties to be exposed in the editor
@attribscope_class @attribscope_struct @attribscope_data @customattribhandler
editable<public> := class<computes>(attribute):
    # Currently, inside DigestGenerator.cpp, we elide generating ToolTip fields
    # in the digest so that we don't need to transitively put the value of the ToolTip 
    # also in the digest. It's possible in the future that we may need to extend that
    # logic to other fields here. Doing so would be trivial. Just grep for "ToolTip"
    # in DigestGenerator.cpp
    @doc("The tooltip displayed in the editor.")
    ToolTip<public>:message = redacted_internal
    @doc("The categories displayed in the editor.")
    Categories<public>:[]message = array{}

@attribscope_class @attribscope_struct @attribscope_data @customattribhandler
editable_text_box<public> := class<final><computes>(editable):
    @doc("True if the editor text box should support multiple lines.")
    MultiLine<public>:logic = false
    @doc("The maximum length of the text.")
    MaxLength<public>:int = -1

@attribscope_class @attribscope_struct @attribscope_data @customattribhandler
editable_slider<public>(t:type) := class<final><computes>(editable):
    @doc("The minimum value of the editor slider.")
    MinValue<public>:?t = false
    @doc("The maximum value of the editor slider.")
    MaxValue<public>:?t = false
    @doc("The amount the slider moves for each step.")
    SliderDelta<public>:?t = false
    @doc("Used to scale a slider exponentially. Common values are found within the range of 1-20.")
    SliderExponent<public>:?t = false
    @doc("Used to change how sensitive the field value is when moving the slider via mouse cursor.")
    MouseLinearDeltaSensitivity<public>:float = 0.0
    MouseShiftMovePixelPerDelta<public>:float = 0.0

@attribscope_class @attribscope_struct @attribscope_data @customattribhandler
editable_number<public>(t:type) := class<final><computes>(editable):
    @doc("The minimum value the editor allows.")
    MinValue<public>:?t = false
    @doc("The maximum value the editor allows.")
    MaxValue<public>:?t = false
    @doc("Snap the spinbox to the nearest delta.")
    SpinBoxDelta<public>:?t = false

@attribscope_class @attribscope_struct @attribscope_data @customattribhandler
editable_vector_slider<public>(t:type) := class<final><computes>(editable):
    @doc("Show the button that allows locking the vector aspect ratio.")
    ShowPreserveRatio<public>:logic = false
    @doc("Show the button that allows normalizing the vector.")
    ShowNormalize<public>:logic = false
    @doc("The minimum value allowed for each vector element.")
    MinComponentValue<public>:?t = false
    @doc("The maximum value allowed for each vector element.")
    MaxComponentValue<public>:?t = false
    @doc("The amount the slider moves for each step.")
    SliderDelta<public>:?t = false
    @doc("Used to scale a slider exponentially. Common values are found within the range of 1-20.")
    SliderExponent<public>:?t = false
    @doc("Used to change how sensitive the field value is when moving the slider via mouse cursor.")
    MouseLinearDeltaSensitivity<public>:float = 0.0
    MouseShiftMovePixelPerDelta<public>:float = 0.0

@attribscope_class @attribscope_struct @attribscope_data @customattribhandler
editable_vector_number<public>(t:type) := class<final><computes>(editable):
    @doc("Show the button that allows locking the vector aspect ratio.")
    ShowPreserveRatio<public>:logic = false
    @doc("Show the button that allows normalizing the vector.")
    ShowNormalize<public>:logic = false
    @doc("The minimum value allowed for each vector element.")
    MinComponentValue<public>:?t = false
    @doc("The maximum value allowed for each vector element.")
    MaxComponentValue<public>:?t = false
    @doc("Snap the spinbox to the nearest delta.")
    SpinBoxDelta<public>:?t = false

@attribscope_class @attribscope_struct @attribscope_data @customattribhandler
editable_container<public> := class<final><computes>(editable):
    @doc("If true, allows reordering of the container elements in the editor.")
    AllowReordering<public>:logic = true
