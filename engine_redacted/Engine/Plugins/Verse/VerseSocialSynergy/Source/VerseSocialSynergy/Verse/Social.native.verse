# Copyright Epic Games, Inc. All Rights Reserved.

using { /Verse.org/AgentGroup }
using { /Verse.org/Native }
using { /Verse.org/Simulation }

# Identity Semantics:
# - Calling `GetLocalParty()` on two players in the same party returns the SAME instance.
# - This enables identity comparison: `PlayerA.GetLocalParty() = PlayerB.GetLocalParty()` succeeds
#   if and only if they are in the same party.
# - Event subscriptions on a party instance will fire for all members' actions,
#   regardless of which member's `GetLocalParty()` call was used to obtain it.

@available{MinUploadedAtFNVersion := 4210}
@doc("Per-member info for party members. Can be extended with additional fields in future versions.")
party_member_info<native><public> := class(member_info_interface):

@available{MinUploadedAtFNVersion := 4210}
(InPlayer:player).GetLocalParty<native><public>()<reads>:agent_group_interface(party_member_info)
