# Copyright Epic Games, Inc. All Rights Reserved.

using { /Verse.org/Native }
using { /Verse.org/Simulation }
using { /Verse.org/Temporary/EpicGamesRestricted/Network }

@doc("An abstract representation of an orientation change in 3d-space.")
@replicatable
@import_as("/*.FVRotation")
rotation<native><public> := struct<concrete><predicts><uht_comparable> {}    # Internal structure is opaque

@doc("Makes a `rotation` from `Axis` and `Angle` in radians using a right-handed sign convention (e.g. {
     }a positive rotation around Up takes Forward to Left).")
@available{MinUploadedAtFNVersion := 3600} # Guarding against function name collision in already published projects. 
@vm_no_effect_token
MakeRotationRadians<native><public>(Axis:vector3, Angle:float)<reads><converges>:rotation

@doc("Degrees version of `MakeRotationRadians`")
@available{MinUploadedAtFNVersion := 3600} # Guarding against function name collision in already published projects. 
MakeRotationDegrees<public>(Axis:vector3, Angle:float)<reads>:rotation=
    MakeRotationRadians(Axis, DegreesToRadians(Angle))

@doc("Makes a `rotation` by applying a pre-rotation of `YawAngle` followed by `PitchAngle` and then `RollAngle`, in that order:\n{
     } * *yaw* is right-handed rotation about the Down axis,\n{
     } * *pitch* is right-handed rotation about the Right axis,\n{
     } * *roll* is right-handed rotation about the Forward axis.")
@available{MinUploadedAtFNVersion := 3600} # Guarding against function name collision in already published projects. 
MakeRotationFromYawPitchRollRadians<public>(YawAngle:float, PitchAngle:float, RollAngle:float)<reads>:rotation=
    MakeRotationFromYawPitchRollDegrees(RadiansToDegrees(YawAngle), RadiansToDegrees(PitchAngle), RadiansToDegrees(RollAngle))

@doc("Degrees version of `MakeRotationFromYawPitchRollRadians`")
@vm_no_effect_token
MakeRotationFromYawPitchRollDegrees<native><public>(YawAngle:float, PitchAngle:float, RollAngle:float)<reads><converges>:rotation

@doc("Makes a `rotation` by applying a post-rotation of `LeftAxisAngle` followed by `UpAxisAngle` and then `ForwardAxisAngle `{
     }in that order. Right-handed convention (e.g. a positive rotation around Up takes +Forward to Left).")
@available{MinUploadedAtFNVersion := 3600} # Guarding against function name collision in already published projects. 
@vm_no_effect_token
MakeRotationFromEulerRadians<native><public>(LeftAxisAngle:float, UpAxisAngle:float, ForwardAxisAngle:float)<reads><converges>:rotation

@doc("Degrees version of `MakeRotationFromEulerRadians`")
@available{MinUploadedAtFNVersion := 3600} # Guarding against function name collision in already published projects. 
MakeRotationFromEulerDegrees<public>(LeftAxisAngle:float, UpAxisAngle:float, ForwardAxisAngle:float)<reads>:rotation=
    MakeRotationFromEulerRadians(DegreesToRadians(LeftAxisAngle), DegreesToRadians(UpAxisAngle), DegreesToRadians(ForwardAxisAngle))

@doc("Makes the identity `rotation`.")
IdentityRotation<native><public>()<converges><predicts>:rotation

@doc("Returns the distance between `Rotation1` and `Rotation2`. The result will be between:\n{
    } * `0.0`, representing equivalent rotations and\n{
    } * `1.0` representing rotations which are 180 degrees apart (i.e., the shortest rotation between them is 180 degrees around some axis).")
@vm_no_effect_token
Distance<native><public>(Rotation1:rotation, Rotation2:rotation)<reads><converges>:float

@doc("Returns the smallest angular distance between `Rotation1` and `Rotation2` in radians.")
@available{MinUploadedAtFNVersion := 3600} # Guarding against function name collision in already published projects. 
@vm_no_effect_token
AngularDistanceRadians<native><public>(Rotation1:rotation, Rotation2:rotation)<reads><converges>:float

@doc("Degrees version of `AngularDistanceRadians`.")
@available{MinUploadedAtFNVersion := 3600} # Guarding against function name collision in already published projects. 
AngularDistanceDegrees<public>(Rotation1:rotation, Rotation2:rotation)<reads>:float=
    RadiansToDegrees(AngularDistanceRadians(Rotation1, Rotation2))

@doc("Apply a `PreRotation` to `PostRotation` as `v * PreRotation * PostRotation`.")
@available{MinUploadedAtFNVersion := 3600} # Guarding against function name collision in already published projects. 
@vm_no_effect_token
operator'*'<public><native>(PreRotation:rotation, PostRotation:rotation)<reads><converges>:rotation

@doc("Makes a `tuple(float, float, float)` with three elements:\n{
     } * *yaw* of `rotation` in radians\n{
     } * *pitch* of `rotation` in radians\n{
     } * *roll* of `rotation` in radians\n{
     }using the conventions of `MakeRotationFromYawPitchRollRadians`.")
@available{MinUploadedAtFNVersion := 3600} # Guarding against function name collision in already published projects. 
(Rotation:rotation).GetYawPitchRollRadians<public>()<reads>:tuple(float, float, float)=
    Degrees := Rotation.GetYawPitchRollDegrees()
    (DegreesToRadians(Degrees(0)), DegreesToRadians(Degrees(1)), DegreesToRadians(Degrees(2)))

@doc("Degrees version of `GetYawPitchRollRadians`.")
@available{MinUploadedAtFNVersion := 3600} # Guarding against function name collision in already published projects. 
@vm_no_effect_token
(Rotation:rotation).GetYawPitchRollDegrees<public><native>()<reads><converges>:tuple(float, float, float)

@doc("Makes a `tuple(float, float, float)` with three elements:\n{
     } * *left axis* `rotation` in radians\n{
     } * *up axis* of `rotation` in radians\n{
     } * *forward axis* of `rotation` in radians\n{
     }using the conventions of `MakeRotationEulerRadians`.")
@available{MinUploadedAtFNVersion := 3600} # Guarding against function name collision in already published projects. 
@vm_no_effect_token
(Rotation:rotation).GetEulerRadians<native><public>()<reads><converges>:tuple(float, float, float)

@doc("Degrees version of `GetEulerRadians`.")
@available{MinUploadedAtFNVersion := 3600} # Guarding against function name collision in already published projects. 
(Rotation:rotation).GetEulerDegrees<public>()<reads>:tuple(float, float, float)=
    Radians := Rotation.GetEulerRadians()
    (RadiansToDegrees(Radians(0)), RadiansToDegrees(Radians(1)), RadiansToDegrees(Radians(2)))

@doc("Makes a `vector3` from the axis of `rotation` for an right-handed angle.\n{
     }If `rotation` is nearly identity, this will return the +Forward axis. See also `GetAngleRadians`.")
@vm_no_effect_token
(Rotation:rotation).GetAxis<native><public>()<predicts><reads><converges>:vector3

@doc("Returns the radians of right-handed `rotation` around the axis of `rotation`. See also `GetAxis`.")
@available{MinUploadedAtFNVersion := 3600} # Guarding against function name collision in already published projects. 
@vm_no_effect_token
(Rotation:rotation).GetAngleRadians<native><public>()<reads><converges>:float

@doc("Degrees version of GetAngleRadians")
@available{MinUploadedAtFNVersion := 3600} # Guarding against function name collision in already published projects. 
(Rotation:rotation).GetAngleDegrees<public>()<reads>:float=
    RadiansToDegrees(Rotation.GetAngleRadians())

@doc("Makes the smallest angular `rotation` from `InitialVector` to `FinalVector` two vectors of arbitrary length such that:\n{
     }`InitialVector * MakeShortestRotationBetween(InitialVector, FinalVector) = FinalVector` and\n{
     }`MakeShortestRotationBetween(InitialVector, FinalVector)?.GetAngleRadians()` is as small as possible.")
@vm_no_effect_token
MakeShortestRotationBetween<native><public>(InitialVector:vector3, FinalVector:vector3)<reads><converges>:rotation

# JIRA: <redacted-ticket>
# TODO: check special forms and extrapolation.
@doc("Used to perform spherical linear interpolation between `From` (when `Ratio = 0.0`) {
     }and `To` (when `Ratio = 1.0`). Expects `0.0 <= Ratio <= 1.0`.")
@vm_no_effect_token
Slerp<native><public>(InitialRotation:rotation, FinalRotation:rotation, Ratio:float)<reads><converges>:rotation

@doc("Makes a `vector3` by applying `Rotation` to `Vector`.")
@vm_no_effect_token
operator'*'<public><native>(Vector:vector3, Rotation:rotation)<predicts><reads><converges>:vector3

@doc("Makes a `rotation` by inverting `Rotation` such that `ApplyRotation(Rotation, Rotation.Invert())) = IdentityRotation`.")
@vm_no_effect_token
(Rotation:rotation).Invert<native><public>()<reads><converges>:rotation

@doc("Returns `Rotation` if it does not contain `NaN`, `Inf` or `-Inf`.")
(Rotation:rotation).IsFinite<native><public>()<decides><converges>:rotation

@doc("Makes a unit `vector3` pointing in the *forward* rotated direction.\n{
     }This is equivalent to: `vector3\{Forward:=1.0, Left:=0.0, Up:=0.0\} * Rotation`.")
(Rotation:rotation).GetForwardAxis<public>()<predicts><reads>:vector3 =

@doc("Makes a unit `vector3` pointing in the *left* rotated direction.\n{
     }This is equivalent to: `vector3\{Forward:=0.0, Left:=1.0, Up:=0.0\} * Rotation`.")
(Rotation:rotation).GetLeftAxis<public>()<predicts><reads>:vector3=

@doc("Makes a unit `vector3` pointing in the *up* rotated direction.\n{
     }This is equivalent to: `vector3\{Forward:=0.0, Left:=0.0, Up:=1.0\} * Rotation`.")
(Rotation:rotation).GetUpAxis<public>()<predicts><reads>:vector3 =

@doc("Makes a `string` representation of `rotation` in axis/degrees format with a right-handed sign convention.\n{
     }`ToString(MakeRotationRadians(vector3\{Left:=0.0, Up:=0.0, Forward:=1.0\}, PiFloat/2.0))` produces the string: {
     }`\"\{Axis = \{Left=0.000000, Up=0.000000, Forward=1.000000\}, Angle = 90.000000\}\"`.")
ToString<public>(Rotation:rotation)<reads>:string=
    "\{Axis = {Rotation.GetAxis()}, Angle = {RadiansToDegrees(Rotation.GetAngleRadians())}\}"

# JIRA: <redacted-ticket>
# TODO: Consider the order of the multiplications. Do we want to fold the constant PiFloat / 180.0?
#       Also, these should be <converges> when this is supported for Verse code.

@doc("Returns radians from `Degrees`.")
DegreesToRadians<public>(Degrees:float)<reads>:float =
    Degrees * PiFloat / 180.0

@doc("Returns degrees from `Radians`.")
RadiansToDegrees<public>(Radians:float)<reads>:float =
    Radians * 180.0 / PiFloat
