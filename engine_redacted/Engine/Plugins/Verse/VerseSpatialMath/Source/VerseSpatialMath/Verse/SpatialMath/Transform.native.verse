# Copyright Epic Games, Inc. All Rights Reserved.

using { /Verse.org/Native }
using { /Verse.org/Simulation }
using { /Verse.org/Temporary/EpicGamesRestricted/Network }

@doc("A combination of scale, rotation, and translation, applied in that order.")
@replicatable
@import_as("/*.FVTransform")
transform<native><public> := struct<concrete><computes><predicts><uht_comparable>:
    @editable
    @doc("The location of this `transform`.")
    Translation<native><public>:vector3 := vector3{Forward:=0.0,Left:=0.0,Up:=0.0}

    @editable
    @doc("The rotation of this `transform`.")
    Rotation<native><public>:rotation := IdentityRotation()

    @editable
    @doc("The scale of this `transform`.")
    @units("x")
    Scale<native><public>:vector3 := vector3{Forward:=1.0,Left:=1.0,Up:=1.0}

@doc("Makes a `vector3` by applying `InTransform` to `InVector`.")
operator'*'<public>(InVector:vector3, InTransform:transform)<reads>:vector3=
    ((InTransform.Scale * InVector) * InTransform.Rotation) + InTransform.Translation

@doc("Makes a `string` representation of `InTransform` where the result is on the form.\n{
     }`\"\{Translation = \{ToString(`InTransform.Translation`)\}, Rotation = \{ToString(`InTransform.Rotation`)\}, Scale = \{ToString(`InTransform.Scale`)\}\}\".")
ToString<public>(InTransform:transform)<reads>:string=
    "\{Translation = {InTransform.Translation}, Rotation = {InTransform.Rotation}, Scale = {InTransform.Scale}\}"