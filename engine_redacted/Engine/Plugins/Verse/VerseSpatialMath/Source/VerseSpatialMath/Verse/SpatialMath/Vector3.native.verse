# Copyright Epic Games, Inc. All Rights Reserved.

using { /Verse.org/Native }
using { /Verse.org/Simulation }
using { /Verse.org/Temporary/EpicGamesRestricted/Network }

# What number should this actually be in verse? For now we use 0.00000001 = 1.e-8 = UE_DOUBLE_SMALL_NUMBER
InternalTolerance<internal>:float= 0.00000001

@doc("3-dimensional vector with `float` components.")
@replicatable
@import_as("/*.FVVector3")
vector3<native><public> := struct<concrete><computes><persistable><uht_comparable><predicts>:
    @editable
    @doc("The Left (was -Y) component of this vector.")
    Left<native><public>:float = 0.0

    @editable
    @doc("The Up (was Z) component of this vector.")
    Up<native><public>:float = 0.0

    @editable
    @doc("The Forward (was X) component of this vector.")
    Forward<native><public>:float = 0.0

@doc("Makes a `vector3` by inverting the `SurfaceNormal` component of `Direction`.")
ReflectVector<public>(Direction:vector3, SurfaceNormal:vector3)<reads>:vector3=
    
@doc("Returns the dot product of `V1` and `V2`.")
DotProduct<public>(V1:vector3, V2:vector3)<predicts><reads>:float=
    V1.Forward * V2.Forward + V1.Left * V2.Left + V1.Up * V2.Up

@doc("Returns the left-handed cross product of `V1` and `V2`.")
CrossProductLeftHanded<public>(V1:vector3, V2:vector3)<reads>:vector3=
    vector3{ Forward := -V1.Left * V2.Up - V1.Up * -V2.Left, Left := -(V1.Up * V2.Forward - V1.Forward * V2.Up), Up := V1.Forward * -V2.Left - -V1.Left * V2.Forward }

@doc("Returns the right-handed cross product of `V1` and `V2`.")
@available{MinUploadedAtFNVersion := 3600} # Guarding against function name collision in already published projects. 
CrossProduct<public>(V1:vector3, V2:vector3)<predicts><reads>:vector3=
    vector3{ Forward := V1.Left * V2.Up - V1.Up * V2.Left, Left := V1.Up * V2.Forward - V1.Forward * V2.Up, Up := V1.Forward * V2.Left - V1.Left * V2.Forward }

@doc("Returns the Euclidean distance between `V1` and `V2`.")
Distance<public>(V1:vector3, V2:vector3)<reads>:float=
    (V2 - V1).Length()

@doc("Returns the squared Euclidean distance between `V1` and `V2`.")
DistanceSquared<public>(V1:vector3, V2:vector3)<reads>:float=
    (V2 - V1).LengthSquared()

@doc("Returns the 2-D Euclidean distance between `V1` and `V2` by ignoring the difference in `Up`.")
DistanceForwardLeft<public>(V1:vector3, V2:vector3)<reads>:float=
    (V2 - V1).LengthForwardLeft()

@doc("Returns the squared 2-D Euclidean distance between `V1` and `V2` by ignoring their difference in `Up`.")
DistanceSquaredForwardLeft<public>(V1:vector3, V2:vector3)<reads>:float=
    (V2 - V1).LengthSquaredForwardLeft()

@doc("Makes a unit length `vector3` pointing in the same direction of `V`.")
(V:vector3).MakeUnitVector<public>()<predicts><reads>:vector3=
    V / V.Length()

@doc("Makes a `string` representation of `V`.")
ToString<public>(V:vector3)<predicts><reads>:string=
    "\{Forward = {V.Forward}, Left = {V.Left}, Up = {V.Up}\}"
    
@doc("Returns the length of `V`.")
(V:vector3).Length<public>()<predicts><reads>:float=
    Sqrt(V.LengthSquared())

@doc("Returns the squared length of `V`.")
(V:vector3).LengthSquared<public>()<computes><predicts>:float=
    (V.Forward*V.Forward + V.Left*V.Left + V.Up*V.Up)

@doc("Returns the length of `V` as if `V.Up = 0.0`.")
(V:vector3).LengthForwardLeft<public>()<reads>:float=
    Sqrt(V.LengthSquaredForwardLeft())

@doc("Returns the squared length of `V` as if `V.Up = 0.0`.")
(V:vector3).LengthSquaredForwardLeft<public>()<reads>:float=
    (V.Forward*V.Forward + V.Left*V.Left)

# JIRA: <redacted-ticket>
@doc("Used to linearly interpolate/extrapolate between `From` (when `Parameter = 0.0`) {
     }and `To` (when `Parameter = 1.0`). Expects that all arguments are finite.\n{
     }Returns `From*(1 - Parameter) + To*Parameter`.")
Lerp<public>(From:vector3, To:vector3, Parameter:float)<reads>:vector3:=
    vector3:
        Forward:=Lerp(From.Forward, To.Forward, Parameter)
        Left:=Lerp(From.Left, To.Left, Parameter)
        Up:=Lerp(From.Up, To.Up, Parameter)

@doc("Makes a `vector3` by inverting the signs of `Operand`.")
prefix'-'<public>(Operand:vector3)<computes>:vector3 =
    vector3{Forward := -Operand.Forward, Left := -Operand.Left, Up := -Operand.Up}

@doc("Makes a `vector3` by component-wise addition of `Left` and `Right`.")
operator'+'<public>(Left:vector3, Right:vector3)<computes>:vector3 =
    vector3{Forward := Left.Forward + Right.Forward, Left := Left.Left + Right.Left, Up := Left.Up + Right.Up}

@doc("Makes a `vector3` by component-wise subtraction of `Right` from `Left`.")
operator'-'<public>(Left:vector3, Right:vector3)<computes>:vector3 =
    vector3{Forward := Left.Forward - Right.Forward, Left := Left.Left - Right.Left, Up := Left.Up - Right.Up}

@doc("Makes a `vector3` by component-wise multiplication of `Left` and `Right`.")
operator'*'<public>(Left:vector3, Right:vector3)<computes>:vector3 =
    vector3{Forward := Left.Forward * Right.Forward, Left := Left.Left * Right.Left, Up := Left.Up * Right.Up}

@doc("Makes a `vector3` by multiplying the components of `Left` by `Right`.")
operator'*'<public>(Left:vector3, Right:float)<computes>:vector3 =
    vector3{Forward := Left.Forward * Right, Left := Left.Left * Right, Up := Left.Up * Right}

@doc("Makes a `vector3` by multiplying the components of `Right` by `Left`.")
operator'*'<public>(Left:float, Right:vector3)<computes>:vector3 =
    vector3{Forward := Left * Right.Forward, Left := Left * Right.Left, Up := Left * Right.Up}

@doc("Makes a `vector3` by dividing the components of `Left` by `Right`.")
operator'/'<public>(Left:vector3, Right:float)<computes><predicts>:vector3 =
    vector3{Forward := Left.Forward / Right, Left := Left.Left / Right, Up := Left.Up / Right}

@doc("Makes a `vector3` by component-wise division of `Left` by `Right`.")
operator'/'<public>(Left:vector3, Right:vector3)<computes>:vector3 =
    vector3{Forward := Left.Forward / Right.Forward, Left := Left.Left / Right.Left, Up := Left.Up / Right.Up}

@doc("Returns `V` if all components are finite.\n{
     }Fails if any of the components are not finite.")
(V:vector3).IsFinite<public>()<decides><computes>:vector3 =
    vector3{Forward:=V.Forward.IsFinite[], Left:=V.Left.IsFinite[], Up:=V.Up.IsFinite[]}

# JIRA: <redacted-ticket>
#TODO: consider what happens with Inf and NaN.

@doc("Succeeds when each component of `V` is within `AbsoluteTolerance` of `0.0`.")
(V:vector3).IsAlmostZero<public>(AbsoluteTolerance:float)<decides><computes>:void =
    Abs(V.Forward) <= AbsoluteTolerance and Abs(V.Left) <= AbsoluteTolerance and Abs(V.Up) <= AbsoluteTolerance

@doc("Succeeds when each component of `V1` and `V2` are within `AbsoluteTolerance` of each other.")
IsAlmostEqual<public>(V1:vector3, V2:vector3, AbsoluteTolerance:float)<decides><computes>:void =
    Abs(V1.Forward - V2.Forward) <= AbsoluteTolerance and Abs(V1.Left - V2.Left) <= AbsoluteTolerance and Abs(V1.Up - V2.Up) <= AbsoluteTolerance