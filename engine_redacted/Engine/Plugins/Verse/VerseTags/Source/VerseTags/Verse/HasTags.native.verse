# Copyright Epic Games, Inc. All Rights Reserved.

using { /Verse.org/Native }

# TODO: should this implement `diagnosable`? If so, how will that work together with the fact that `entity` is already implementing `diagnosable`?
@doc("An interface representing a mutable collection of tags.")
has_tags<public><native> := interface:
    @doc("Adds a `tag` instance to this container. Returns a `tag_key` that is uniquely associated with the added instance.")
    AddTag<native_callable><public>(Tag:tag)<transacts>:tag_key
    
    @doc("Removes the tag instance associated with the `tag_key`, succeeds if an instance was removed, fails otherwise.")
    RemoveTag<native_callable><public>(Key:tag_key)<transacts><decides>:void
    
    @doc("Removes all tag instances of type `tag_type`, succeeds if at least one instance was removed, fails otherwise.")
    RemoveAllTags<native_callable><public>(tag_type:castable_subtype(tag))<transacts><decides>:void
    
    @doc("Removes all tag instances that are not of type `tag_type`, succeeds if at least one instance was removed, fails otherwise.")
    RemoveAllTagsExcept<native_callable><public>(tag_type:castable_subtype(tag))<transacts><decides>:void
    
    @doc("Removes all tag instances that are not of any of the types in `tag_types`, succeeds if at least one instance was removed, fails otherwise.")
    RemoveAllTagsExcept<native_callable><public>(tag_types:[]castable_subtype(tag))<transacts><decides>:void

    @doc("Succeeds if at least one tag of type `tag_type` is found in this container, fails otherwise.")
    ContainsTag<native_callable><public>(tag_type:castable_subtype(tag))<decides><reads>:void
    
    @doc("Fails if at least one type in `tag_types` cannot be found in this container, succeeds otherwise. Note that this means that if `tag_types` is empty this call succeeds.")
    ContainsAllTags<native_callable><public>(tag_types:[]castable_subtype(tag))<decides><reads>:void
    
    @doc("Succeeds if at least of the types in `tag_types` is found in this container, fails otherwise. Note that this means that if `tag_types` is empty this call fails.")
    ContainsAnyTag<native_callable><public>(tag_types:[]castable_subtype(tag))<decides><reads>:void