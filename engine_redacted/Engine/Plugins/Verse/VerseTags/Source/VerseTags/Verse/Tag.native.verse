# Copyright Epic Games, Inc. All Rights Reserved.

using { /Verse.org/Native }

@import_as("/Script/VerseTags.UVerseTagBase")
tag_base<native> := class<abstract>:

@doc("A base type used for tagging objects in order to hierarchically evaluate an objects classification.")
tag<native><public> := class<abstract><castable>(tag_base){}

@doc("Get a string representation of the tag. Useful for debugging.")
ToString<native>(:tag)<computes>:string

@doc("Succeeds if Left and Right are equal.")
Equals<native>(Left:tag, Right:tag)<decides><transacts>:void
