# Copyright Epic Games, Inc. All Rights Reserved.

using { /Verse.org/Native }

@import_as("/Script/VerseTags.UVerseTagContainerBase")
tag_container_base<native> := class<abstract>:

@doc("Returns abbreviated human readable Tag list. Useful for debugging.")
ToString<native>(:redacted_internal)<computes>:string

@doc("Succeeds if all tags in Left are found in Right and all tags in Right are found in Left.")
Equals<native>(Left:redacted_internal, Right:redacted_internal)<decides><transacts>:void
