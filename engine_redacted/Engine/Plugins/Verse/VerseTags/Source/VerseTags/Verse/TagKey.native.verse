# Copyright Epic Games, Inc. All Rights Reserved.

using { /Verse.org/Native }

# TODO: Evaluate if this should have a public constructor such that it can be default constructed by creators.
@doc("A `tag_key` is the return value from adding a `tag` to a container implementing the `has_tags` interface, and is used to selectively remove such an instance from the same container.")
@import_as("/*.FVerseTagKey")
tag_key<public><native> := struct<internal><uht_comparable>{}
