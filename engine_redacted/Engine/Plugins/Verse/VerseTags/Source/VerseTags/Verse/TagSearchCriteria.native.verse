# Copyright Epic Games, Inc. All Rights Reserved.

using { /Verse.org/Native }

tag_search_sort_type<native><public>:=enum:
    Unsorted,
    Sorted

@doc("Advanced tag search criteria")
@deprecated{Message := "`tag_search_criteria` is part of the old tag API. Use `FindCreativeObjectsWithTag` with a `tag` type instead."}
tag_search_criteria<native><public> := class:
    @doc("Tags required to be on the object.")
    
    RequiredTags<native><public>:[]tag := array{}

    @doc("Tags that are used if no required tags are specified. These are treated as if any of them will do.")
    
    PreferredTags<native><public>:[]tag := array{}

    @doc("Tags that may NOT be on the object. All items with these tags are excluded from the search.")
    
    ExclusionTags<native><public>:[]tag := array{}

    @doc("Flag to request sorting the results by tag.")
    
    SortType<native><public>:tag_search_sort_type := tag_search_sort_type.Sorted
