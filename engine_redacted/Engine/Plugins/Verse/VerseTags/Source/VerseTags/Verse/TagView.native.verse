# Copyright Epic Games, Inc. All Rights Reserved.

using { /Verse.org/Native }

# Internal note: This is intentionally implement to allow for 'hidden' tags that are not visible to external code.
# This allows API authors to create tags on objects they use internally, but are not visible by other parties if they
# Are not publicly accessible.
@doc("A queryable collection of tags.")
@deprecated{Message := "`tag_view` is part of the old tag API. Use the `has_tags` interface instead."}
tag_view<native><public> := interface:
    @doc("Determine if TagToCheck is present in this container, also checking against parent tags \{\"A.1\"\}.Has(\"A\") will return True, \{\"A\"\}.Has(\"A.1\") will return False If TagToCheck is not Valid it will always return False.")
    Has<public>(TagToCheck:tag)<decides><reads>:void

    @doc("Checks if this container contains ANY of the tags in the specified container, also checks against parent tags \{\"A.1\"\}.HasAny(\{\"A\",\"B\"\}) will return True, \{\"A\"\}.HasAny(\{\"A.1\",\"B\"\}) will return False If InTags is empty/invalid it will always return False.")
    HasAny<public>(InTags:[]tag)<decides><reads>:void

    @doc("Checks if this container contains ALL of the tags in the specified container, also checks against parent tags \{\"A.1\",\"B.1\"\}.HasAll(\{\"A\",\"B\"\}) will return True, \{\"A\",\"B\"\}.HasAll(\{\"A.1\",\"B.1\"\}) will return False If InTags is empty/invalid it will always return True, because there were no failed checks.")
    HasAll<public>(InTags:[]tag)<decides><reads>:void
