# Copyright Epic Games, Inc. All Rights Reserved.

using { /Verse.org }
using { /Verse.org/Native }
using { /Verse.org/Simulation }

# Tags are herarchical and can be set up to for classification via inheritance. 
# In the following example any object havinng a `banana_tag` would also be classified as having `fruit_tag`.

# fruit_tag<public> := class(tag){}
# apple_tag<public> := class(fruit_tag){}
# banana_tag<public> := class(fruit_tag){}

# A Verse-authored tag that can be added to objects such as entities
example_tag<public> := class(tag){}
