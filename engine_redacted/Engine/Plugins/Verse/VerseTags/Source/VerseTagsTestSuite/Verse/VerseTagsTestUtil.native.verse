# Copyright Epic Games, Inc. All Rights Reserved.

using { /Verse.org/Native }
using { /Verse.org/Simulation }

#############################################################
# Animal tags
#############################################################

@hide_in_editor
verse_tags_test_tag_animal<native><internal> := class<abstract>(tag){}

@hide_in_editor
verse_tags_test_tag_animal_insect<native><internal> := class<abstract>(verse_tags_test_tag_animal) {}

@hide_in_editor
verse_tags_test_tag_animal_insect_butterfly<native><internal> := class(verse_tags_test_tag_animal_insect) {}

@hide_in_editor
verse_tags_test_tag_animal_insect_wasp<native><internal> := class(verse_tags_test_tag_animal_insect) {}

@hide_in_editor
verse_tags_test_tag_animal_mammal<native><internal> := class<abstract>(verse_tags_test_tag_animal) {}

@hide_in_editor
verse_tags_test_tag_animal_mammal_human<native><internal> := class(verse_tags_test_tag_animal_mammal) {}

@hide_in_editor
verse_tags_test_tag_animal_mammal_horse<native><internal> := class(verse_tags_test_tag_animal_mammal) {}

#############################################################
# Vehcile tags
#############################################################

@hide_in_editor
verse_tags_test_tag_vehicle<native><internal> := class<abstract>(tag){}

@hide_in_editor
verse_tags_test_tag_vehicle_bike<native><internal> := class<abstract>(verse_tags_test_tag_vehicle){}

@hide_in_editor
verse_tags_test_tag_vehicle_bike_ducati<native><internal> := class(verse_tags_test_tag_vehicle_bike){}

@hide_in_editor
verse_tags_test_tag_vehicle_car<native><internal> := class<abstract>(verse_tags_test_tag_vehicle){}

@hide_in_editor
verse_tags_test_tag_vehicle_car_saab<native><internal> := class(verse_tags_test_tag_vehicle_car){}