# Copyright Epic Games, Inc. All Rights Reserved.
# NOTE: _Always_ append to the end of this test script. If you edit the existing test, you must make sure to update the
# dependent test code in `VerseAssistTests`.

GlobalInt:int = 5
GlobalLogic:logic = true
GlobalFloat:float = 0.1
GlobalChar:char = 'c'
GlobalString:string = "This is a string."
GlobalArrayInt:[]int = array{1, 2, 3}

FooClassWithAttribute<public> := class:    
    TestPublicMember<public>:int = 50

ProcTakesIntReturnsVoid(ParamInt:int):void = 
    return 

ProcReturnsInt():int = 
    MyInt := 5
    return 5

ProcReturnsType():type =
    int

ProcTestingProcParamCompletionFiltering():void = 
    ProcTakesIntReturnsVoid(5)
    return 

ProcTestingNoClause():void = 
    return 
    
TestClassFunctionsOnly := class:
    TestClassProc():int = 
        return 5

    TestClassAnotherProc():void=
        return

ProcWithControlFlow(limit:int):int=
    var TestLimit:int = 0
    for(i := 0 .. limit):
        set TestLimit = i

    if (TestLimit < 1):
        return TestLimit
    else:
        return -1

Proc():void=
    return

ProcMutability():float=
    var TestFloat:float = 5.0
    set TestFloat = 6.0
    return TestFloat

TestClass := class:
    var TestInt:int = 5

# Append all new test code you need starting from here and move this comment to the EOF