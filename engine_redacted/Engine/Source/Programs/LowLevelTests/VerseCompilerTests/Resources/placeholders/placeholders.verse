# Copyright Epic Games, Inc. All Rights Reserved.
# NOTE: _Always_ append to the end of this test script. If you edit the existing test, you must make sure to update the
# dependent test code in `VerseAssistTests`.
# This is a copy of the `minimal.verse` test script but includes placeholders. This is in a separate package
# because otherwise placeholders stop compilation from happening successfully.

GlobalInt:int = 5
GlobalLogic:logic = true
GlobalFloat:float = 0.1
GlobalChar:char = 'c'
GlobalString:string = "This is a string."
stub{}

FooClass := class:
    BarInt:int = 500
    stub{}

FooClassWithAttribute<public> := class:
    TestPublicMember<public>:int = 50

ProcTakesIntReturnsVoid(ParamInt:int):void= 
    return 

ProcReturnsInt():int = 
    MyInt := 5
    stub{}
    return 5

ProcTestingProcParamCompletionFiltering():void= 
    ProcTakesIntReturnsVoid(5)
    if (stub{}):
       stub{}
    return

ProcTestingNoClause():void = 
    for (stub{} := stub{} .. stub{})
        stub{}       

    return

GlobalArray:[]int = array{stub{}}

vec2 := class:
    x:int = 0
    y:int = 0

ProcReturnsVec2(v:vec2):stub{}=
    return v
    
# Append all new test code you need starting from here and move this comment to the EOF