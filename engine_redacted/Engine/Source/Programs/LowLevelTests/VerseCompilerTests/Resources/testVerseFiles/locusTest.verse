# Copyright Epic Games, Inc. All Rights Reserved.


BaseModule<public> := module:

    @attribscope_class  
    my_class_attrib := class(attribute):



    derived_class_1 := class(base_class):
        DerivedValue: float

    @my_class_attrib
    derived_class_2 := class(base_class):
        Value_2: int
        DerivedValue: float

