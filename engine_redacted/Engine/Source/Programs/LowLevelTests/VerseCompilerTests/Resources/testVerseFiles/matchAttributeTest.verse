# Copyright Epic Games, Inc. All Rights Reserved.


BaseModule<public> := module:
    @attribscope_class  
    my_class_attrib := class(attribute):

    @attribscope_class  
    unused_attrib := class(attribute):

    TestModule<public> := module:


        @my_class_attrib
