# Copyright Epic Games, Inc. All Rights Reserved.

BaseModule1<public> := module:



    derived_class_1_1 := class(base_class):
        DerivedValue: float

    derived_class_2_1 := class(base_class):
        Value_2: int
        DerivedValue: float

