# Copyright Epic Games, Inc. All Rights Reserved.

BaseModule2<public> := module:



    derived_class_1_2 := class(base_class):
        DerivedValue: float

    derived_class_2_2 := class(base_class):
        Value_2: int
        DerivedValue: float

