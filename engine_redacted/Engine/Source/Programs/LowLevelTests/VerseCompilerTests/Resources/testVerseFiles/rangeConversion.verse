# Copyright Epic Games, Inc. All Rights Reserved.


BaseModule<public> := module:

    @attribscope_class  
    my_class_attrib := class(attribute):

