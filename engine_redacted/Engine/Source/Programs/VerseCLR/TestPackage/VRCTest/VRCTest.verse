# Copyright Epic Games, Inc. All Rights Reserved.

using { /Verse.org/Native }

test_class := struct:
    i : int = 0

# 0 arg
Test0Void():void = 
    return

Test0RInt():int = 
    return 42

Test0RFloat():float = 
    return 42.0

Test0RString():string = 
    return "42"

Test0RLogicTrue():logic = 
    return true

# 1 arg
Test1RInt(V:int):int = 
    return V

Test1RFloat(V:float):float = 
    return V

Test1RString(V:string):string = 
    return V

Test1RLogic(V:int):logic = 
    return true

Test1IVoid(V:int):void = 
    return

Test1FVoid(V:float):void = 
    return

Test1SVoid(V:string):void = 
    return

Test1VVoid():void = 
    return

# 2 arg
Test2IFInt(V:int, F:int):int = 
    return V + F

Test2ISString(V:int, S:string):string = 
    return ToString(V) + S

Test2VSVoid(V:int, S:string):void = 
    return

# 3 arg
Test3RString(V:int, F:float, S:string):string = 
    return ToString(V) + " " + S

Test3RFloat(F:float, S:string, V:int):float = 
    return F

Test3RLogic(F:float, V:int, S:string):logic = 
    return true

Test3RVoid(V:float, F:float, S:string):void = 
    return

# nan and minus 1
TestReturningNegativeOne():int = 
    return -1

TestReturningNegativeTwo():int = 
    return -2

TestReturningMaxInt():int = 
    return 9223372036854775807

TestReturningLargeFloat():float = 
    return 9223372036854775807.0

TestReturnUknownPrint():test_class = 
    tc : test_class = test_class{}
    return tc

Main():int = 
    return 42

TestSum(X:int, Y:int):int = return X+Y

Fac(X:int):int =
  if (X=0):
    return 1
  else:
    return X*Fac(X-1)

M1 := module:
    MFun<public>(X:int):int = return X*10+1

M2 := module:
    MFun(X:int):int = return X*100+10    

GFun(X:int):int = return M1.MFun(X)

# Add some localization messages to make it useful as a simple test project for VerseCompilerCmd:
#    VerseCompilerCmd VRCTest.vproject -plugin=VerseLocalizationGen

# At least one example for each supported form.
# 1) TheMsg<localizes> := "The Message"
# Not supported

# 2) TheMsg<localizes> : message = "The Message"
Form2a<localizes> : message = "The Message2"
Form2b<localizes> : message = "The {0u004d}essage"
Form2c<localizes> : message = "The \{0u004d\}essage \\r\n\t\"\'\#\<\>\&\~"

# 3) TheMsg<localizes>(Name:string) := "The Message"
# Not supported

# 4) TheMsg<localizes>(Name:string) : message = "The Message"
Form4a<localizes>(Name:string) : message = "The Message"

# 5) TheMsg<localizes>(Name:string) := "The Message to {Name}"
# Not supported

# 6) TheMsg<localizes>(Name:string) : message = "The Message to {Name}"
Form6a<localizes>(Name:string) : message = "The Message to {Name}"
Form6b<localizes>(Name:string) : message = "The {0u004d}essage to {Name} \\\r\n\t\"\'\#\<\>\&\~"
Form6c<localizes>(Name:string) : message = "The \{0u004d\}essage to {Name}"
Form6d<localizes>(Name:string, Number:int) : message = "Hey {Name}, your number is {Number}"
Form6e<localizes>(Name:string, Number:int) : message = "Hey {Number} {Name}, your number is {Number} and your name is {Name}"
