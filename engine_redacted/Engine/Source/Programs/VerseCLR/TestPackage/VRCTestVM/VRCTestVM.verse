# Copyright Epic Games, Inc. All Rights Reserved.

# Starting with something simple
Main():int = 
    42

# Testing if expresssion
IfZero(X:int):int =
    R := if (X = 0) { 10 } else { 20 }
    return R

IfPos(X:int):int =
    R := if (X > 0) { 10 } else { 20 }
    R

# Recursion
Fac(X:int)<computes>:int =
  if (X=0):
    1
  else:
    X*Fac(X-1)

# Modules
M1 := module:
    MFun<public>(X:int):int = X*10+1

M2 := module:
    MFun(X:int):int = X*100+10    

GFun(X:int):int = M1.MFun(X)

# Tuples as arguments
ShiftAdd(A:int, B:int)<computes>:int = A*10+B
Swap(A:int, B:int):tuple(int, int) = (B,A)

ShiftAddWithSwap(A:int, B:int):int = ShiftAdd(Swap(A, B))

ShiftAddWithSwap2(A:int, B:int):int =
    T := (A,B)
    ShiftAdd(Swap(T))

# Can't be called from command line since VerseCLR doesn't handle tuples as arguments
ShiftAdd((A:int,B:int), C:int):int = ShiftAdd(ShiftAdd(A,B), C)
ShiftAdd(A:int, B:int, C:int):int = ShiftAdd((A,B), C)

# Array
CreateArray(X:int)<computes>:[]int =
    if (X>0):
        array{X} + CreateArray(X-1)
    else:
        array{}

IndexArray(Size:int, Index:int)<computes>:int =
    A := CreateArray(Size)

# Earlier versions of the compiler created code that built an incorrect array:
# A = [1+3, 1+4, 2+3, 2+4] 
# instead of the expected:
# A = [[1+3,1+4], [2+3, 2+4]]
WasIncorrectArrayAdd()<decides><transacts>:int = 
    A:[][]int = for (I:array{1,2}) . for (J:array{3,4}) . (I+J)
    A.Length = 2
    A[0][0] = 1+3
    A[0][1] = 1+4
    A[1][0] = 2+3
    A[1][1] = 2+4
    A.Length

# for
ForArrayEmpty()<transacts><decides>:int = 
    A:[]int = for (I:array{}) . I
    A.Length = 0
    return A.Length

ForArrayOne()<transacts><decides>:int = 
    A:[]int = for (I:array{1}) . I
    A.Length = 1
    A[0] = 1
    return A.Length

ForArrayThree()<transacts><decides>:int = 
    A:[]int = for (V:array{10,20,30}) . V
    A.Length = 3
    A[0] = 10
    A[1] = 20
    A[2] = 30
    return 3 # A.Length

ForArrayThreeIndicies()<transacts><decides>:int = 
    A:[]int = for (I->V:array{10,20,30}) . I
    A.Length = 3
    A[0] = 0
    A[1] = 1
    A[2] = 2
    return 3 # A.Length

#for on maps
ForMapEmptyKeys()<transacts><decides>:int = 
    A:[]int = for(K->V:map{}) . K
    A.Length = 0
    return A.Length   

ForMapEmptyValues()<transacts><decides>:int = 
    A:[]int = for(V:map{}) . V
    A.Length = 0
    return A.Length 

ForMapEmptyValues2()<transacts><decides>:int = 
    A:[]int = for(K->V:map{}) . V
    A.Length = 0
    return A.Length 

ForMapThreeKeys()<transacts><decides>:int = 
    A:[]int = for(K->V:map{1=>10, 2=>20, 3=>30}) . K
    A.Length = 3
    A[0] = 1
    A[1] = 2
    A[2] = 3
    return A.Length   

ForMapThreeValues()<transacts><decides>:int = 
    A:[]int = for(V:map{1=>10, 2=>20, 3=>30}) . V
    A.Length = 3
    A[0] = 10
    A[1] = 20
    A[2] = 30
    return A.Length 

ForMapThreeValues2()<transacts><decides>:int = 
    A:[]int = for(K->V:map{1=>10, 2=>20, 3=>30}) . V
    A.Length = 3
    A[0] = 10
    A[1] = 20
    A[2] = 30
    A.Length

# + doesn't work for maps 
#CreateMap(X:int)<computes>:[int]int =
#    if (X>0):
#        map{ X => 100+X} + CreateMap(X-1)
#    else:
#        map{}

AssignAdd(X:int, Y:int)<transacts>:int =
    var V : int = X
    set V += Y
    V

AssignSub(X:int, Y:int)<transacts>:int =
    var V : int = X
    set V -= Y
    V

AssignMul(X:int, Y:int)<transacts>:int =
    var V : int = X
    set V *= Y
    V

AssignDiv(X:float, Y:float)<transacts>:float =
    var V : float = X
    set V /= Y
    V

# Some slightly more complicated for
ForFilter(X:int)<transacts>:int = 
    A:[]int = for (I:array{10,20,30}, I <> X) . I
    A.Length

ForForFilterFilter(X:int, Y:int)<transacts>:int =
    A:[]int = for (I:array{10,20,30}, I <> X, J:array{10,20}, J <> Y) . (I + J)
    A.Length

ForFailBody(Size:int, Fail:int)<transacts><decides>:int =
    Array := CreateArray(Size)
    Array2 := for(X:Array):
                Fail = 0
                X
    Array2.Length

WrapForFailBody(Size:int, Fail:int)<transacts>:int =
    if (L:=ForFailBody[Size, Fail]):
        return L
    else:
        return -42


ForForFailBody(NumX:int, NumY:int, Fail:int):int =
    Xs := CreateArray(NumX)
    Ys := CreateArray(NumY)
    if (XYs := for (X:Xs, Y:Ys) { XY := X*10 + Y; XY <> Fail; XY }):
        XYs.Length
    else:
        -42
  

AddTuplePair(X:tuple(int, int))<computes>:int =
    return X(0) + X(1)

NestedDouble((X:int, Y:int))<computes>:int =
    return X + Y

# this isn't actually a tuple, but this is checking such syntax that it isn't
NestedSingle((X:int))<computes>:int =
    return X


Sign(X:int)<transacts>:int =
    if (X < 0):
        return -1
    if (X > 0):
        return 1
    0

FacWithVar(X:int)<transacts>:int =
    var F:int = 1
    var N:int = X
    loop:
        if (N <= 0):
            break;
        set F = F * N
        set N -= 1
    F

SumArray(X:int)<transacts><decides>:int =
    Array:[]int = CreateArray(X)
    var S:int = 0
    for (V:Array) . set S += V
    S+S = X*(X+1)
    S

ArrayWithVar(X:int)<transacts><decides>:int =
    var Array:[]int = array{}
    var N:int = X
    loop:
        if (N <= 0):
            break;
        set Array += array{N}
        set N -= 1
    (X < 0 or Array.Length = X) 
    (X < 1 or (Array[0] = X and Array[X-1] = 1))
    Array.Length
 
# Just a helper to make testing easier. Returns 1 when tests pass.
RunAllTests()<transacts>:int =
    ArgTuple2 := (42, 7)
    if:
        AssignAdd(40,2) = 42
        AssignSub(42,2) = 40
        AssignMul(21,2) = 42
        AssignDiv(84.0, 20.0) = 4.2

        CreateArray(10).Length = 10

        CreateArray(0).Length = 0

        CreateArray(10)[0] = 10
        CreateArray(10)[1] = 9
        IndexArray(10, 0) = 110
        IndexArray(10, 10) = 42
        IndexArray(10, -1) = 42
        IndexArray(10, 100) = 42

        WasIncorrectArrayAdd[] = 2

        ForArrayEmpty[] = 0
        ForArrayOne[] = 1
        ForArrayThree[] = 3
        ForArrayThreeIndicies[] = 3

        ForMapEmptyKeys[] = 0
        ForMapEmptyValues[] = 0               
        ForMapEmptyValues2[] = 0

        ForMapThreeKeys[] = 3
        ForMapThreeValues[] = 3               
        ForMapThreeValues2[] = 3

        ForFilter(0) = 3
        ForFilter(20) = 2

        ForForFilterFilter(20, 30) = 4

        WrapForFailBody(5, 0) = 5
        WrapForFailBody(5, 1) = -42

        Fac(7) = 5040
        FacWithVar(7) = 5040

        ShiftAdd(10, 5) = 105

        NestedDouble(10, 20) = 30
        NestedDouble((42, 11)) = 53
        NestedDouble(ArgTuple2) = 49

        AddTuplePair(ArgTuple2) = 49
        AddTuplePair(10, 11) = 21
        AddTuplePair(666, 111) = 777

        NestedSingle(666) = 666
        
        true? # support "true" and "?" operator
        Sign(-5) = -1
        Sign(0) = 0
        Sign(42) = 1

        SumArray[3] = 6
        ArrayWithVar[5] = 5
    then:
        1
    else:
        0
