# Copyright Epic Games, Inc. All Rights Reserved.

using {/Verse.org/Native}

Print<public><native>(:string)<converges>:void

DebugBreak<public><native>()<converges>:void

NextTick<public><native>()<suspends>:void

create_field_test_class<public><native> := class:
    A<public>:int
    B<public>:int
    C<public>:int
