# Copyright Epic Games, Inc. All Rights Reserved.

# No extra args: just the script path.
# <versecmd-test>
# name = "no_args"
# [expect]
# stdout_matches = ["len=1", "path=.*\\.verse"]
# </versecmd-test>
#
# Extra args: length includes script path, arg values accessible by index.
# <versecmd-test>
# name = "with_args"
# args = ["hello", "world"]
# [expect]
# stdout_matches = ["len=3", "arg1=hello", "arg2=world"]
# </versecmd-test>
#
# Args with special characters (spaces, punctuation).
# <versecmd-test>
# name = "special_chars"
# args = ["hello world", "it's"]
# [expect]
# stdout_matches = ["arg1=hello world", "arg2=it's"]
# </versecmd-test>

PrintArgs(Args:[]string)<suspends>:void =
    Print("len={Args.Length}")

Main(Args:[]string, Env:[string]string)<suspends>:void=
    PrintArgs(Args)
