# Copyright Epic Games, Inc. All Rights Reserved.

# Single environment variable.
# <versecmd-test>
# name = "single"
# [env]
# VERSECMD_TEST_A = "alpha"
# [expect]
# stdout_matches = ["A=alpha", "missing=ok"]
# </versecmd-test>
#
# Multiple environment variables.
# <versecmd-test>
# name = "multiple"
# [env]
# VERSECMD_TEST_A = "alpha"
# VERSECMD_TEST_B = "bravo"
# [expect]
# stdout_matches = ["A=alpha", "B=bravo"]
# </versecmd-test>
#
# No env vars set: lookups fail gracefully.
# <versecmd-test>
# name = "missing"
# [expect]
# stdout_matches = ["^missing=ok$"]
# </versecmd-test>
#
# Env var with empty string value.
# <versecmd-test>
# name = "empty_value"
# [env]
# VERSECMD_TEST_A = ""
# [expect]
# stdout_matches = ["A=\n", "missing=ok"]
# </versecmd-test>

PrintEnv(Env:[string]string)<suspends>:void =
    if (A := Env["VERSECMD_TEST_A"]):
        Print("A={A}")
    if (B := Env["VERSECMD_TEST_B"]):
        Print("B={B}")
    if (Env["VERSECMD_NONEXISTENT_XYZ_98765"]):
        Print("unexpected")
    else:
        Print("missing=ok")

Main(Args:[]string, Env:[string]string)<suspends>:void=
    PrintEnv(Env)
