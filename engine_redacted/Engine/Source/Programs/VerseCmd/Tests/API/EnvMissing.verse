# Copyright Epic Games, Inc. All Rights Reserved.

# <versecmd-test>
# [expect]
# stdout_matches = ["^missing$"]
# </versecmd-test>

Main(Args:[]string, Env:[string]string)<suspends>:void=
    if (Env["VERSECMD_NONEXISTENT_VAR_XYZ_12345"]):
        Print("found")
    else:
        Print("missing")
