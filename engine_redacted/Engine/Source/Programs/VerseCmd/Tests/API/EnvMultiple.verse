# Copyright Epic Games, Inc. All Rights Reserved.

# <versecmd-test>
# [env]
# VERSE_A = "alpha"
# VERSE_B = "bravo"
# [expect]
# stdout_matches = ["^alpha bravo$"]
# </versecmd-test>

Main(Args:[]string, Env:[string]string)<suspends>:void=
    if (A := Env["VERSE_A"]; B := Env["VERSE_B"]):
        Print("{A} {B}")
    else:
        Exit(1)
