# Copyright Epic Games, Inc. All Rights Reserved.

# No args: script completes normally without calling Exit.
# <versecmd-test>
# name = "implicit_zero"
# [expect]
# stdout_matches = ["^done$"]
# </versecmd-test>
#
# Exit(0): explicit zero exit code.
# <versecmd-test>
# name = "explicit_zero"
# args = ["exit0"]
# </versecmd-test>
#
# Exit(42): nonzero exit code.
# <versecmd-test>
# name = "nonzero"
# args = ["exit42"]
# [expect]
# exit = 42
# </versecmd-test>
#
# Exit(): zero-argument overload.
# <versecmd-test>
# name = "zero_arg"
# args = ["zero_arg"]
# </versecmd-test>
#
# Print output before Exit is still delivered.
# <versecmd-test>
# name = "print_before_exit"
# args = ["print_exit"]
# [expect]
# exit = 7
# stdout_matches = ["^visible$"]
# </versecmd-test>

Main(Args:[]string, Env:[string]string)<suspends>:void=
    if (Args.Length < 2):
        Print("done")
    else if (Cmd := Args[1]):
        case (Cmd):
            "exit0" => Exit(0)
            "exit42" => Exit(42)
            "zero_arg" => Exit()
            "print_exit" =>
                Print("visible")
                Exit(7)
            _ => {}