# Copyright Epic Games, Inc. All Rights Reserved.

# <versecmd-test>
# [expect]
# exit = 0
# </versecmd-test>

Main(Args:[]string, Env:[string]string)<suspends>:void=
    Exit()
