# Copyright Epic Games, Inc. All Rights Reserved.

# <versecmd-test>
# [expect]
# stdout_matches = ["^T1: [0-9.]+\nT2: [0-9.]+$"]
# </versecmd-test>

# Get timestamp twice in separate transactions; verify monotonicity and reasonable range (after 2020-01-01)
Main(Args:[]string, Env:[string]string)<suspends>:void=
    T1 := GetSecondsSinceEpoch()
    Yield()
    T2 := GetSecondsSinceEpoch()
    
    if (T1 > 1577836800.0):
        Print("T1: {T1}")
    else:
        Err("GetSecondsSinceEpoch returned a time before 2020. Keep dreaming, GetSecondsSinceEpoch.")

    if (T2 >= T1):
        Print("T2: {T2}")
    else:
        Err("GetSecondsSinceEpoch was non-monotonically increasing.")