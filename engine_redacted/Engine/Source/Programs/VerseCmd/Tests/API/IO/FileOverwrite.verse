# Copyright Epic Games, Inc. All Rights Reserved.

# <versecmd-test>
# [expect]
# stdout_matches = ["^second$"]
# files_exist = ["data.txt"]
# files_not_exist = ["phantom.txt"]
# [expect.files_match]
# "data.txt" = ["^second$"]
# </versecmd-test>

Main(Args:[]string, Env:[string]string)<suspends>:void=
    SaveTextFile("data.txt", "first")
    SaveTextFile("data.txt", "second")
    LoadResult := LoadTextFile("data.txt")
    if (Content := LoadResult.GetSuccess[]):
        Print(Content)
    else:
        Exit(1)
