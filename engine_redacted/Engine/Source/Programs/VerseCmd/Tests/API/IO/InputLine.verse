# Copyright Epic Games, Inc. All Rights Reserved.

# <versecmd-test>
# name = "read_line"
# stdin = "hello\n"
# [expect]
# stdout_matches = ["^You said: hello$"]
# </versecmd-test>
#
# <versecmd-test>
# name = "eof"
# stdin = ""
# [expect]
# stdout_matches = ["^Got EOF$"]
# </versecmd-test>

Main(Args:[]string, Env:[string]string)<suspends>:void=
    MaybeLine := InputLine()
    if (Line := MaybeLine?):
        Print("You said: {Line}")
    else:
        Print("Got EOF")
