# Copyright Epic Games, Inc. All Rights Reserved.

# Loads a file whose name is given as the first argument and prints its contents.
# Different invocations test success, not-found, and empty-file edge cases.

# <versecmd-test>
# name = "success"
# args = ["input.txt"]
# [setup_files]
# "input.txt" = "Hello from Verse"
# [expect]
# stdout_matches = ["^Hello from Verse$"]
# </versecmd-test>
#
# <versecmd-test>
# name = "not_found"
# args = ["nonexistent.txt"]
# [expect]
# stdout_matches = ["^Error loading file$"]
# </versecmd-test>
#
# <versecmd-test>
# name = "empty_file"
# args = ["empty.txt"]
# [setup_files]
# "empty.txt" = ""
# [expect]
# stdout_matches = ["^Loaded 0 characters$"]
# </versecmd-test>

Main(Args:[]string, Env:[string]string)<suspends>:void=
    if (Filename := Args[1]):
        Result := LoadTextFile(Filename)
        if (Content := Result.GetSuccess[]):
            if (Content.Length > 0):
                Print(Content)
            else:
                Print("Loaded {Content.Length} characters")
        else:
            Print("Error loading file")
    else:
        Exit(1)
