# Copyright Epic Games, Inc. All Rights Reserved.

# <versecmd-test>
# stdin = "alpha\nbeta\n"
# [expect]
# stdout_matches = ["^alpha\nbeta\neof$"]
# </versecmd-test>

ReadAll()<suspends>:void =
    loop:
        MaybeLine := InputLine()
        if (Line := MaybeLine?):
            Print(Line)
        else:
            break
    Print("eof")

Main(Args:[]string, Env:[string]string)<suspends>:void=
    ReadAll()
