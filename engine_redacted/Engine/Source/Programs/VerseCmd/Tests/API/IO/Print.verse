# Copyright Epic Games, Inc. All Rights Reserved.

# <versecmd-test>
# [expect]
# stdout_matches = ["^Hello, World!$"]
# </versecmd-test>

Main(Args:[]string, Env:[string]string)<suspends>:void=
    Print("Hello, World!")
