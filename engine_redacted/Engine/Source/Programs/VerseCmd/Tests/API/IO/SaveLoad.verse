# Copyright Epic Games, Inc. All Rights Reserved.

# <versecmd-test>
# [expect]
# stdout_matches = ["^Hello from Verse$"]
# files_exist = ["test_output.txt"]
# [expect.files_match]
# "test_output.txt" = ["^Hello from Verse$"]
# </versecmd-test>

Main(Args:[]string, Env:[string]string)<suspends>:void=
    SaveResult := SaveTextFile("test_output.txt", "Hello from Verse")
    if (SaveResult.GetError[]):
        Exit(1)
    LoadResult := LoadTextFile("test_output.txt")
    if (Content := LoadResult.GetSuccess[]):
        Print(Content)
    else:
        Exit(1)
