# Copyright Epic Games, Inc. All Rights Reserved.

# <versecmd-test>
# [expect]
# stdout_matches = ["^0$"]
# files_exist = ["empty.txt"]
# </versecmd-test>

Main(Args:[]string, Env:[string]string)<suspends>:void=
    SaveResult := SaveTextFile("empty.txt", "")
    if (SaveResult.GetError[]):
        Exit(1)
    LoadResult := LoadTextFile("empty.txt")
    if (Content := LoadResult.GetSuccess[]):
        Print("{Content.Length}")
    else:
        Exit(1)
