# Copyright Epic Games, Inc. All Rights Reserved.

using { /Verse.org/Random }

# <versecmd-test>
# [expect]
# stdout_matches = ["alpha", "bravo", "charlie", "int=[1-9][0-9]\n", "float=0\\."]
# </versecmd-test>

# Shuffle preserves all elements (order will vary)
PrintAll(Arr:[]string)<suspends>:void =
    for (W : Arr):
        Print(W)

Main(Args:[]string, Env:[string]string)<suspends>:void=
    Words := array{"alpha", "bravo", "charlie"}
    Shuffled := Shuffle(Words)
    PrintAll(Shuffled)

    # GetRandomInt in [10, 99] — always two digits
    Print("int={GetRandomInt(10, 99)}")

    # GetRandomFloat in [0.0, 1.0) — starts with "0."
    Print("float={GetRandomFloat(0.0, 1.0)}")
