# Copyright Epic Games, Inc. All Rights Reserved.

# <versecmd-test>
# args = ["alpha", "bravo", "charlie"]
# [expect]
# stdout_matches = ["^3 args saved$"]
# files_exist = ["args.txt"]
# [expect.files_match]
# "args.txt" = ["alpha", "bravo", "charlie"]
# </versecmd-test>

Main(Args:[]string, Env:[string]string)<suspends>:void=
    # Skip Args[0] (script path)
    Content := Join(UserArgs, "\n")
    SaveResult := SaveTextFile("args.txt", Content)
    if (SaveResult.GetError[]):
        Exit(1)
    Print("{UserArgs.Length} args saved")
