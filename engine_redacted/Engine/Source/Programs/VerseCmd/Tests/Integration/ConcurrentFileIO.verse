# Copyright Epic Games, Inc. All Rights Reserved.

# <versecmd-test>
# [expect]
# stdout_matches = ["^alpha bravo$"]
# files_exist = ["a.txt", "b.txt"]
# </versecmd-test>


Main(Args:[]string, Env:[string]string)<suspends>:void=
