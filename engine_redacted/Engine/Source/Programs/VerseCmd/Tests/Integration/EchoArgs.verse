# Copyright Epic Games, Inc. All Rights Reserved.

# <versecmd-test>
# args = ["hello", "world", "verse"]
# [expect]
# stdout_matches = ["^hello\nworld\nverse$"]
# </versecmd-test>

# Skip Args[0] which is the script path
PrintArgs(Args:[]string)<suspends>:void =
        Print(Arg)

Main(Args:[]string, Env:[string]string)<suspends>:void=
    PrintArgs(Args)
