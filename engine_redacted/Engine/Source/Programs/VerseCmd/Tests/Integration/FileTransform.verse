# Copyright Epic Games, Inc. All Rights Reserved.

# <versecmd-test>
# [setup_files]
# "input.txt" = "Hello World"
# [expect]
# stdout_matches = ["^Transformed 11 characters$"]
# files_exist = ["output.txt"]
# [expect.files_match]
# "output.txt" = ["^hello world$"]
# </versecmd-test>

ToLower(C:char):char =
    case(C):
        'A'=>'a'
        'B'=>'b'
        'C'=>'c'
        'D'=>'d'
        'E'=>'e'
        'F'=>'f'
        'G'=>'g'
        'H'=>'h'
        'I'=>'i'
        'J'=>'j'
        'K'=>'k'
        'L'=>'l'
        'M'=>'m'
        'N'=>'n'
        'O'=>'o'
        'P'=>'p'
        'Q'=>'q'
        'R'=>'r'
        'S'=>'s'
        'T'=>'t'
        'U'=>'u'
        'V'=>'v'
        'W'=>'w'
        'X'=>'x'
        'Y'=>'y'
        'Z'=>'z'
        _=>C
ToLower(S:string):string = for(C:S). ToLower(C)

Main(Args:[]string, Env:[string]string)<suspends>:void=
    LoadResult := LoadTextFile("input.txt")
    if (Content := LoadResult.GetSuccess[]):
        Transformed := ToLower(Content)
        SaveResult := SaveTextFile("output.txt", Transformed)
        if (SaveResult.GetError[]):
            Exit(1)
        Print("Transformed {Content.Length} characters")
    else:
        Exit(1)
