# Copyright Epic Games, Inc. All Rights Reserved.

# <versecmd-test>
# stdin = "line one\nline two\nline three\n"
# [expect]
# stdout_matches = ["^3 lines$"]
# files_exist = ["output.txt"]
# [expect.files_match]
# "output.txt" = ["line one", "line two", "line three"]
# </versecmd-test>

ReadAll()<suspends>:[]string =
    var Acc:[]string = array{}
    loop:
        MaybeLine := InputLine()
        if (Line := MaybeLine?):
            set Acc = Acc + array{Line}
        else:
            break
    Acc

Main(Args:[]string, Env:[string]string)<suspends>:void=
    Lines := ReadAll()
    Content := Join(Lines, "\n")
    SaveResult := SaveTextFile("output.txt", Content)
    if (SaveResult.GetError[]):
        Exit(1)
    Print("{Lines.Length} lines")
