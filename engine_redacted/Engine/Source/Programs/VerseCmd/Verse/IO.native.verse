# Copyright Epic Games, Inc. All Rights Reserved.

using {/Verse.org/Native}

io_error<native><public> := class{}
access_error<native> := class(io_error){}
not_found_error<native> := class(io_error){}

# File I/O
LoadTextFile<native><public>(Filename:string)<suspends>:result(string, io_error)
SaveTextFile<native><public>(Filename:string, Contents:string)<suspends>:result(void, io_error)

# Console I/O
Print<public><native>(:string)<suspends>:void
InputLine<native><public>()<suspends>:?string

# Process control
Yield<public><native>()<suspends>:void
Exit<native><public>(ExitCode:int)<suspends>:false
Exit<public>()<suspends>:false = Exit(0)
