# Copyright Epic Games, Inc. All Rights Reserved.

using {/Verse.org/Native }




texture<native><public> := class<computes><final><public>(redacted_internal, redacted_internal, redacted_internal):
