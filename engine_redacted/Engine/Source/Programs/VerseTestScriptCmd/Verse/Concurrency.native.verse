# Copyright Epic Games, Inc. All Rights Reserved.

using { /Verse.org/Concurrency }
using { /Verse.org/Tests/VerseTestScriptCmd }

test_native_defers<public><native> := class<abstract>:
    TaskA<public><native_callable>()<suspends>:void
    SpawnTask<public><native>():void
    CancelTask<public>()<suspends>:void=
        CancelTask_Native()
    CancelTask_Native<public><native>():void

CancelAndForget<public><native>(Task:task(t) where t:type):void
