# Copyright Epic Games, Inc. All Rights Reserved.

# Coroutine related utility functionality used in VerseTestScriptCmd
CoroUtils<public> := module:

    # Resets coroutine system, labels etc.
    Reset<native><public>():void

    # Prevents the next call to Reset() from clearing the event log string
    SuppressNextReset<native><public>():void

    # Returns the total number of currently active coroutines
    GetNumActiveCoroutines<native><public>():int

    # Coroutine that waits for the specified number of ticks
    WaitTicks<native><public>(numberOfTicks:int)<suspends><predicts>:void

    # Returns true if we have an active content scope
    HasActiveContentScope<native><public>()<transacts>:logic

    # Add a label to the event log string
    LogEvent<native><public>(text:string)<transacts><predicts>:void

    # Get a formatted string representing all event labels
    # It Contains all concatenated label strings plus a comma for each tick that occurred
    GetEventLogString<native><public>()<transacts><predicts>:string

    # Add a label to the event log string, but wipe the event log string before
    ResetThenLogEvent<native><public>(text:string)<transacts>:void

    # A version of (event:)Signal without <no_rollback>.
    SignalTransacts<public><native>(Event:event(t), Val:t where t:type)<transacts>:void

