# Copyright Epic Games, Inc. All Rights Reserved.

FloatConstants<public> := module:
    DoubleMax<public>:float = 1.7976931348623158e+308
    PosInfty<public>:float = 1.0 / 0.0 # One way to get +Infinity
    SmallestNormal<public>:float = 2.2250738585072014e-308
    SmallestNormalTimesOneHalf<public>:float = SmallestNormal * 1.5
    LargestSubnormal<public>:float = 2.225073858507201e-308
    HalfSmallestNormal<public>:float = 1.1125369292536007e-308
    JustUnderHalfSmallestNormal<public>:float = 1.1125369292536e-308
    SmallestSubnormal<public>:float = 5e-324
    Pow2_52<public>:float = 4503599627370496.0 # Exact
    Pow2_Minus28<public>:float = 1.0 / 268435456.0 # Exact
    Pow2_Minus25<public>:float = 8.0 * Pow2_Minus28 # Exact
        
