# Copyright Epic Games, Inc. All Rights Reserved.

FunctionUtils<public> := module:
    InvokeComputesReads<public><native>(F(:t)<computes><reads>:u, :t where t:type, u:type)<computes><reads>:u
