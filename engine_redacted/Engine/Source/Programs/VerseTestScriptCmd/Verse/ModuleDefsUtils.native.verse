# Copyright Epic Games, Inc. All Rights Reserved.

moduleFunc_default(i:int):int=return(i*i)
moduleFunc_public<public>(i:int):int=return(i)
moduleFunc_internal<internal>(i:int):int=return(-i)

moduleData_default:int=42
moduleData_public<public>:int=1337
moduleData_internal<internal>:int=8675
# Module-scoped pointers are not yet supported.
# var moduleData_mutable<public>:int=309
