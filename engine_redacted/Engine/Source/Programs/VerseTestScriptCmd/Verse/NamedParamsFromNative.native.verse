# Copyright Epic Games, Inc. All Rights Reserved.

InterfaceWithNamedParams<public> := module:
    test_interface<public><native> := interface:
        TestFunction<public><native_callable>(A:int, B:int, ?C:int = 2, ?D:int = 3)<decides>:void
    
    CallWithABC<public><native>(Interface:test_interface, A:int, B:int, ?C:int = 3, ?D:int = 4)<decides>:void

