# Copyright Epic Games, Inc. All Rights Reserved.

NativeIntLess<public><native>(X:int, Y:int)<computes><decides>:void
NativeAlwaysLess<public><native>(X:any, Y:any)<computes><decides>:void
NativeNeverLess<public><native>(X:any, Y:any)<computes><decides>:void
NativeErrLess<public><native>(X:any, Y:any)<computes><decides>:void
