# Copyright Epic Games, Inc. All Rights Reserved.

Joined<public><native>(:string, :persistent_key, :any)<decides>:void
Dropped<public><native>(:persistent_key):void
RecordUpdates<public><native>():void
TakeRecordedUpdates<public><native>():[]tuple(string, string, persistent_key, comparable)
GetOrCreateKey<public><native>(:int)<allocates>:persistent_key
# Second argument must be a class
FromJsonV1<native><public>(:string, t:type)<decides><computes>:t
