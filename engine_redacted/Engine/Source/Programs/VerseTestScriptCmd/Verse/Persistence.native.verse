# Copyright Epic Games, Inc. All Rights Reserved.

persistent_key<public><native> := class<persistent><module_scoped_var_weak_map_key><unique><computes><allocates>:
    SetValid<public><native>(:logic):void

Persistence<public> := module{}
