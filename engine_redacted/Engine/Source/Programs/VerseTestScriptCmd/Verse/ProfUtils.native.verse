# Copyright Epic Games, Inc. All Rights Reserved.

# Profile related utility functionality used in VerseTestScriptCmd
ProfUtils<public> := module:
    GetNumActiveProfiles<native><public>()<reads>:int
    GetNumStartedProfiles<native><public>()<reads>:int
    GetNumRetiredProfiles<native><public>()<reads>:int
    GetUserTagStr<native><public>()<reads>:string
    GetLastProfileFile<native><public>()<reads>:string
    GetLastProfileRow<native><public>()<reads>:int

    # don't call these directly - even from test scenarios
    Register<native>()<transacts>:void
	Unregister<native>()<transacts>:void
    Reset<native>()<transacts>:void;
