# Copyright Epic Games, Inc. All Rights Reserved.

session_key<public><native> := class<module_scoped_var_weak_map_key><unique><computes><allocates> {}

Session<public> := module:
    Reset<public><native>():void
    GetOrCreateKey<public><native>(:int)<allocates>:session_key
