# Copyright Epic Games, Inc. All Rights Reserved.


test_value<public> := class:
        var X<public>:int = 42



# No _defun_XXXXX function will be created for this Value since it has accessors


# Easy as in here we have all information about the interface_with_accessors
no_defun_easy<public> := class(interface_with_accessors):






