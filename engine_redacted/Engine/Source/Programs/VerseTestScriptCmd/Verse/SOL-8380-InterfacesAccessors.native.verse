# Copyright Epic Games, Inc. All Rights Reserved.

SOL8380_class_only_accessors<public><native> := class():
    GetValueNative<private><native>()<transacts>:int
    SetValueNative<private><native>(NewValue:int)<transacts>:void


SOL8380_class_direct_accessors<public><native> := class(SOL8380_interface):
    GetValueNative<private><native>()<transacts>:int
    SetValueNative<private><native>(NewValue:int)<transacts>:void

SOL8380_class_direct<public><native> := class(SOL8380_interface):

SOL8380_class_indirect_accessors<public><native> := class(SOL8380_class_direct):
    GetValueNative<private><native>()<transacts>:int
    SetValueNative<private><native>(NewValue:int)<transacts>:void


