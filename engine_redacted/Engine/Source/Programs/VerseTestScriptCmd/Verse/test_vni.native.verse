# Copyright Epic Games, Inc. All Rights Reserved.

using { /Verse.org/Concurrency }
using { /Verse.org/Tests/VerseTestScriptCmd }

############################################
# Helper functions
Log(Msg:string)           : void   = {CoroUtils.LogEvent(Msg)}
GetLogStr()               : string = {return CoroUtils.GetEventLogString()}
Wait(Ticks:int)<suspends> : void   = {CoroUtils.WaitTicks(Ticks)}

#--------------------------------------------------------------------------------------------------
# Structs used for native tuple testing

tuple_struct<public><native> := struct:
    Elem0<public><native>:int

    Elem1<public><native>:float

tuple_struct_nested<public><native> := struct:
    Elem0<public><native>:int

    Elem1<public><native>:tuple_struct

nonnative_struct<public> := struct:
    Elem0<public>:int

    Elem1<public>:float

# Simple types to test whether they are properly included in VNI generated `struct_subnested`

struct_a<public><native> := struct:
    Elem0<public><native>:int

struct_b<public><native> := struct:
    Elem0<public><native>:int

struct_c<public><native> := struct:
    Elem0<public><native>:int

struct_d<public><native> := struct:
    Elem0<public><native>:int

struct_e<public><native> := struct:
    Elem0<public><native>:int

struct_f<public><native> := struct:
    Elem0<public><native>:int

class_a<public><native> := class:
    Elem0<public><native>:struct_f
    Elem1<public><native>:tuple(int, float)

interface_a<public><native> := interface:
    ConstArray<public>:[]int
    var VarArray<public>:[]int

class_a_interface_a<public><native> := class(class_a, interface_a):
    ConstArray<override><native>:[]int = array{1,2,3}

    var VarArray<override><native>:[]int = array{4,5,6}

    Elem0<override><native>:struct_f = struct_f{Elem0:=7}
    Elem1<override><native>:tuple(int, float) = (8, 9.0)


fruits<public><native> := enum:
    Apple
    Banana
    Mango
    Orange
    Pear
    Strawberry
    Watermelon

struct_subnested<public><native> := struct:
    ElemArray<public><native>:[]struct_a

    ElemArrayArray<public><native>:[][]struct_b

    ElemTuple<public><native>:tuple(struct_c, struct_c)

    ElemOpt<public><native>:?struct_d

    # key=>value
    ElemMap<public><native>:[int]struct_e

    #ElemSub<public><native>:subtype(class_a)

    ElemEnum<public><native>:fruits


#~~~~~~~~~~~~~~~~~~~~~~~~~~~~~~~~~~~~~~~~~~~~~~~~
# Used to test <native_callable> functions in a class

callable_class<public><native> := class:
    Data1<public><native>:string = "Data1"

    Method<public><native_callable>(Arg1:string):string =
        "Method" + Arg1 + Data1

    NativeMethod<public><native><native_callable>(Arg1:string):logic

    Immediate<public><native_callable>(Arg1:string)<suspends>:string =
        Log(Data1)
        test_vni.Immediate()
        "Immediate" + Arg1 + Data1

    NativeImmediate<public><native><native_callable>(Arg1:string)<suspends>:string

    Coroutine<public><native_callable>(Arg1:string)<suspends>:void =
        Log("Coro" + Arg1 + Data1)
        Log("U")
        race:
            block:
                Log("V")
                Wait(3)
                Log("W")
            block:
                Log("X")
                Wait(2)
                Log("Y")
        Log("Z")

    CoroOverride<public><native_callable>(Arg1:string)<suspends>:void =
        Log("Coro" + Arg1 + Data1)
        Log("A")
        race:
            block:
                Log("B")
                Wait(3)
                Log("C")
            block:
                Log("D")
                Wait(2)
                Log("E")
        Log("F")

callable_subclass<public> := class(callable_class):
    Method<override>(Arg1:string):string =
        "method" + Arg1 + Data1

    CoroOverride<override>(Arg1:string)<suspends>:void =
        Log("coro" + Arg1 + Data1)
        Log("a")
        race:
            block:
                Log("b")
                Wait(3)
                Log("c")
            block:
                Log("d")
                Wait(2)
                Log("e")
        Log("f")

#--------------------------------------------------------------------------------------------------
# Used to test null field handling in DataMemberValidation.versetest.
# Encountering a null field in a native class should lead to an immediate runtime error.

native_uobject<public><native> := class<unique> {}

null_object_test_class<public><native> := class:
    MyField<public><native>:native_uobject
    MyOptional<public><native>:?native_uobject
    MyArray<public><native>:[]native_uobject
    MyTuple<public><native>:tuple(native_uobject)
    MyMapK<public><native>:[native_uobject]int
    MyMapV<public><native>:[int]native_uobject

MakeObjectWithNullField<public><native>()<converges>:null_object_test_class
MakeObjectWithNullOptional<public><native>()<converges>:null_object_test_class
MakeObjectWithNullArrayElement<public><native>()<converges>:null_object_test_class
MakeObjectWithNullTupleElement<public><native>()<converges>:null_object_test_class
MakeObjectWithNullMapKey<public><native>()<converges>:null_object_test_class
MakeObjectWithNullMapValue<public><native>()<converges>:null_object_test_class
MakeObjectWithEmptyArray<public><native>()<converges>:null_object_test_class

#--------------------------------------------------------------------------------------------------
# Nested module VNI native tests used in VerseTestScriptCmd
# - allows use in .versetest files and has digest generated for them
test_vni<public> := module:

    #~~~~~~~~~~~~~~~~~~~~~~~~~~~~~~~~~~~~~~~~~~~
    # Functions used for testing native structs

    # Test multiple result using native struct
    MultiResultStruct<public><native>():tuple_struct

    # Test using native strict type as parameter and result
    StructThrough<public><native>(Param1:tuple_struct, Param2:int):tuple_struct

    # Test multiple result using native struct with nested struct
    MultiResultStructNested<public><native>():tuple_struct_nested

    #~~~~~~~~~~~~~~~~~~~~~~~~~~~~~~~~~~~~~~~~~~~
    # Functions used for testing native tuples

    # Test multiple result using native tuple type
    MultiResultTuple<public><native>():tuple(int, float)

    # Test using native tuple type as parameter and result
    # Note that may not have a single tuple parameter - it can be confused with tuple unpacking
    TupleThrough<public><native>(Param1:tuple(int, float), Param2:int):tuple(int, float)

    # Test multiple result using native tuple type with nested newly discovered tuple
    MultiResultTupleNested<public><native>():tuple(int, tuple(float, float))

    # Test multiple result using native tuple type with nested native struct
    MultiResultTupleNestedStruct<public><native>():tuple(int, tuple_struct)

    #~~~~~~~~~~~~~~~~~~~~~~~~~~~~~~~~~~~~~~~~~~~
    # Functions used for testing native callable module coroutines

    # Async and native C++ though completes immediately
    Immediate<public><native>()<suspends>:void

    # Calls `Immediate()` above via Verse `spawn`
    SpawnImmediate<public><native_callable>():task(void) = spawn{Immediate()}

    # Can be called by C++, is async and calls `Immediate()` above via Verse 
    CallImmediate<public><native_callable>()<suspends>:void =
        Log("A")
        Immediate()
        Log("B")

    # Can be called by C++, is async, native C++, completes immediately and has result
    CallNativeImmediate<public><native><native_callable>()<suspends>:string

    # Can be called by C++, is async and calls `race` and other concurrency exprs via Verse 
    CallRace<public><native_callable>()<suspends>:void =
        Log("A")
        race:
            block:
                Log("B")
                Wait(3)
                Log("C")
            block:
                Log("D")
                Wait(2)
                Log("E")
        Log("F")

    # Can be called by C++, is async and calls `race` and other concurrency exprs with arguments via Verse 
    CallRaceArgs<public><native_callable>(Wait1:int, Wait2:int)<suspends>:void =
        Log("A")
        race:
            block:
                Log("B")
                Wait(Wait1)
                Log("C")
            block:
                Log("D")
                Wait(Wait2)
                Log("E")
        Log("F")

    # Calls `CallImmediate()` above via C++
    CallImmediateFromNative<public><native>():void

    # Calls `CallNativeImmediate()` above via C++
    CallNativeImmediateFromNative<public><native>():string

    # Calls `CallNativeRace()` above via C++
    CallRaceFromNative<public><native>():void

    # Calls `CallNativeRaceArgs()` above via C++
    CallRaceArgsFromNative<public><native>():void


    #~~~~~~~~~~~~~~~~~~~~~~~~~~~~~~~~~~~~~~~~~~~~~~~~
    # Coroutine argument and return value marshalling

    # Parameter tuple has 8 byte alignment with 7 bytes of trailing padding,
    # return value has 1 byte alignment but should not overlap with parameter tuple.
    ParamAlignment<public><native>(X:float, Y:logic)<suspends>:logic


    #~~~~~~~~~~~~~~~~~~~~~~~~~~~~~~~~~~~~~~~~~~~~~~~~
    # Functions used for testing native callable instance coroutines

    # Calls callable_class.Method()
    CallClassMethodFromNative<public><native>(Callable:callable_class, Arg1:string):string

    # Calls callable_class.NativeMethod()
    CallClassNativeMethodFromNative<public><native>(Callable:callable_class, Arg1:string):logic

    # Calls callable_class.Immediate()
    CallClassImmediateFromNative<public><native>(Callable:callable_class, Arg1:string):string

    # Calls callable_class.NativeImmediate()
    CallClassNativeImmediateFromNative<public><native>(Callable:callable_class, Arg1:string):string

    # Calls callable_class.Coroutine()
    CallClassCoroutineFromNative<public><native>(Callable:callable_class, Arg1:string):void

    # Calls callable_class.Coroutine()
    CallClassCoroOverrideFromNative<public><native>(Callable:callable_class, Arg1:string):void

#--------------------------------------------------------------------------------------------------
# Parametric type tests
test_native_parametric_type<public><native>(t:type) := class:
    DataMember<public><native>:t
    GetNative<public><native>():t
    Get<public><native_callable>():t=DataMember
GetFromTestNativeParametricType<public><native>(T:test_native_parametric_type(t) where t:type):t
GetFromTestNativeParametricTypeInt<public><native>(T:test_native_parametric_type(int)):int

test_native_parametric_type_inheritance1<public><native>(t:type) := class:
    Property1<public><native>:t
    Get<public>():t = Property1

test_native_parametric_type_inheritance2<public><native>(t:type) := class(test_native_parametric_type_inheritance1(t)):
    Property2<public><native>:t
    Get<override>():t = Property2

verse_vm_todo{
test_native_type_function_sup<public><native> := class {}

test_native_type_function_sub<public><native> := class(test_native_type_function_sup):
    Property<public>:int = 42

TestNativeTypeFunction<public><native>(t:subtype(test_native_type_function_sup)):t
} #verse_vm_todo

test_native_parametric_iface<public>(t:type) := interface:
    Method<public>():t

test_native_parametric_impl<public><native> := class(test_native_parametric_iface(int)):
    Method<override><native>():int

test_awaitable<public><native> := class(awaitable(int)):
    Await<override><native>()<suspends>:int

NativeNestedTupleParameters<public> := module:
    F<public><native>(A:int, (B:int, C:int), D:int):int


UseTarget<public><native>(Target:test_target)<suspends>:void

verse_vm_only {
TestNativeFunctionAsType<public><native>(X:int)<closed><computes>:string

test_native_function_as_type<public><native> := class:
    X<public>:TestNativeFunctionAsType
}

NativeTwoParams<public><native>(X:int, Y:int):void