# Copyright Epic Games, Inc. All Rights Reserved.

using { /Verse.org/Native }

# A whole module that is versioned out
@available{MinUploadedAtFNVersion := 4294967295}
MissingAvailableAtModule := module:                                # versioned out
    SomeFunctionThatWillNeverExist<native><public>():int           # versioned out

TestAvailableAttribute<public> := module:

    # functions
    @available{MinUploadedAtFNVersion := 0}
    AvailableFunctionReturns42<native><public>():int               # exists

    @available{}
    AvailableFunctionReturns43<native><public>():int               # exists

    @available{MinUploadedAtFNVersion := 4294967295}
    NonExistantFunction<native><public>():int                      # versioned out

    # class types
    @available{MinUploadedAtFNVersion := 4294967295}
    NonExistantNativeClass<native><public> := class:               # versioned out
        @available{MinUploadedAtFNVersion := 0}
        Returns42<native><public>():int                            # versioned out from parent

    @available{MinUploadedAtFNVersion := 0}
    AvailableNativeClass<native><public> := class:                 # exists
        Returns42<native><public>():int                            # exists
        @available{MinUploadedAtFNVersion := 4294967295}
        NonExistantReturns43<native><public>():int                 # versioned out

    # duplicate classes
    @available{MinUploadedAtFNVersion := 0}
    MaybeAvailableClass<native><public> := class:                  # exists
        Value13<public><native>:int = 13                           # exists
        @available{MinUploadedAtFNVersion := 4294967295}
        Value13<public><native>:int = 13                           # versioned out
        FuncReturns42<native><public>():int                        # exists
        @available{MinUploadedAtFNVersion := 4294967295}
        FuncReturns42<native><public>():int                        # versioned out

    @available{MinUploadedAtFNVersion := 4294967295}
    MaybeAvailableClass<native><public> := class:                  # versioned out
        FuncReturns42<native><public>():int                        # versioned out
