# Copyright Epic Games, Inc. All Rights Reserved.

unique_class<public> := class<unique>{}
unique_interface<public> := interface<unique>{}