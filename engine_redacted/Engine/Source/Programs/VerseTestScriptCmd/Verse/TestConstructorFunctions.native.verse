# Copyright Epic Games, Inc. All Rights Reserved.

external_constructor_function_class1<public> := class:
    Property1<public>:int

MakeExternalConstructorFunctionClass1<public><constructor>(Arg:int) := external_constructor_function_class1:
    Property1 := Arg
