# Copyright Epic Games, Inc. All Rights Reserved.

UseRational<native><public>(X:rational):rational
