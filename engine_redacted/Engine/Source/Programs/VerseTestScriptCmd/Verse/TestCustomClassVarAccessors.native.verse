# Copyright Epic Games, Inc. All Rights Reserved.
using { /Verse.org/Tests/VerseTestScriptCmd }

accessors_test<public><native> := class:
    block:
        CoroUtils.LogEvent("{X}") #to test calling a native accessor before its created
    var X<public><getter(GetX)><setter(SetX)>:int = external{}
    var XData<private><native>:int = 0

accessors_test_runtime_error<public><native> := class:
    var X<public><getter(GetX)><setter(SetX)>:int = external{}
    var XData<private><native>:int = 0