# Copyright Epic Games, Inc. All Rights Reserved.

# Generator functionality test support used in VerseTestScriptCmd
TestDangerousTypePassing<public> := module:
    IsAMostDangerousNull<public><native>(t:type):logic
