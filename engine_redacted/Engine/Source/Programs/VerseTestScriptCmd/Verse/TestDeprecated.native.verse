# Copyright Epic Games, Inc. All Rights Reserved.

@deprecated{Message := "Test fixture for the @deprecated attribute warning."}
TestDeprecated<public>():void = {}
