# Copyright Epic Games, Inc. All Rights Reserved.

duplication_test_inner_base<native><public> := class<unique>:
    var X<native><public>:int

duplication_test_outer_base<native><public> := class<unique>:
    Inner<public>:duplication_test_inner_base

duplication_test_module<public> := module:
    NativeConstruct<native><public>(Type:subtype(duplication_test_outer_base)):duplication_test_outer_base
    NativeDuplicate<native><public>(Inst:duplication_test_outer_base):duplication_test_outer_base