# Copyright Epic Games, Inc. All Rights Reserved.

# (class, struct, data, function, interface) and both payload forms (V1 empty,
# V2 with args). If VerseFrameworkTestScriptCmd builds with this file included,
# the attribute syntax, brace-init field set, and allowed-scope set are all
# wired up correctly. Does not regression-test FVerseExportMetadataInjection
# diagnostics emission -- the injection visits PublicUser/InternalUser scope
# packages only; end-to-end coverage of the injection itself lives in the UEFN
# upload smoke test.

using { /Verse.org/Diagnostics }
using { /Verse.org/Native }

# class - V1 (empty payload, non-parameterized publishing mutator)
@export_metadata{Category := "PublishingMutators", Key := "TestMutatorV1", Value := ""}
TestExportMetadataClassV1 := class{}

# class - V2 (with args)
@export_metadata{Category := "PublishingMutators", Key := "TestMutatorV2", Value := "args=track:t1,laps:3;"}
TestExportMetadataClassV2 := class{}

# struct
@export_metadata{Category := "TestCategory", Key := "TestStruct", Value := ""}
TestExportMetadataStruct := struct{}

# interface
@export_metadata{Category := "TestCategory", Key := "TestInterface", Value := ""}
TestExportMetadataInterface := interface{}

# function
@export_metadata{Category := "TestCategory", Key := "TestFunction", Value := ""}
TestExportMetadataFunction():void = {}

# data
@export_metadata{Category := "TestCategory", Key := "TestData", Value := ""}
TestExportMetadataData:int = 7

# coexistence with @doc / @experimental / @available
@doc("Coexistence smoke")
@experimental
@available{MinUploadedAtFNVersion := 0}
@export_metadata{Category := "TestCategory", Key := "TestCoexistence", Value := ""}
TestExportMetadataCoexistence := class{}

# attribute scope set; downstream extraction would emit both entries)
@export_metadata{Category := "FirstCategory", Key := "FirstKey", Value := ""}
@export_metadata{Category := "SecondCategory", Key := "SecondKey", Value := ""}
TestExportMetadataMultiple := class{}

# Annotated class can be declared cleanly. End-to-end use-site exercise lives in
# the consumer in a user-scope vpackage that the injection actually visits.
@export_metadata{Category := "TestCategory", Key := "TestUsable", Value := ""}
TestExportMetadataUsable := class{}
