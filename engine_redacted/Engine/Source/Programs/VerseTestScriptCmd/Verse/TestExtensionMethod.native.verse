# Copyright Epic Games, Inc. All Rights Reserved.
 
using{/Verse.org/Native}

@doc("Only here for testing digest of extension methods.")
(S:string).TakeExtension<public>(X:int):string= { for (I -> C : S, I < X) { C } }

(X:int).PlusTwo<public>(Y:int, Z:int):int = X + Y + Z
