# Copyright Epic Games, Inc. All Rights Reserved.

verse_vm_only {
TestFunctionAsType<public>(X:int)<closed><computes>:string = "{X}"

test_function_as_type<public> := class:
    X<public>:TestFunctionAsType

test_function_as_type_iface<public> := interface:
    X<public>:TestFunctionAsType
}
