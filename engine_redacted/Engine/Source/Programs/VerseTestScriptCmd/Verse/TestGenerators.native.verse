# Copyright Epic Games, Inc. All Rights Reserved.

# Generator functionality test support used in VerseTestScriptCmd
TestGenerators<public> := module:
    # generator testing

    TestChild<native><public> := class(TestBase):
        ValueChild<native><public>:int = 0

    TestEmptyChild<native><public> := class(TestBase):

    GetIntegerArray<native><public>()<transacts> : []int

    GetIntegerGenerator<native><public>()<transacts> : generator(int)
    GetAnotherIntegerGenerator<native><public>()<transacts> : generator(int)
    GetEmptyIntegerGenerator<native><public>()<transacts> : generator(int)

    GetFloatGenerator<native><public>()<transacts> : generator(float)
    GetAnotherFloatGenerator<native><public>()<transacts> : generator(float)

    SumFloatGenerator<native><public>(arg:generator(float))<transacts> : float

    MaybeGetAGenerator<public>()<transacts><decides>:?generator(int) = false

    GetTestValueFromAnyGenerator<native><public>(arg:generator(t) where t:type)<transacts> : float

    First<native><public>(iter:generator(t) where t:type)<transacts> : ?t
    Last<native><public>(iter:generator(t) where t:type)<transacts> : ?t

    GetTestBaseGenerator<native><public>()<transacts> : generator(TestBase)
    GetTestChildGenerator<native><public>()<transacts> : generator(TestChild)
    GetTestEmptyChildGenerator<native><public>()<transacts> : generator(TestEmptyChild)
    GetEmptyTestEmptyChildGenerator<native><public>()<transacts> : generator(TestEmptyChild)
