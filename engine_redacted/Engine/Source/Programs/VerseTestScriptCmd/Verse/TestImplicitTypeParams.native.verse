# Copyright Epic Games, Inc. All Rights Reserved.

TestImplicitTypeParams<public> := module:
    # Ensure digest for functions with `where` generates correctly
    Identity<public>(X:t where t:type):t = X
    class1<public> := class:
        Property<public>:int
        Method<public>(X:t where t:type):tuple(int, t) = (Property, X)
    AsClass1<public>(X:t where t:subtype(class1)):t = X
