# Copyright Epic Games, Inc. All Rights Reserved.

# Define a public module and module alias to test digest generation for the module alias.
ModuleToBeAliased<public> := module:
    Magic<public>:int=1234567
ModuleAlias<public> := import(/Verse.org/Tests/VerseTestScriptCmd/ModuleToBeAliased)