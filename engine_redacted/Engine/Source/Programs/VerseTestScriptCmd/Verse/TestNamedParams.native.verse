# Copyright Epic Games, Inc. All Rights Reserved.

TestNamedParamsDefaults<public> := module:
    Int1<public>:int = 3
    Float1<public>:float = 3.0
    String1<public>:string = "!"
    Logic1<public>:logic = true

test_named_params_class1<public> := class<computes>:
    Property<public>:int

test_named_params_enum1<public> := enum:
    A
    B
    C

TestNamedParams<public> := module:
    # Ensure digest for functions with named parameters generates correctly.
    Log<public>(Arg:string, ?Level:int, ?InColor:string = "yellow"):tuple(string, int, string) = (Arg, Level, InColor)
    
    Int1<public>(?Arg:int = 1):int = Arg
    DefaultInt<public>:int = 1
    Int2<public>(?Arg:int = DefaultInt):int = Arg
    Int3<public>(?Arg:int = TestNamedParamsDefaults.Int1):int = Arg
    
    Float1<public>(?Arg:float = 1.0):float = Arg
    DefaultFloat<public>:float = 1.0
    Float2<public>(?Arg:float = DefaultFloat):float = Arg
    Float3<public>(?Arg:float = TestNamedParamsDefaults.Float1):float = Arg    
    
    String1<public>(?Arg:string = "hello"):string = Arg
    DefaultString<public>:string = "hello"
    String2<public>(?Arg:string = DefaultString):string = Arg
    String3<public>(?Arg:string = TestNamedParamsDefaults.String1):string = Arg
    
    Logic1<public>(?Arg:logic = false):logic = Arg
    DefaultLogic<public>:logic = false
    Logic2<public>(?Arg:logic = DefaultLogic):logic = Arg
    Logic3<public>(?Arg:logic = TestNamedParamsDefaults.Logic1):logic = Arg
    
    Class1<public>(?Arg:test_named_params_class1 = test_named_params_class1{Property := 1}):test_named_params_class1 = Arg

    Enum1<public>(?Arg:test_named_params_enum1 = test_named_params_enum1.A):test_named_params_enum1 = Arg

    Identity<public>(X:int):int = X

    Function1<public>(?Arg:type{_(:int):int} = Identity):int = Arg(1)

test_named_params<public> := class:
    DefaultFloat<public>:float = 4.0
    Float4<public>(?Arg:float = (test_named_params:)DefaultFloat):float = Arg

