# Copyright Epic Games, Inc. All Rights Reserved.

mutable_struct<public><native> := struct<computes>:
    X<public><native>:int = 0

mutable_wrapper_struct<public><native> := struct<computes>:
    Struct<public><native>:mutable_struct = mutable_struct{}

native_deeply_mutable<public><native> := class:
    var Array<public><native>:[]int = array{}
    var Struct<public><native>:mutable_struct = mutable_struct{}

    var ArrayArray<public><native>:[][]int = array{}
    var MapArray<public><native>:[int][]int = map{}
    var StructStruct<public><native>:mutable_wrapper_struct = mutable_wrapper_struct{}
