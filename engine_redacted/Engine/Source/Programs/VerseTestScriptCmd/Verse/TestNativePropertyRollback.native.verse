# Copyright Epic Games, Inc. All Rights Reserved.

Numbers<native><public> := enum:
    One
    Two

tiny_native_class<native><public> := class:
    X<native><public>:int = 0

native_property_rollback_test<native><public> := class:
    var MyLogic<native><public>:logic = false
    var MyChar<native><public>:char = 0o00
    var MyChar32<native><public>:char32 = 0u0000
    var MyInt<native><public>:int = 0
    var MyFloat<native><public>:float = 0.0
    var MyRational<native><public>:rational = 0
    var MyOption<native><public>:?float = false
    var MyTuple<native><public>:tuple(char, char, char) = (0o00, 0o00, 0o00)
    var MyArray<native><public>:[]int = array{}
    var MyMap<native><public>:[char]int = map{}
    var MyString<native><public>:string = ""
    var MyObject<native><public>:tiny_native_class = tiny_native_class{X := 0}
    var MyType<native><public>:type = void
