# Copyright Epic Games, Inc. All Rights Reserved.

TestNestedNamedTupleParameters<public> := module:

    SquareSelf<native><public>(A:int):int

    SumTwo<native><public>(?X:int=1, ?Y:int=1):int

    SumThree<native><public>(U:int, (?X:int=1, ?Y:int=1)):int

    SumFour<native><public>(A:int, (X:int, (Y:int, (Z:int)))):int

    SumFourNamed<native><public>(A:int, (X:int, (Y:int, (?Z:int=0)))):int
    
    SumFourArray<native><public>(A:int, (U:[]int, ?X:int=1, ?Y:int=1)):int
    
    sum_class<native><public> := class:
        Const1<public><native>:int
        AddThreeMethod<native><public>(?X:int=1, ?Y:int=1):int

    SumFourFlat<native><public>(A:int, X:int, Y:int, Z:int):int

    SumTwoTuple<native><public>(X:tuple(int, int)):int