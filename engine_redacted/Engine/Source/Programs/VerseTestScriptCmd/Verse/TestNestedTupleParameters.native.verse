# Copyright Epic Games, Inc. All Rights Reserved.

ExternalNestedTupleParameters<public> := module:
    F<public>(A:int, (B:int, C:int), D:int):int =
        A + B + C + D
