# Copyright Epic Games, Inc. All Rights Reserved.

TestParametricFunction<public>(F(:t)<transacts>:u, X:t where t:type, u:type):u = F(X)

TestNestedUniversalQuantifier<public>((X:t where t:type), Y:t):t = X

(X:t).TestParametricExtensionMethod<public>(Y:t where t:type):t = X
