# Copyright Epic Games, Inc. All Rights Reserved.

test_external_parametric_type<public>(t:type, u:type) := class:
    Method<public>(X:t):u=U
    U<public>:u

test_external_parametric_type2<public>(t:type) := class:
    T<public>:test_external_parametric_type(t,int)

test_external_parametric_type3<public>(t:type) := []?t

test_external_parametric_type4<public>(t:type) := interface:
    Method<public>(:comparable)<transacts><decides>:t

test_external_parametric_type5<public>(t:subtype(comparable)) := class(test_external_parametric_type4(t)):
    Property<public>:t
    Method<override>(Arg:comparable)<computes><decides>:t = Property = Arg

test_external_parametric_type6<public>(t:type) := class:
    Get<public>()<reads><decides>:t = RedactedInternal?
