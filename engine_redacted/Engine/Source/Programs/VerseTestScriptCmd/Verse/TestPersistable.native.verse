# Copyright Epic Games, Inc. All Rights Reserved.

external_persistable_class1<public> := class<persistable><final>:
    A<public>:int

external_persistable_struct1<public> := struct<persistable>:
    A<public>:int

external_persistable_enum1<public> := enum<persistable>:
    A, B

native_persistable_class1<public><native> := class<persistable><final>:
    A<public><native>:int

native_persistable_class2<public><native> := class<persistable><final>:
    A<public>:external_persistable_class1

native_persistable_struct1<public><native> := struct<persistable>:
    A<public><native>:int
        
native_persistable_struct2<public><native> := struct<persistable>:
    A<public>:external_persistable_class1

native_persistable_enum1<public><native> := enum<persistable>:
    A, B
