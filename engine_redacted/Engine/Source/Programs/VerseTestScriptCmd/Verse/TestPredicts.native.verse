# Copyright Epic Games, Inc. All Rights Reserved.

TestPredictsManagedObjects<public> := module:
    some_managed_class<public><native> := class:
        Z<public><predicts>:int = 43

        # some unused fields to test codegen/vni/binding
        TestField2<predicts>:int = 3
        TestField3<predicts><native>:int = 4


        CheckContentScopeWorldContext<public><native>()<predicts>:void

    managed_with_setobjectid<public><native> := class:
        GetSecretInt<native><public>()<predicts>:int

    managed_with_net_relevancy<public><native> := class:
        SetNetRelevantForClient<public><native>(:logic, :int)<transacts>:void
        Crash<public><native>()<predicts>:void
        SetYTo3<public><native>()<predicts>:void
        var Y<public><predicts><native>:int = 45

    func_listener_test_class<public><native> := class:
        FireCount<public><native>()<transacts>:int
        FireSelfAt<public><native>(Index:int)<transacts><decides>:any
        FireFuncIDAt<public><native>(Index:int)<transacts><decides>:int

    Setup<public><native>()<transacts>:void

    ServerMakeManagedObject<public><native>()<transacts>:some_managed_class
    ServerMakeManagedObjectWithNetRelevancy<public><native>()<transacts>:managed_with_net_relevancy
    ServerMakeFuncListenerTestObject<public><native>()<transacts>:func_listener_test_class

    ClientPretendManagedObjectsHaventReplicatedYet<public><native>(Pretend:logic)<transacts>:void
    ClientCallPredictsReplicationRetryFunc<public><native>()<transacts>:void

TestPredictsNativeEvent<public> := module:
    some_class<public><native> := class:
        NativeSignalSomeEvent<public><native>(:int)<predicts>:void
        GetSomeEventSubscriptionCount<public><native>()<predicts>:int

TestPredictsPredictedBool<public> := module:
    # wraps a C++ FPredictedBool so unit tests can drive and inspect every state transition
    predicted_bool_test_class<public><native> := class:
        ServerSet<public><native>(:logic)<predicts>:void
        ClientSet<public><native>(:logic)<predicts>:void
        OnServerValueChanged<public><native>(:int)<predicts>:void
        GetValue<public><native>()<predicts>:logic
        GetEpoch<public><native>()<predicts>:int
        GetServerEpoch<public><native>()<predicts>:int
        GetRollbackCount<public><native>()<predicts>:int
        FirePendingTimers<public><native>()<predicts>:void

    ServerMakePredictedBoolTestObject<public><native>()<transacts>:predicted_bool_test_class
