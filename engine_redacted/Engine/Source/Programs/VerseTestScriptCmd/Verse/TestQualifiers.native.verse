# Copyright Epic Games, Inc. All Rights Reserved.

# These definitions need to be exported to the digest with qualifiers in order
# to suppress shadowing errors when compiling the digest.
ModuleWithShadowedDefinitions<public> := module:
    e<public> := enum{A,B}
    i<public> := interface{F<public>():int}
    c<public> := class{X<public>:int=0}
    Module<public>:=module{X<public>:int=0}

    ModuleWithShadowingDefinitions<public> := module:
        (ModuleWithShadowingDefinitions:)e<public> := enum{C,D}
        (ModuleWithShadowingDefinitions:)i<public> := interface((ModuleWithShadowedDefinitions:)i){}
        (ModuleWithShadowingDefinitions:)c<public> := class((ModuleWithShadowedDefinitions:)c){X<override>:int=1}
        (ModuleWithShadowingDefinitions:)Module<public>:=module{X<public>:int=1}

TestLocalQualifiersInDigestGeneration<public> := module:
    Foo<public>((local:)A:int, (local:)B:int):void = {}
