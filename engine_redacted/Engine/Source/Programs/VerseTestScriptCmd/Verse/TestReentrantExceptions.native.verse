# Copyright Epic Games, Inc. All Rights Reserved.

verse_vm_todo
{
ReentrantExceptions<public> := module:
    RaiseBlueprintException<native><public>():void
    CallReentrantCallback<public><native_callable>(N:int):void =
        if (N = 0):
            RaiseBlueprintException()
        else:
            CallReentrant(N - 1)
    CallReentrant<public><native>(N:int):void
}
