# Copyright Epic Games, Inc. All Rights Reserved.

reference_test_base<native><public> := class<unique>:
    ValidateObjectInstanceGraph<native><public>():void