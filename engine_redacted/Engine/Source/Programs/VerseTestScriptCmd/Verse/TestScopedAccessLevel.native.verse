# Copyright Epic Games, Inc. All Rights Reserved.

verse_vm_todo
{
TestScopedAccessLevels<public> := module:
    VerseVerseScope<public> := scoped{/Verse.org/Verse}
    VerseScope<public> := scoped{/Verse.org}

ModuleWithInternals<public> := module:
    InternalEnum<internal> := enum:
        eA
        eB
        eC
}
