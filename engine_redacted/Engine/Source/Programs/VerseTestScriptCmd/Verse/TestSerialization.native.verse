# Copyright Epic Games, Inc. All Rights Reserved.

guid<public><native> := struct<computes>:
    A<public><native>:int
    B<public><native>:int
    C<public><native>:int
    D<public><native>:int

custom_serialization_version<public><native> := struct<computes>:
    GUID<public><native>:guid
    FriendlyName<public><native>:string
    Version<public><native>:int

serialization_mode<public><native> := enum:
    VersionedBinary
    UnversionedBinary
    Text
DefaultSerializationMode<public>:serialization_mode = serialization_mode.VersionedBinary

serialization_version<public><native> := struct<computes>:
    UEVersion<public><native>:int
    CustomSerializationVersions<public><native>:[]custom_serialization_version

MakeDefaultSerializationVersion<public><native>()<converges>:serialization_version

serialized_data<public><native> := struct<computes>:
    Version<public><native>:serialization_version=MakeDefaultSerializationVersion()
    EnableDeltaSerialization<public><native>:logic=true
    Imports<public><native>:[]any=array{}

serializable_class<public><native> := class{}

# Serialize some instance of a subclass of serializable_class simulating the given serialization_version.
# The serialized data doesn't encode what subclass Instance is.
# Caveat: while *loading* older serialization versions is expected to work, saving older serialization versions is not.
# It works for the optional property case I tested, but it's not something you should take for granted.
@rtfm_always_open
Serialize<public><native>(
    Instance:serializable_class,
    ?Imports:[]any=array{},
    ?Version:serialization_version=MakeDefaultSerializationVersion(),
    ?Mode:serialization_mode=DefaultSerializationMode,
    ?EnableDeltaSerialization:logic=false
)<computes>:serialized_data

# Deserialize an instance of some subclass of serializable_class from some serialized_data produced by Serialize.
# The serialized data doesn't encode what subclass of serializable_class was originally saved, and it must be specified
# explicitly through the deserialize_as_class parameter, which allows simulating deserializing with a newer version of
# a class.
@rtfm_always_open
Deserialize<public><native>(SerializedData:serialized_data, deserialize_as_class:subtype(serializable_class))<computes>:deserialize_as_class

# Roundtrips an instance of a class through the serializer and deserializer.
SerializeAndDeserialize<public>(
    Instance:serializable_class,
    deserialize_as_class:subtype(serializable_class),
    ?Imports:[]any=array{},
    ?Mode:serialization_mode=DefaultSerializationMode,
    ?EnableDeltaSerialization:logic=false
)<computes>:deserialize_as_class=
