# Copyright Epic Games, Inc. All Rights Reserved.

verse_vm_todo
{
StmCallbackErrors<public> := module:
    MakeRuntimeErrorOnCommit<public><native>()<transacts>:void
}
