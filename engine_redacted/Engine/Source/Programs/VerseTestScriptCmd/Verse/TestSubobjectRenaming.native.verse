# Copyright Epic Games, Inc. All Rights Reserved.

TestSubobjectRenaming<public> := module:
    a_class<native><public> := class<computes> {}
    b_class<native><public> := class<computes>:
        A<public>:a_class = a_class{}

    GetUObjectPathName<public><native>(Obj:any)<transacts>:string

    Globals<public> := module:
        V1<public>:b_class = b_class{}
        V2<public>:b_class = b_class{}

        A1<public>:[]b_class = array{b_class{}, V2, V2}

        V3<public>:b_class = b_class{}
        M1<public>:[int]b_class = map{2 => b_class{}, 42 => V3}

        V4<public>:b_class = b_class{}
        O1<public>:?b_class = option{V4}
        O2<public>:?b_class = option{b_class{}}

        V5<public>:b_class = b_class{}
        MA1<public>:[int][]b_class = map{23 => array{}, 3 => array{b_class{}, V5}}

    SomeOtherModule<public> := module:
        # this tests naming a UObject instance of a different module, which should *not* rename
        # Globals.V2's UObject to "SomeArray"
        SomeArray<public>:[]b_class = array{Globals.V2}
