# Copyright Epic Games, Inc. All Rights Reserved.

TestTypeAliases<public> := module:
    int_to_int_function<public> := type{_(:int):int}

    int_to_int_function_to_int_to_int_function_function<public> := type{_(:int_to_int_function):int_to_int_function}

    test_struct<public> := struct{X<public>:int}
    struct_to_int_function<public> := type{_(:test_struct):int}

    internal_type_alias<internal> := type{_(:int):int}
    internal_type_alias_to_internal_type_alias_function<public> := type{_(:internal_type_alias):internal_type_alias}

    test_parametric_class<public>(t:type) := class{X<public>:t}

    parametric_class_of_int_to_int<public> := type{_(:test_parametric_class(int)):int}

    parametric_type_alias_parametric_class_to_t<public>(u:type) := type{_(:test_parametric_class(u)):u}
    parametric_type_alias_t_to_parametric_class<public>(v:type) := type{_(:v):test_parametric_class(v)}
    constrained_parametric_type_alias<public>(w:subtype(comparable)) := type{_(:test_parametric_class(w))<transacts><decides>:test_parametric_class(w)}

    test_struct_alias<public> := test_struct

    test_class<public> := class{}
    test_class_alias<public> := test_class