# Copyright Epic Games, Inc. All Rights Reserved.
using { /Verse.org/Native }

Using<public> := module:
    A<public> := module:
    B<public> := module:
    C<public> := module:
        using {A}
        class2<public> := class(class1) {}
    D<public> := module:
        using {B}
        class2<public> := class(class1) {}
