# Copyright Epic Games, Inc. All Rights Reserved.

using { /Verse.org/Native }

# Utility functionality used in VerseTestScriptCmd
TestUtils<public> := module:

    autortfm_some_class<native><public> := class:
        Const<public><native>:int

    MakeAutoRTFMObject<native><public>()<transacts>:autortfm_some_class

    # Gets the old value of a native static, and replaces it with `i`.
    GetAndSetNativeStatic<native><public>(i:int)<transacts>:int

    # Returns the bits of the given float as an integer.
    BitcastFloatToInt<native><public>(f:float):int

    # Cause an AutoRTFM AbortedByLanguage failure, or a no-op if AutoRTFM is disabled.
    # This behaves as a Verse runtime error.
    CauseAutoRTFMLanguageAbort<native><public>()<transacts><decides>:void

    # Cause an AutoRTFM AbortedByRequest failure, or a no-op if AutoRTFM is disabled.
    # This behaves as a Verse transactional failure.
    CauseAutoRTFMRuntimeAbort<native><public>()<transacts><decides>:void

    # Cause an AutoRTFM AbortedByCascadingAbort failure, or a no-op if AutoRTFM is disabled.
    # This behaves as a Verse runtime error.
    CauseAutoRTFMCascadingAbort<native><public>()<transacts>:void

    # Cause an AutoRTFM AbortedByCascadingRetry failure, or a no-op if AutoRTFM is disabled.
    # This behaves as a Verse runtime error, with a retry of the top-most transaction.
    CauseAutoRTFMRetryTransaction<native><public>()<transacts><decides>:void

    # Invoke a <decides> script function passed into a native function and returns true if it succeeds
    InvokeDecides<native><public>(Decide(Arg:int)<computes><decides>:void, Arg:int):logic

    # Invoke a <transacts><decides> script function passed into a native function.
    # InvokeTransacts does not create a nested transaction, so failures in the
    # callback function will fail the transaction of the caller.
    InvokeTransacts<native><public>(Transacts()<transacts><decides>:void)<transacts><decides>:void

    # Same as `InvokeTransacts` but does create a nested transaction
    InvokeTransactsNested<native><public>(Transacts()<transacts><decides>:void)<transacts><decides>:void

    # Test if two classes are equal
    AreClassesEqual<native><public>(lhs:type, rhs:type)<transacts><decides>:type

    verse_vm_todo{
    
    # Returns the Verse paths of all module scoped instances in the current test script, filtered by compatibility with a desired subtype
    # Note that SubTypeOf applies to the type of the _definition_, _not_ the type of the _value_ 
    GetVersePathsOfModuleScopedInstances<native><public>(subtype_of:type):[]string

    # Returns the Verse paths of all classes in the current test script, filtered by compatibility with a desired subtype
    GetVersePathsOfEditableClasses<native><public>(subtype_of:type):[]string

    } #verse_vm_todo

    # A native class with a native getter / setter field.
    native_class<native><public> := class<concrete>():
        # Native getter / setter field, which calls NativeGetter / NativeSetter respectively.
        var NativeGetterSetter<public><getter(NativeGetter)><setter(NativeSetter)> : int = external{}
        # Always returns 42.
        # Does nothing with the value.
