# Copyright Epic Games, Inc. All Rights Reserved.

test_var<public> := class:
    var Property<public>:int

test_var_protected<public> := class:
    var<protected> Property<public>:int

#lenient native var initialization
c_test<native><public> := class:
    block:
        set ValueOfADuringBlock = A
    A<native><public>:int
    var ValueOfADuringBlock<native><public>:int = -1

# test lenient native self via ctor func calls
e_test<native><public> := class:
    var A<native><public>:int = 0
    var B<native><public>:string = "Zero"

# test nested class/struct usage in ctor
f_field<native><public> := class<allocates>:
    var A<native><public>:int = 0
    var B<native><public>:int = 0

f_struct<native><public> := struct<allocates>:
    A<native><public>:[]f_field = array{f_field{}}
    B<native><public>:[]f_field = array{f_field{}}

f_test<native><public> := class:
    block:
        if:
            set F.A[0].A = 10
        if:
            set F.B[0].B = 20
    F<native><public>:f_struct = f_struct{}

#test `SetField` on deeply mutable struct in ctor
g_struct<native><public> := struct<computes>:
    A<native><public>:int = 0

g_test<native><public> := class:
    block:
        set G.A = 10
    var G<native><public>:g_struct = g_struct{}