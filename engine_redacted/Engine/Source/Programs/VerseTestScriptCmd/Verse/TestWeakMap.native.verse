# Copyright Epic Games, Inc. All Rights Reserved.

TestWeakMap<public>(X:weak_map(int, int)):weak_map(int, int) = X
