# Copyright Epic Games, Inc. All Rights Reserved.

using { /Verse.org/Native }


# Like Err, but with only the converges effect.
DebugUnreachable<public><native>()<converges>:false

@doc("Returns true if the debugger is enabled")
IsDebuggerEnabled<public><native>()<computes>:logic

TestCanCallerAccessEpicInternalFromNative<public><native>()<computes>:logic

@doc("Returns the current verse call stack as an array of frames, innermost first")
DebugStackTrace<public><native>()<computes>:[]string

@doc("Verifies that PrintScriptCallstack() includes the VerseVM callstack.")
TestPrintScriptCallstack<public><native>()<computes>:logic

@doc("Verifies that the callstack captured for the most recent Verse runtime error is valid.")
TestRuntimeErrorHasValidCallstack<public><native>()<computes><decides>:void

@doc("Silently raises a runtime warning; returns true if the warning text contained a valid callstack.")
TestRuntimeWarningCallstack<public><native>()<computes>:logic

# Converts any value to a comparable value using some implementation-defined comparison behavior.
ToComparable<public><native>(Any:any)<computes>:comparable

MakeNull<public><native>(t:type)<converges>:t
MakeNullFunction<public><native>():any

TestForEachSubobject<public><native>(Value:any, Callback(Subobject:any):void):void

ToUtf8CodeUnit<public><native>(Char:char)<computes>:int
ToUnicodeCodePoint<public><native>(Char:char32)<computes>:int

ToChar<public><native>(Utf8CodeUnit:int)<computes>:char
ToChar32<public><native>(UnicodeCodePoint:int)<computes>:char32

ToInteger<public><native>(Value:string)<computes>:int

IsAutoRTFMEnabled<public><native>()<transacts><decides>:void
IsAutoRTFMRetryEnabled<public><native>()<transacts><decides>:void

RunGCFull<public><native>()<suspends>:void

NextTick<public><native>()<suspends>:void

IsUObject<public><native>(Value:any)<transacts><decides>:void

GetCVarUObjectProbability<public><native>()<transacts>:float